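# Translucent, unlit, two-sided slime material for the female Palespooky Outlander character body.
# World position offset combines SplineThicken width, a stretch along UV1 V and time-driven sine waves phased by vertex red.
# A tapered SphereMask, distorted by panning swirl noise, drives curve-atlas lookups for edge glow and opacity.
# Color plus edge glow feeds emissive adjustments, while opacity gets a depth fade except on mobile.
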Material /BRCosmetics/Characters/Player/Female/Medium/Bodies/F_MED_Palespooky_Outlander/Materials/M_PaleSpooky_Slime
graph {
 properties {
  bCastRayTracedShadows=false
  bUsedWithMeshDeformer=true
  bUsedWithMorphTargets=true
  bUsedWithSkeletalMesh=true
  bUsedWithSplineMeshes=true
  blendMode=BLEND_Translucent
  pixelDepthOffsetMode=PDOM_Legacy
  shadingModel=MSM_Unlit
  twoSided=true
  usedShadingModels=Unlit
 }
 nodes [
  n1 = ScalarParameter() {defaultValue=3, parameterName=WidthBase}
  n2 = ScalarParameter() {defaultValue=2, parameterName=WidthTip}
  n3 = AppendVector(A=Constant2Vector(), B=ScalarParameter() {defaultValue=8, parameterName=UpperStretch})
  n4 = OneMinus(None=ComponentMask(None=TextureCoordinate() {coordinateIndex=1}) {g=true})
  n5 = Subtract(A=Power(Base=@n4) {constExponent=8}) {constB=0.2}
  n6 = Multiply(A=VertexColor().R) {constB=2}
  n7 = Time()
  n8 = Add(
   A=Divide(A=@n4, B=ScalarParameter() {defaultValue=0.8, parameterName=WPOWaveSize})
   B=Multiply(A=Add(A=@n6, B=@n7), B=ScalarParameter() {defaultValue=0.3, parameterName=WPOWaveSpeed})
  )
  n9 = Sine(None=@n8) {period=1.22}
  n10 = Sine(None=@n8) {period=1.16}
  n11 = Sine(None=@n8) {period=1.61}
  n12 = Multiply(
   A=AppendVector(A=AppendVector(A=@n9, B=@n10), B=Multiply(A=@n11) {constB=0.2})
   B=ScalarParameter() {defaultValue=2, parameterName=WPOWaveStrength}
  )
  Add(
   A=Add(
    A=MaterialFunctionCall(WidthBase=@n1, WidthTip=@n2) {materialFunction=&SplineThicken}.WorldPositionOffset
    B=LinearInterpolate(B=@n3, Alpha=@n5) {constB=0}
   )
   B=Multiply(A=@n4, B=Add(A=@n3, B=@n12))
  )
  n13 = VectorParameter() {parameterName=Color}
  n14 = Add(A=ComponentMask(None=TextureCoordinate()) {g=true}, B=ScalarParameter() {defaultValue=0.2, parameterName=Offset_color_Fresnel}) {constB=0}
  n15 = ScalarParameter() {defaultValue=2, parameterName=Taper1}
  n16 = ComponentMask(None=TextureCoordinate()) {r=true}
  n17 = LinearInterpolate(A=Multiply(A=@n15) {constB=-1}, B=Add(A=@n15), Alpha=@n16) {constA=-0.1, constB=1.1}
  n18 = ScalarParameter() {defaultValue=0.6, parameterName="Taper1_positon(UbaseV,0-1)"}
  n19 = ComponentMask(None=TextureCoordinate()) {g=true}
  n20 = Multiply(A=Divide(A=Constant() {r=1}, B=@n18), B=@n19)
  n21 = Constant() {r=1}
  n22 = LinearInterpolate(A=Divide(A=@n21, B=Subtract(A=@n21, B=@n18)), Alpha=@n19) {constB=0}
  n23 = OneMinus(None=Power(Base=If(A=@n20, "A > B"=@n22, "A == B"=@n20, "A < B"=@n20) {constB=1}, Exp=ScalarParameter() {defaultValue=4, parameterName=Taper1_power}) {constExponent=4})
  n24 = VertexColor()
  n25 = AppendVector(A=LinearInterpolate(Alpha=@n24.R) {constA=-0.15, constB=0.15}, B=LinearInterpolate(Alpha=@n24.R) {constA=0.09, constB=0.2})
  n26 = VectorParameter() {parameterName=sphere_bottom_texture, defaultValue=<rgba 0.1 0.9 1 1>}
  n27 = Multiply(A=Time(), B=ScalarParameter() {defaultValue=0.9, parameterName=PanSpeed})
  n28 = Add(A=Panner(Time=@n27, Speed=Multiply(A=@n25, B=@n26.B)) {speedX=0.1, speedY=0.1}, B=ScalarParameter() {parameterName=MotionOffset})
  n29 = ComponentMask(None=@n26.RGB) {g=true, r=true}
  n30 = Multiply(
   A=TextureSample(UVs=Multiply(A=@n28, B=@n29)) {samplerType=SAMPLERTYPE_Normal, texture=/Game/Global/Textures/Default/Noise/T_SwirlNoise_N}.R
   B=ScalarParameter() {defaultValue=0.3, parameterName=Distort1}
  ) {constB=0.3}
  n31 = ScalarParameter() {defaultValue=0.65, group=Flame_Area_Ctrl, parameterName=V_tile1}
  n32 = LinearInterpolate(B=@n31, Alpha=ComponentMask(None=TextureCoordinate()) {g=true})
  n33 = ScalarParameter() {defaultValue=0.0452, parameterName=Sphere1_offset}
  n34 = SphereMask(
   A=AppendVector(A=LinearInterpolate(A=@n17, B=@n16, Alpha=@n23), B=Add(A=@n30, B=@n32))
   B=AppendVector(A=Constant() {r=0.5}, B=Add(A=@n31, B=@n33))
   Radius=ScalarParameter() {defaultValue=0.6, group=Flame_Area_Ctrl, parameterName=Radius1, sortPriority=1}
   Hardness=ScalarParameter() {defaultValue=0.4, group=Flame_Area_Ctrl, parameterName=Hardness1, sortPriority=2}
  ) {attenuationRadius=1.5, hardnessPercent=0}
  n35 = CurveAtlasRowParameter(CurveTime=@n34) {atlas=&AthenaCurveAtlas01, curve=/Game/Athena/LinearColorCurves/Curve_Birthday2019Pickaxe, defaultValue=112}
  n36 = Multiply(
   A=LinearInterpolate(Alpha=Power(Base=@n14)) {constB=1.5}
   B=Multiply(A=VectorParameter() {parameterName=EdgeColor, defaultValue=<rgba 0.0422 0.3542 0.5 1>}.RGB, B=@n35.B)
  )
  MaterialFunctionCall(Emissive=Multiply(A=Add(A=@n13.RGB, B=@n36), B=ScalarParameter() {defaultValue=2, parameterName=Emissive_intensity})) {materialFunction=&MF_FX_EmissiveAdjustments}
  n37 = TextureSample(UVs=Multiply(A=Panner(Time=@n27) {speedY=0.1}, B=Constant2Vector() {g=1, r=0.1})) {samplerType=SAMPLERTYPE_Normal, texture=&T_SwirlNoise_N}
  n38 = TextureCoordinate()
  n39 = CurveAtlasRowParameter(CurveTime=OneMinus(None=ComponentMask(None=@n38) {g=true})) {
   atlas=&AthenaCurveAtlas01
   curve=/Game/Characters/Player/Male/Medium/Bodies/M_MED_VoyagerRemix/FX/Formal/Curve_VoyagerRemix2
   defaultValue=119
  }
  n40 = CurveAtlasRowParameter(CurveTime=Multiply(A=LinearInterpolate(A=@n37.R, Alpha=@n39.G), B=@n34)) {
   atlas=&AthenaCurveAtlas01
   curve=&Curve_VoyagerRemix2
   defaultValue=119
  }
  n41 = Multiply(A=@n39.G, B=@n40.R) {constA=1}
  n42 = DepthFade(Opacity=@n41, FadeDistance=ScalarParameter() {defaultValue=2, parameterName=DepthFade}) {fadeDistanceDefault=2}
  Multiply(
   A=ShadingPathSwitch(Default=@n42, Deferred=@n42, Mobile=@n41) {desc=Migrated from FeatureLevelSwitch}
   B=ScalarParameter() {defaultValue=0.9, parameterName=OverallOpacity}
  )
 ]
}
