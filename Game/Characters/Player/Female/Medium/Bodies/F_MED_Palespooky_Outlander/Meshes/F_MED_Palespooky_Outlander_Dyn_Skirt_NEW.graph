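# Physics asset for the dynamic skirt of the female medium Palespooky Outlander player character.
# Five six-link chains (back, back-left, back-right, left, right) hang from pelvis-linked master bodies, each a box followed by capsules with angular limits, tightest on the left.
# Cross-constraints with free rotation tie some neighboring chains together.
# Pelvis, legs, right foot (no left), root spheres and a large flat sim-space box supply the surrounding body shapes.
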PhysicsAsset /Game/Characters/Player/Female/Medium/Bodies/F_MED_Palespooky_Outlander/Meshes/F_MED_Palespooky_Outlander_Dyn_Skirt_NEW
bodies {
 calf_l [{shapeType=Capsule, length=48.3631, radius=10.2689, center=<-29.3305 -3.2389 -0.1386>, rotation=<rot 87.128 -179.9944 -180>}]
 calf_r [{shapeType=Capsule, length=43.9752, radius=13.5214, center=<32.9029 3.7773 -0.8658>, rotation=<rot 87.1161 -179.9944 -180>}]
 dyn_dress_bk_01 [{shapeType=Box, extentX=13.4375, extentY=13.0934, extentZ=2.7489, center=<5.954 0 -0.0003>}]
 dyn_dress_bk_02 [{shapeType=Capsule, length=13.9619, radius=3, shapeName=dyn_dress_l_02_capsule, center=<5.6266 0 -0.0001>, rotation=<rot 0 0 90>}]
 dyn_dress_bk_03 [{shapeType=Capsule, length=11.482, radius=5.3286, shapeName=dyn_dress_r_03_capsule, center=<5.2086 0 0.0002>, rotation=<rot 0 0 -90>}]
 dyn_dress_bk_04 [{shapeType=Capsule, length=20.8344, radius=5.9291, shapeName=dyn_dress_r_04_capsule, center=<5.3637 0 0.0002>, rotation=<rot 0 0 90>}]
 dyn_dress_bk_05 [{shapeType=Capsule, length=23.5796, radius=5.4726, shapeName=dyn_dress_l_05_capsule, center=<4.7244 0 0>, rotation=<rot 0 0 90>}]
 dyn_dress_bk_06 [{shapeType=Capsule, length=28.3363, radius=5.3798, shapeName=dyn_dress_r_06_capsule, center=<6.4814 0 0.0002>, rotation=<rot 0 0 90>}]
 dyn_dress_bk_MSTR [{shapeType=Capsule, length=0.2, radius=0.2}]
 dyn_dress_bk_l_01 [{shapeType=Box, extentX=13.4375, extentY=10.4803, extentZ=3.3978, center=<5.954 0 -0.0003>, rotation=<rot 0 0 10.0004>}]
 dyn_dress_bk_l_02 [{shapeType=Capsule, length=13.9619, radius=3, shapeName=dyn_dress_l_02_capsule, center=<5.6266 0 -0.0001>, rotation=<rot 0 0 90>}]
 dyn_dress_bk_l_03 [{shapeType=Capsule, length=9.6809, radius=3.5705, shapeName=dyn_dress_r_03_capsule, center=<5.2086 0 0.0002>, rotation=<rot 0 0 -90>}]
 dyn_dress_bk_l_04 [{shapeType=Capsule, length=10.5312, radius=4.171, shapeName=dyn_dress_r_04_capsule, center=<5.3637 0 0.0002>, rotation=<rot 0 0 90>}]
 dyn_dress_bk_l_05 [{shapeType=Capsule, length=14.3399, radius=3.133, shapeName=dyn_dress_l_05_capsule, center=<4.7244 0 0>, rotation=<rot 0 0 90>}]
 dyn_dress_bk_l_06 [{shapeType=Capsule, length=20.356, radius=3.4513, shapeName=dyn_dress_r_06_capsule, center=<6.4814 0 0.0002>, rotation=<rot 0 0 90>}]
 dyn_dress_bk_l_MSTR [{shapeType=Capsule, length=0.2, radius=0.2}]
 dyn_dress_bk_r_01 [{shapeType=Box, extentX=13.4375, extentY=15.8537, extentZ=2.7524, center=<5.954 0 -0.0003>}]
 dyn_dress_bk_r_02 [{shapeType=Capsule, length=12.2372, radius=4.7248, shapeName=dyn_dress_l_02_capsule, center=<5.6266 0 3>, rotation=<rot 0 0 90>}]
 dyn_dress_bk_r_03 [{shapeType=Capsule, length=36.1505, radius=5.9417, shapeName=dyn_dress_r_03_capsule, center=<5.2086 0 3>, rotation=<rot 0 0 -90>}]
 dyn_dress_bk_r_04 [{shapeType=Capsule, length=35.4017, radius=6.9927, shapeName=dyn_dress_r_04_capsule, center=<5.3637 0 3>, rotation=<rot 0 0 90>}]
 dyn_dress_bk_r_05 [{shapeType=Capsule, length=35.3387, radius=6.51, shapeName=dyn_dress_l_05_capsule, center=<4.7244 0 3>, rotation=<rot 0 0 90>}]
 dyn_dress_bk_r_06 [{shapeType=Capsule, length=35.5612, radius=6.8283, shapeName=dyn_dress_r_06_capsule, center=<6.4814 0 3>, rotation=<rot 0 0 90>}]
 dyn_dress_bk_r_MSTR [{shapeType=Capsule, length=0.2, radius=0.2}]
 dyn_dress_l_01 [{shapeType=Box, extentX=13.4375, extentY=15.8537, extentZ=3.6637, center=<5.954 0 -0.0003>}]
 dyn_dress_l_02 [{shapeType=Capsule, length=13.1199, radius=3.8421, shapeName=dyn_dress_l_02_capsule, center=<5.6266 0 3>, rotation=<rot 0 0 90>}]
 dyn_dress_l_03 [{shapeType=Capsule, length=8.8388, radius=4.4125, shapeName=dyn_dress_r_03_capsule, center=<6.3616 0 3>, rotation=<rot 0 0 -90>}]
 dyn_dress_l_04 [{shapeType=Capsule, length=9.6892, radius=5.013, shapeName=dyn_dress_r_04_capsule, center=<5.3637 0 3>, rotation=<rot 0 0 90>}]
 dyn_dress_l_05 [{shapeType=Capsule, length=13.4978, radius=3.9751, shapeName=dyn_dress_l_05_capsule, center=<6.7003 0 3>, rotation=<rot 0 0 -90>}]
 dyn_dress_l_06 [{shapeType=Capsule, length=19.5139, radius=4.2934, shapeName=dyn_dress_r_06_capsule, center=<5.0466 0 3>, rotation=<rot 0 0 90>}]
 dyn_dress_l_MSTR [{shapeType=Capsule, length=0.2, radius=0.2}]
 dyn_dress_r_01 [{shapeType=Box, extentX=13.4375, extentY=12.6653, extentZ=2.8441, center=<5.954 0 -0.0003>, rotation=<rot 0 0 -3.3758>}]
 dyn_dress_r_02 [{shapeType=Capsule, length=13.9619, radius=3, shapeName=dyn_dress_l_02_capsule, center=<5.6266 0 3>, rotation=<rot 0 0 90>}]
 dyn_dress_r_03 [{shapeType=Capsule, length=8.1033, radius=3.5705, shapeName=dyn_dress_r_03_capsule, center=<7.0965 0 3>, rotation=<rot 0 0 -90>}]
 dyn_dress_r_04 [{shapeType=Capsule, length=17.3635, radius=6.9414, shapeName=dyn_dress_r_04_capsule, center=<5.3637 0 3>, rotation=<rot 0 0 90>}]
 dyn_dress_r_05 [{shapeType=Capsule, length=19.7083, radius=5.9035, shapeName=dyn_dress_l_05_capsule, center=<6.2727 0 3>, rotation=<rot 0 0 90>}]
 dyn_dress_r_06 [{shapeType=Capsule, length=19.5864, radius=6.2218, shapeName=dyn_dress_r_06_capsule, center=<4.4763 0 3>, rotation=<rot 0 0 90>}]
 dyn_dress_r_MSTR [{shapeType=Capsule, length=0.2, radius=0.2}]
 dyn_root_simspace [{shapeType=Box, extentX=500, extentY=500, extentZ=20, shapeName=dyn_root_simspace_box, center=<0 0 -6.4521>, rotation=<rot 0 -90 0>}]
 foot_r [{shapeType=Capsule, length=4.0765, radius=9.7403, shapeName=foot_r_capsule, center=<4.8882 -4.8881 -0.0001>, rotation=<rot 90 0 44.9989>}]
 pelvis [{shapeType=Capsule, length=8.5443, radius=8.218, center=<-5.3601 -2.3275 -0.0005>}]
 root [
  {shapeType=Sphere, radius=24.9585, shapeName=root_sphere, center=<0.0002 2.4787 8.5972>}
  {shapeType=Sphere, radius=183.4375, shapeName=root_sphere, center=<10 -8.6403 -175.5073>}
 ]
 thigh_l [{shapeType=Capsule, length=31.0312, radius=7.2673, center=<-22.5439 -2.5 1.9881>, rotation=<rot 86.0417 -180 -180>}]
 thigh_r [{shapeType=Capsule, length=32.1341, radius=6.1644, center=<22.5439 2.5 -1.9881>, rotation=<rot 85.9372 -180 -180>}]
}
constraints [
 {bone1Name=dyn_dress_bk_01, bone2Name=dyn_dress_bk_MSTR, swing1LimitDegrees=45, swing2LimitDegrees=45, twistLimitDegrees=5}
 {bone1Name=dyn_dress_bk_02, bone2Name=dyn_dress_bk_01, swing1LimitDegrees=45, swing2LimitDegrees=45, twistLimitDegrees=5}
 {bone1Name=dyn_dress_bk_03, bone2Name=dyn_dress_bk_02, swing1LimitDegrees=45, swing2LimitDegrees=45, twistLimitDegrees=5}
 {bone1Name=dyn_dress_bk_04, bone2Name=dyn_dress_bk_03, swing1LimitDegrees=45, swing2LimitDegrees=45, twistLimitDegrees=5}
 {bone1Name=dyn_dress_bk_05, bone2Name=dyn_dress_bk_04, swing1LimitDegrees=45, swing2LimitDegrees=45, twistLimitDegrees=5}
 {bone1Name=dyn_dress_bk_06, bone2Name=dyn_dress_bk_05, swing1LimitDegrees=45, swing2LimitDegrees=45, twistLimitDegrees=5}
 {bone1Name=dyn_dress_bk_l_01, bone2Name=dyn_dress_bk_l_MSTR, swing1LimitDegrees=45, swing2LimitDegrees=45, twistLimitDegrees=5}
 {bone1Name=dyn_dress_bk_l_02, bone2Name=dyn_dress_bk_l_01, swing1LimitDegrees=45, swing2LimitDegrees=45, twistLimitDegrees=5}
 {bone1Name=dyn_dress_bk_l_03, bone2Name=dyn_dress_bk_l_02, swing1LimitDegrees=45, swing2LimitDegrees=45, twistLimitDegrees=5}
 {bone1Name=dyn_dress_bk_l_04, bone2Name=dyn_dress_bk_l_03, swing1LimitDegrees=45, swing2LimitDegrees=45, twistLimitDegrees=5}
 {bone1Name=dyn_dress_bk_l_05, bone2Name=dyn_dress_bk_l_04, swing1LimitDegrees=45, swing2LimitDegrees=45, twistLimitDegrees=5}
 {bone1Name=dyn_dress_bk_l_06, bone2Name=dyn_dress_bk_l_05, swing1LimitDegrees=45, swing2LimitDegrees=45, twistLimitDegrees=5}
 {bone1Name=dyn_dress_l_01, bone2Name=dyn_dress_l_MSTR, swing1LimitDegrees=10, swing2LimitDegrees=10, twistLimitDegrees=5}
 {bone1Name=dyn_dress_l_02, bone2Name=dyn_dress_l_01, swing1LimitDegrees=10, swing2LimitDegrees=5, twistLimitDegrees=5}
 {bone1Name=dyn_dress_l_03, bone2Name=dyn_dress_l_02, swing1LimitDegrees=2, swing2LimitDegrees=5, twistLimitDegrees=5}
 {bone1Name=dyn_dress_l_04, bone2Name=dyn_dress_l_03, swing1LimitDegrees=2, swing2LimitDegrees=5, twistLimitDegrees=5}
 {bone1Name=dyn_dress_l_05, bone2Name=dyn_dress_l_04, swing1LimitDegrees=2, swing2LimitDegrees=5, twistLimitDegrees=5}
 {bone1Name=dyn_dress_l_06, bone2Name=dyn_dress_l_05, swing1LimitDegrees=2, swing2LimitDegrees=5, twistLimitDegrees=5}
 {bone1Name=dyn_dress_bk_r_01, bone2Name=dyn_dress_bk_r_MSTR, swing1LimitDegrees=45, swing2LimitDegrees=45, twistLimitDegrees=5}
 {bone1Name=dyn_dress_bk_r_02, bone2Name=dyn_dress_bk_r_01, swing1LimitDegrees=10, swing2LimitDegrees=45, twistLimitDegrees=5}
 {bone1Name=dyn_dress_bk_r_03, bone2Name=dyn_dress_bk_r_02, swing1LimitDegrees=10, swing2LimitDegrees=45, twistLimitDegrees=5}
 {bone1Name=dyn_dress_bk_r_04, bone2Name=dyn_dress_bk_r_03, swing1LimitDegrees=10, swing2LimitDegrees=45, twistLimitDegrees=5}
 {bone1Name=dyn_dress_bk_r_05, bone2Name=dyn_dress_bk_r_04, swing1LimitDegrees=10, swing2LimitDegrees=45, twistLimitDegrees=5}
 {bone1Name=dyn_dress_bk_r_06, bone2Name=dyn_dress_bk_r_05, swing1LimitDegrees=10, swing2LimitDegrees=45, twistLimitDegrees=5}
 {bone1Name=dyn_dress_r_01, bone2Name=dyn_dress_r_MSTR, swing1LimitDegrees=10, swing2LimitDegrees=10, twistLimitDegrees=5}
 {bone1Name=dyn_dress_r_02, bone2Name=dyn_dress_r_01, swing1LimitDegrees=10, swing2LimitDegrees=20, twistLimitDegrees=20}
 {bone1Name=dyn_dress_r_03, bone2Name=dyn_dress_r_02, swing1LimitDegrees=10, swing2LimitDegrees=20, twistLimitDegrees=20}
 {bone1Name=dyn_dress_r_04, bone2Name=dyn_dress_r_03, swing1LimitDegrees=10, swing2LimitDegrees=20, twistLimitDegrees=20}
 {bone1Name=dyn_dress_r_05, bone2Name=dyn_dress_r_04, swing1LimitDegrees=10, swing2LimitDegrees=20, twistLimitDegrees=20}
 {bone1Name=dyn_dress_r_06, bone2Name=dyn_dress_r_05, swing1LimitDegrees=10, swing2LimitDegrees=20, twistLimitDegrees=20}
 {
  bone1Name=dyn_dress_bk_MSTR
  bone2Name=pelvis
  swing1LimitDegrees=45
  swing1Motion=Free
  swing2LimitDegrees=45
  swing2Motion=Free
  twistLimitDegrees=45
  twistMotion=Free
 }
 {
  bone1Name=dyn_dress_bk_l_MSTR
  bone2Name=pelvis
  swing1LimitDegrees=45
  swing1Motion=Free
  swing2LimitDegrees=45
  swing2Motion=Free
  twistLimitDegrees=45
  twistMotion=Free
 }
 {
  bone1Name=dyn_dress_l_MSTR
  bone2Name=pelvis
  swing1LimitDegrees=45
  swing1Motion=Free
  swing2LimitDegrees=45
  swing2Motion=Free
  twistLimitDegrees=45
  twistMotion=Free
 }
 {
  bone1Name=dyn_dress_bk_r_MSTR
  bone2Name=pelvis
  swing1LimitDegrees=45
  swing1Motion=Free
  swing2LimitDegrees=45
  swing2Motion=Free
  twistLimitDegrees=45
  twistMotion=Free
 }
 {
  bone1Name=dyn_dress_r_MSTR
  bone2Name=pelvis
  swing1LimitDegrees=45
  swing1Motion=Free
  swing2LimitDegrees=45
  swing2Motion=Free
  twistLimitDegrees=45
  twistMotion=Free
 }
 {
  bone1Name=thigh_l
  bone2Name=pelvis
  swing1LimitDegrees=45
  swing1Motion=Free
  swing2LimitDegrees=45
  swing2Motion=Free
  twistLimitDegrees=45
  twistMotion=Free
 }
 {
  bone1Name=calf_l
  bone2Name=thigh_l
  swing1LimitDegrees=45
  swing1Motion=Free
  swing2LimitDegrees=45
  swing2Motion=Free
  twistLimitDegrees=45
  twistMotion=Free
 }
 {
  bone1Name=thigh_r
  bone2Name=pelvis
  swing1LimitDegrees=45
  swing1Motion=Free
  swing2LimitDegrees=45
  swing2Motion=Free
  twistLimitDegrees=45
  twistMotion=Free
 }
 {
  bone1Name=calf_r
  bone2Name=thigh_r
  swing1LimitDegrees=45
  swing1Motion=Free
  swing2LimitDegrees=45
  swing2Motion=Free
  twistLimitDegrees=45
  twistMotion=Free
 }
 {
  bone1Name=dyn_dress_bk_03
  bone2Name=dyn_dress_bk_l_03
  swing1LimitDegrees=45
  swing1Motion=Free
  swing2LimitDegrees=45
  swing2Motion=Free
  twistLimitDegrees=45
  twistMotion=Free
 }
 {
  bone1Name=dyn_dress_bk_05
  bone2Name=dyn_dress_bk_l_05
  swing1LimitDegrees=45
  swing1Motion=Free
  swing2LimitDegrees=45
  swing2Motion=Free
  twistLimitDegrees=45
  twistMotion=Free
 }
 {
  bone1Name=dyn_dress_l_03
  bone2Name=dyn_dress_bk_l_03
  swing1LimitDegrees=45
  swing1Motion=Free
  swing2LimitDegrees=45
  swing2Motion=Free
  twistLimitDegrees=45
  twistMotion=Free
 }
 {
  bone1Name=pelvis
  bone2Name=root
  swing1LimitDegrees=45
  swing1Motion=Free
  swing2LimitDegrees=45
  swing2Motion=Free
  twistLimitDegrees=45
  twistMotion=Free
 }
 {
  bone1Name=foot_r
  bone2Name=calf_r
  swing1LimitDegrees=45
  swing1Motion=Free
  swing2LimitDegrees=45
  swing2Motion=Free
  twistLimitDegrees=45
  twistMotion=Free
 }
 {
  bone1Name=dyn_dress_bk_03
  bone2Name=dyn_dress_bk_r_03
  swing1LimitDegrees=45
  swing1Motion=Free
  swing2LimitDegrees=45
  swing2Motion=Free
  twistLimitDegrees=45
  twistMotion=Free
 }
 {
  bone1Name=dyn_dress_bk_04
  bone2Name=dyn_dress_bk_r_04
  swing1LimitDegrees=45
  swing1Motion=Free
  swing2LimitDegrees=45
  swing2Motion=Free
  twistLimitDegrees=45
  twistMotion=Free
 }
 {
  bone1Name=dyn_root_simspace
  bone2Name=root
  swing1LimitDegrees=45
  swing1Motion=Free
  swing2LimitDegrees=45
  swing2Motion=Free
  twistLimitDegrees=45
  twistMotion=Free
 }
]
