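# Physics asset for the dynamic sleeves of the female medium Palespooky Outlander player body.
# It gives each side a box body on the second sleeve bone and one on the second lower-arm twist bone.
# A constraint on each side joins the sleeve to its forearm with matching swing and twist limits, bounding sleeve movement relative to the arm.
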PhysicsAsset /Game/Characters/Player/Female/Medium/Bodies/F_MED_Palespooky_Outlander/Meshes/F_MED_Palespooky_Outlander_Dyn_Sleeve
bodies {
 dyn_sleeve_l_02 [{shapeType=Box, extentX=8.3066, extentY=11.551, extentZ=3.7702, center=<4.2412 0.0027 -0.0173>, rotation=<rot -13.0233 1.9493 1.7721>}]
 dyn_sleeve_r_02 [{shapeType=Box, extentX=9.3801, extentY=7.6978, extentZ=3.8105, center=<5.0755 0 0>}]
 lowerarm_twist_02_l [{shapeType=Box, extentX=15.8229, extentY=8.1889, extentZ=4.4416, center=<0 -0.1015 -0.2765>, rotation=<rot 0 0 20.1602>}]
 lowerarm_twist_02_r [{shapeType=Box, extentX=16.5213, extentY=8.698, extentZ=3.7081, center=<0 0.1949 0.8041>, rotation=<rot -0.0001 -0.0001 13.6267>}]
}
constraints [
 {bone1Name=dyn_sleeve_r_02, bone2Name=lowerarm_twist_02_r, swing1LimitDegrees=45, swing2LimitDegrees=45, twistLimitDegrees=45}
 {bone1Name=dyn_sleeve_l_02, bone2Name=lowerarm_twist_02_l, swing1LimitDegrees=45, swing2LimitDegrees=45, twistLimitDegrees=45}
]
