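# Post-process style AnimBlueprint for the Palespooky Outlander female body cosmetic, adding secondary motion to dress, skirt, sleeve and hand/foot VFX bones.
# RigidBody nodes simulate these parts, with Anim Dynamics on the front skirt.
# Skirt chains branch from cached 'preSkirt', using thigh-driven constraints before and after simulation to follow the legs.
# An emote/frontend bool blends between skirt setups with different physics assets, while crouch and emote values drive corrective bone offsets.
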AnimBlueprint /BRCosmetics/Characters/Player/Female/Medium/Bodies/F_MED_Palespooky_Outlander/Meshes/F_MED_Palespooky_Outlander_AnimBP
target_skeleton=/Game/Characters/Player/Female/Medium/Bodies/F_MED_Palespooky_Outlander/Meshes/F_MED_Palespooky_Outlander_Skeleton
graphs [
 AnimGraph {nodes [
   n1 = Get crouching() -> (crouching:bool)
   n2 = "dyn_dress_bk_r_MSTR.rotateY = calf_twist_01_r.rotateZ - Bone Driven Controller"!(ComponentPose:Component Space Pose Link Structure, bAlphaBoolEnabled:bool=@n1) -> (Pose:Component Space Pose Link Structure) {
    is=title AnimGraphNode_BoneDrivenController
    alphaInputType=Bool
    bAffectTargetRotationY=true
    sourceComponent=RotationZ
    alphaBoolBlend {blendOption=Linear}
    sourceBone {boneName=calf_twist_01_r}
    targetBone {boneName=dyn_dress_bk_r_MSTR}
   }
   "dyn_dress_r_MSTR.rotateY = calf_twist_01_r.rotateZ - Bone Driven Controller"!(ComponentPose:Component Space Pose Link Structure=@n2, bAlphaBoolEnabled:bool=@n1) -> (Pose:Component Space Pose Link Structure) {
    is=title AnimGraphNode_BoneDrivenController
    alphaInputType=Bool
    bAffectTargetRotationY=true
    sourceComponent=RotationZ
    alphaBoolBlend {blendOption=Linear}
    sourceBone {boneName=calf_twist_01_r}
    targetBone {boneName=dyn_dress_r_MSTR}
   }
   n3 = "Constraint - dyn_dress_bk_r_05 (R)"!(ComponentPose, Alpha=0.8) -> (Pose) {
    alphaBoolBlend {blendOption=Linear}
    boneToModify {boneName=dyn_dress_bk_r_05}
    constraintSetup [{transformType=Rotation, targetBone {boneName=calf_twist_01_r}}]
    constraintWeights [1]
   }
   "Constraint - dyn_dress_r_05 (R)"!(ComponentPose=@n3, Alpha=0.8) -> (Pose) {
    alphaBoolBlend {blendOption=Linear}
    boneToModify {boneName=dyn_dress_r_05}
    constraintSetup [{transformType=Rotation, targetBone {boneName=calf_twist_01_r}}]
    constraintWeights [1]
   }
   "Transform (Modify) Bone - Bone: dyn_dress_r_04"(Alpha=1, Translation="0,0,0", Rotation="50,-30,0", Scale="1,1,1") {
    rotationMode=BMM_Additive
    rotationSpace=BCS_BoneSpace
    translationMode=BMM_Additive
    translationSpace=BCS_BoneSpace
    alphaBoolBlend {blendOption=Linear}
    alphaScaleBiasClamp {bInterpResult=true, interpSpeedDecreasing=30, interpSpeedIncreasing=30}
    boneToModify {boneName=dyn_dress_r_04}
    rotation {pitch=50, yaw=-30}
   }
   n4 = Get crouch alpha() -> (crouch alpha:double)
   n5 = "Transform (Modify) Bone - Bone: dyn_dress_bk_r_03"(Alpha=@n4, Translation="0,0,0", Rotation="0,-30,0", Scale="1,1,1") {
    rotationMode=BMM_Additive
    rotationSpace=BCS_BoneSpace
    translationMode=BMM_Additive
    translationSpace=BCS_BoneSpace
    alphaBoolBlend {blendOption=Linear}
    alphaScaleBiasClamp {bInterpResult=true, interpSpeedDecreasing=30, interpSpeedIncreasing=30}
    boneToModify {boneName=dyn_dress_bk_r_03}
    rotation {yaw=-30}
   }
   n6 = "Transform (Modify) Bone - Bone: dyn_dress_bk_r_04"(ComponentPose=@n5, Alpha=@n4, Translation="0,0,0", Rotation="0,-30,0", Scale="1,1,1") {
    rotationMode=BMM_Additive
    rotationSpace=BCS_BoneSpace
    translationMode=BMM_Additive
    translationSpace=BCS_BoneSpace
    alphaBoolBlend {blendOption=Linear}
    alphaScaleBiasClamp {bInterpResult=true, interpSpeedDecreasing=30, interpSpeedIncreasing=30}
    boneToModify {boneName=dyn_dress_bk_r_04}
    rotation {yaw=-30}
   }
   n7 = "Transform (Modify) Bone - Bone: dyn_dress_bk_r_05"(ComponentPose=@n6, Alpha=@n4, Translation="0,0,0", Rotation="0,-30,0", Scale="1,1,1") {
    rotationMode=BMM_Additive
    rotationSpace=BCS_BoneSpace
    translationMode=BMM_Additive
    translationSpace=BCS_BoneSpace
    alphaBoolBlend {blendOption=Linear}
    alphaScaleBiasClamp {bInterpResult=true, interpSpeedDecreasing=30, interpSpeedIncreasing=30}
    boneToModify {boneName=dyn_dress_bk_r_05}
    rotation {yaw=-30}
   }
   "Transform (Modify) Bone - Bone: dyn_dress_bk_r_06"(ComponentPose=@n7, Alpha=@n4, Translation="0,0,0", Rotation="0,-30,0", Scale="1,1,1") {
    rotationMode=BMM_Additive
    rotationSpace=BCS_BoneSpace
    translationMode=BMM_Additive
    translationSpace=BCS_BoneSpace
    alphaBoolBlend {blendOption=Linear}
    alphaScaleBiasClamp {bInterpResult=true, interpSpeedDecreasing=30, interpSpeedIncreasing=30}
    boneToModify {boneName=dyn_dress_bk_r_06}
    rotation {yaw=-30}
   }
   n8 = RigidBody() {
    bClampLinearTranslationLimitToRefPose=true
    bOverrideWorldGravity=true
    overridePhysicsAsset=/Game/Characters/Player/Female/Medium/Bodies/F_MED_Palespooky_Outlander/Meshes/F_MED_Palespooky_Outlander_Dyn_VFX_Hand
    simulationSpace=BaseBoneSpace
    alphaBoolBlend {blendOption=Linear}
    baseBoneRef {boneName=spine_05}
    componentAppliedLinearAccClamp=<3000 3000 3000>
    componentLinearAccScale=<0.3 0.3 0.3>
    componentLinearVelScale=<0.3 0.3 0.3>
    overrideWorldGravity {z=200}
   }
   n9 = RigidBody(ComponentPose=@n8) {
    bClampLinearTranslationLimitToRefPose=true
    bOverrideWorldGravity=true
    overridePhysicsAsset=/Game/Characters/Player/Female/Medium/Bodies/F_MED_Palespooky_Outlander/Meshes/F_MED_Palespooky_Outlander_Dyn_VFX_Feet
    simulationSpace=BaseBoneSpace
    alphaBoolBlend {blendOption=Linear}
    baseBoneRef {boneName=pelvis}
    componentAppliedLinearAccClamp=<3000 3000 3000>
    componentLinearAccScale=<0.3 0.3 0.3>
    componentLinearVelScale=<0.3 0.3 0.3>
    overrideWorldGravity {z=200}
   }
   RigidBody(ComponentPose=@n9) {
    bClampLinearTranslationLimitToRefPose=true
    bOverrideWorldGravity=true
    overridePhysicsAsset=/Game/Characters/Player/Female/Medium/Bodies/F_MED_Palespooky_Outlander/Meshes/F_MED_Palespooky_Outlander_Dyn_Sleeve
    simulationSpace=BaseBoneSpace
    alphaBoolBlend {blendOption=Linear}
    baseBoneRef {boneName=spine_05}
    componentAppliedLinearAccClamp=<0 0 0>
    componentLinearAccScale {x=0.1, z=-0.5}
    componentLinearVelScale {y=0.3, z=-0.1}
    overrideWorldGravity {z=-500}
   }
   n10 = "Constraint - dyn_dress_fr_mid_l_MSTR (R)"!(ComponentPose=Local To Component(LocalPose=Use cached pose 'preSkirt'()), Alpha=1) -> (Pose) {
    alphaBoolBlend {blendOption=Linear}
    boneToModify {boneName=dyn_dress_fr_mid_l_MSTR}
    constraintSetup [{transformType=Rotation, targetBone {boneName=thigh_twist_01_l}}]
    constraintWeights [1]
   }
   n11 = "Constraint - dyn_dress_fr_mid_l_01 (R)"!(ComponentPose=@n10, Alpha=1) -> (Pose) {
    alphaBoolBlend {blendOption=Linear}
    boneToModify {boneName=dyn_dress_fr_mid_l_01}
    constraintSetup [{transformType=Rotation, targetBone {boneName=thigh_twist_01_l}}]
    constraintWeights [1]
   }
   n12 = "Constraint - dyn_dress_fr_mid_l_02 (R)"!(ComponentPose=@n11, Alpha=1) -> (Pose) {
    alphaBoolBlend {blendOption=Linear}
    boneToModify {boneName=dyn_dress_fr_mid_l_02}
    constraintSetup [{transformType=Rotation, targetBone {boneName=thigh_twist_01_l}}]
    constraintWeights [1]
   }
   n13 = "Constraint - dyn_dress_fr_mid_l_03 (R)"!(ComponentPose=@n12, Alpha=1) -> (Pose) {
    alphaBoolBlend {blendOption=Linear}
    boneToModify {boneName=dyn_dress_fr_mid_l_03}
    constraintSetup [{transformType=Rotation, targetBone {boneName=thigh_twist_01_l}}]
    constraintWeights [1]
   }
   n14 = "Constraint - dyn_dress_fr_mid_l_04 (R)"!(ComponentPose=@n13, Alpha=1) -> (Pose) {
    alphaBoolBlend {blendOption=Linear}
    boneToModify {boneName=dyn_dress_fr_mid_l_04}
    constraintSetup [{transformType=Rotation, targetBone {boneName=thigh_twist_01_l}}]
    constraintWeights [1]
   }
   n15 = "Constraint - dyn_dress_fr_l_MSTR (PR)"!(ComponentPose=@n14, Alpha=1, ConstraintWeights_0=1, ConstraintWeights_1:float=1) -> (Pose) {
    alphaBoolBlend {blendOption=Linear}
    boneToModify {boneName=dyn_dress_fr_l_MSTR}
    constraintSetup [{targetBone {boneName=dyn_dress_fr_mid_l_MSTR}}, {transformType=Rotation, targetBone {boneName=dyn_dress_l_MSTR}}]
    constraintWeights [1,1]
   }
   n16 = "Constraint - dyn_dress_fr_l_01 (PR)"!(ComponentPose=@n15, Alpha=1, ConstraintWeights_0=1, ConstraintWeights_1:float=1) -> (Pose) {
    alphaBoolBlend {blendOption=Linear}
    boneToModify {boneName=dyn_dress_fr_l_01}
    constraintSetup [{targetBone {boneName=dyn_dress_fr_mid_l_01}}, {transformType=Rotation, targetBone {boneName=dyn_dress_l_01}}]
    constraintWeights [1,1]
   }
   n17 = "Constraint - dyn_dress_fr_l_02 (PR)"!(ComponentPose=@n16, Alpha=1, ConstraintWeights_0=1, ConstraintWeights_1:float=1) -> (Pose) {
    alphaBoolBlend {blendOption=Linear}
    boneToModify {boneName=dyn_dress_fr_l_02}
    constraintSetup [{targetBone {boneName=dyn_dress_fr_mid_l_02}}, {transformType=Rotation, targetBone {boneName=dyn_dress_l_02}}]
    constraintWeights [1,1]
   }
   n18 = "Constraint - dyn_dress_fr_l_03 (PR)"!(ComponentPose=@n17, Alpha=1, ConstraintWeights_0=1, ConstraintWeights_1:float=1) -> (Pose) {
    alphaBoolBlend {blendOption=Linear}
    boneToModify {boneName=dyn_dress_fr_l_03}
    constraintSetup [{targetBone {boneName=dyn_dress_fr_mid_l_03}}, {transformType=Rotation, targetBone {boneName=dyn_dress_l_03}}]
    constraintWeights [1,1]
   }
   n19 = "Constraint - dyn_dress_fr_l_04 (PR)"!(ComponentPose=@n18, Alpha=1, ConstraintWeights_0=1, ConstraintWeights_1:float=1) -> (Pose) {
    alphaBoolBlend {blendOption=Linear}
    boneToModify {boneName=dyn_dress_fr_l_04}
    constraintSetup [{targetBone {boneName=dyn_dress_fr_mid_l_04}}, {transformType=Rotation, targetBone {boneName=dyn_dress_l_04}}]
    constraintWeights [1,1]
   }
   n20 = "Constraint - dyn_dress_fr_mid_r_MSTR (P)"!(ComponentPose=@n19, Alpha=1) -> (Pose) {
    alphaBoolBlend {blendOption=Linear}
    boneToModify {boneName=dyn_dress_fr_mid_r_MSTR}
    constraintSetup [{targetBone {boneName=thigh_twist_01_r}}]
    constraintWeights [1]
   }
   n21 = "Constraint - dyn_dress_fr_mid_r_01 (P)"!(ComponentPose=@n20, Alpha=1) -> (Pose) {
    alphaBoolBlend {blendOption=Linear}
    boneToModify {boneName=dyn_dress_fr_mid_r_01}
    constraintSetup [{targetBone {boneName=thigh_twist_01_r}}]
    constraintWeights [1]
   }
   n22 = "Constraint - dyn_dress_fr_mid_r_02 (P)"!(ComponentPose=@n21, Alpha=1) -> (Pose) {
    alphaBoolBlend {blendOption=Linear}
    boneToModify {boneName=dyn_dress_fr_mid_r_02}
    constraintSetup [{targetBone {boneName=thigh_twist_01_r}}]
    constraintWeights [1]
   }
   n23 = "Constraint - dyn_dress_fr_mid_r_03 (P)"!(ComponentPose=@n22, Alpha=1) -> (Pose) {
    alphaBoolBlend {blendOption=Linear}
    boneToModify {boneName=dyn_dress_fr_mid_r_03}
    constraintSetup [{targetBone {boneName=thigh_twist_01_r}}]
    constraintWeights [1]
   }
   n24 = "Constraint - dyn_dress_fr_mid_r_04 (P)"!(ComponentPose=@n23, Alpha=1) -> (Pose) {
    alphaBoolBlend {blendOption=Linear}
    boneToModify {boneName=dyn_dress_fr_mid_r_04}
    constraintSetup [{targetBone {boneName=thigh_twist_01_r}}]
    constraintWeights [1]
   }
   n25 = "Constraint - dyn_dress_fr_r_MSTR (PR)"!(ComponentPose=@n24, Alpha=1, ConstraintWeights_0=1, ConstraintWeights_1:float=1) -> (Pose) {
    alphaBoolBlend {blendOption=Linear}
    boneToModify {boneName=dyn_dress_fr_r_MSTR}
    constraintSetup [{targetBone {boneName=dyn_dress_fr_mid_r_MSTR}}, {transformType=Rotation, targetBone {boneName=dyn_dress_l_MSTR}}]
    constraintWeights [1,1]
   }
   n26 = "Constraint - dyn_dress_fr_r_01 (PR)"!(ComponentPose=@n25, Alpha=1, ConstraintWeights_0=1, ConstraintWeights_1:float=1) -> (Pose) {
    alphaBoolBlend {blendOption=Linear}
    boneToModify {boneName=dyn_dress_fr_r_01}
    constraintSetup [{targetBone {boneName=dyn_dress_fr_mid_r_01}}, {transformType=Rotation, targetBone {boneName=dyn_dress_l_01}}]
    constraintWeights [1,1]
   }
   n27 = "Constraint - dyn_dress_fr_r_02 (PR)"!(ComponentPose=@n26, Alpha=1, ConstraintWeights_0=1, ConstraintWeights_1:float=1) -> (Pose) {
    alphaBoolBlend {blendOption=Linear}
    boneToModify {boneName=dyn_dress_fr_r_02}
    constraintSetup [{targetBone {boneName=dyn_dress_fr_mid_r_02}}, {transformType=Rotation, targetBone {boneName=dyn_dress_l_02}}]
    constraintWeights [1,1]
   }
   n28 = "Constraint - dyn_dress_fr_r_03 (PR)"!(ComponentPose=@n27, Alpha=1, ConstraintWeights_0=1, ConstraintWeights_1:float=1) -> (Pose) {
    alphaBoolBlend {blendOption=Linear}
    boneToModify {boneName=dyn_dress_fr_r_03}
    constraintSetup [{targetBone {boneName=dyn_dress_fr_mid_r_03}}, {transformType=Rotation, targetBone {boneName=dyn_dress_l_03}}]
    constraintWeights [1,1]
   }
   n29 = "Constraint - dyn_dress_fr_r_04 (PR)"!(ComponentPose=@n28, Alpha=1, ConstraintWeights_0=1, ConstraintWeights_1:float=1) -> (Pose) {
    alphaBoolBlend {blendOption=Linear}
    boneToModify {boneName=dyn_dress_fr_r_04}
    constraintSetup [{targetBone {boneName=dyn_dress_fr_mid_r_04}}, {transformType=Rotation, targetBone {boneName=dyn_dress_l_04}}]
    constraintWeights [1,1]
   }
   n30 = "Constraint - dyn_skirt_fr_01 (PP)"!(ComponentPose=@n29, Alpha=1, ConstraintWeights_0=1, ConstraintWeights_1:float=1) -> (Pose) {
    alphaBoolBlend {blendOption=Linear}
    boneToModify {boneName=dyn_skirt_fr_01}
    constraintSetup [{targetBone {boneName=dyn_dress_fr_mid_l_01}}, {targetBone {boneName=dyn_dress_fr_mid_r_01}}]
    constraintWeights [1,1]
   }
   n31 = "Constraint - dyn_skirt_fr_02 (PP)"!(ComponentPose=@n30, Alpha=1, ConstraintWeights_0=1, ConstraintWeights_1:float=1) -> (Pose) {
    alphaBoolBlend {blendOption=Linear}
    boneToModify {boneName=dyn_skirt_fr_02}
    constraintSetup [{targetBone {boneName=dyn_dress_fr_mid_l_02}}, {targetBone {boneName=dyn_dress_fr_mid_r_02}}]
    constraintWeights [1,1]
   }
   n32 = "Constraint - dyn_skirt_fr_03 (PP)"!(ComponentPose=@n31, Alpha=1, ConstraintWeights_0=1, ConstraintWeights_1:float=1) -> (Pose) {
    alphaBoolBlend {blendOption=Linear}
    boneToModify {boneName=dyn_skirt_fr_03}
    constraintSetup [{targetBone {boneName=dyn_dress_fr_mid_l_03}}, {targetBone {boneName=dyn_dress_fr_mid_r_03}}]
    constraintWeights [1,1]
   }
   n33 = Anim Dynamics - Bone: dyn_skirt_fr_01(ComponentPose=@n32, Alpha=1, ExternalForce="0,0,0") {
    angularDampingOverride=0.98
    bGravityOverrideInSimSpace=true
    bOverrideAngularDamping=true
    bUseGravityOverride=true
    alphaBoolBlend {blendOption=Linear}
    boundBone {boneName=dyn_skirt_fr_01}
    componentAppliedLinearAccClamp=<3000 3000 3000>
    componentLinearAccScale=<0.1 0.1 0.1>
    componentLinearVelScale=<0.1 0.1 0.1>
    gravityOverride {x=1, z=-980}
    physicsBodyDefinitions [{
      sphereCollisionRadius=0
      boundBone {boneName=dyn_skirt_fr_01}
      boxExtents=<10 20 20>
      constraintSetup {angularLimitsMax=<0 0 5>, angularLimitsMin=<0 -10 -5>, angularTarget=<0 0 0>}
      localJointOffset=<3 0 0>
     }]
    relativeSpaceBone {boneName=root}
   }
   n34 = Anim Dynamics - Bone: dyn_skirt_fr_02(ComponentPose=@n33, Alpha=1, ExternalForce="0,0,0") {
    angularDampingOverride=0.98
    bGravityOverrideInSimSpace=true
    bOverrideAngularDamping=true
    bUseGravityOverride=true
    alphaBoolBlend {blendOption=Linear}
    boundBone {boneName=dyn_skirt_fr_02}
    componentAppliedLinearAccClamp=<3000 3000 3000>
    componentLinearAccScale=<0.1 0.1 0.1>
    componentLinearVelScale=<0.1 0.1 0.1>
    gravityOverride {x=1, z=-980}
    physicsBodyDefinitions [{
      sphereCollisionRadius=0
      boundBone {boneName=dyn_skirt_fr_02}
      boxExtents=<10 20 20>
      constraintSetup {angularLimitsMax=<10 5 5>, angularLimitsMin=<-10 -30 -5>, angularTarget=<0 0 0>}
      localJointOffset=<3 0 0>
     }]
    relativeSpaceBone {boneName=root}
   }
   n35 = Anim Dynamics - Bone: dyn_skirt_fr_03(ComponentPose=@n34, Alpha=1, ExternalForce="0,0,0") {
    angularDampingOverride=0.98
    bGravityOverrideInSimSpace=true
    bOverrideAngularDamping=true
    bUseGravityOverride=true
    alphaBoolBlend {blendOption=Linear}
    boundBone {boneName=dyn_skirt_fr_03}
    componentAppliedLinearAccClamp=<3000 3000 3000>
    componentLinearAccScale=<0.1 0.1 0.1>
    componentLinearVelScale=<0.1 0.1 0.1>
    gravityOverride {x=1, z=-980}
    physicsBodyDefinitions [{
      sphereCollisionRadius=0
      boundBone {boneName=dyn_skirt_fr_03}
      boxExtents=<10 20 20>
      constraintSetup {angularLimitsMax=<10 10 5>, angularLimitsMin=<-10 -30 -5>, angularTarget=<0 0 0>}
      localJointOffset=<3 0 0>
     }]
    relativeSpaceBone {boneName=root}
   }
   n36 = Get rigids alpha() -> (rigids alpha:double)
   n37 = Get gravity_override() -> (gravity_override:Vector)
   n38 = RigidBody!(ComponentPose=@n35, Alpha=@n36, OverrideWorldGravity:Vector=@n37, ExternalForce) -> (Pose) {
    bClampLinearTranslationLimitToRefPose=true
    bOverrideWorldGravity=true
    overridePhysicsAsset=/Game/Characters/Player/Female/Medium/Bodies/F_MED_Palespooky_Outlander/Meshes/F_MED_Palespooky_Outlander_Dyn_Skirt_EMOTES
    simulationSpace=BaseBoneSpace
    alphaBoolBlend {blendOption=Linear}
    baseBoneRef {boneName=root}
    componentLinearAccScale=<0.5 0.5 0.55>
    overrideWorldGravity {z=-980}
   }
   n39 = "Constraint - dyn_dress_fr_mid_l_MSTR (P)"!(ComponentPose=@n38, Alpha=1) -> (Pose) {
    alphaBoolBlend {blendOption=Linear}
    boneToModify {boneName=dyn_dress_fr_mid_l_MSTR}
    constraintSetup [{targetBone {boneName=thigh_twist_01_l}}]
    constraintWeights [1]
   }
   n40 = "Constraint - dyn_dress_fr_mid_l_01 (P)"!(ComponentPose=@n39, Alpha=1) -> (Pose) {
    alphaBoolBlend {blendOption=Linear}
    boneToModify {boneName=dyn_dress_fr_mid_l_01}
    constraintSetup [{targetBone {boneName=thigh_twist_01_l}}]
    constraintWeights [1]
   }
   n41 = "Constraint - dyn_dress_fr_mid_l_02 (P)"!(ComponentPose=@n40, Alpha=1) -> (Pose) {
    alphaBoolBlend {blendOption=Linear}
    boneToModify {boneName=dyn_dress_fr_mid_l_02}
    constraintSetup [{targetBone {boneName=thigh_twist_01_l}}]
    constraintWeights [1]
   }
   n42 = "Constraint - dyn_dress_fr_mid_l_03 (P)"!(ComponentPose=@n41, Alpha=1) -> (Pose) {
    alphaBoolBlend {blendOption=Linear}
    boneToModify {boneName=dyn_dress_fr_mid_l_03}
    constraintSetup [{targetBone {boneName=thigh_twist_01_l}}]
    constraintWeights [1]
   }
   n43 = "Constraint - dyn_dress_fr_mid_l_04 (P)"!(ComponentPose=@n42, Alpha=1) -> (Pose) {
    alphaBoolBlend {blendOption=Linear}
    boneToModify {boneName=dyn_dress_fr_mid_l_04}
    constraintSetup [{targetBone {boneName=thigh_twist_01_l}}]
    constraintWeights [1]
   }
   n44 = "Constraint - dyn_dress_fr_mid_r_MSTR (P)"(ComponentPose=@n43, Alpha=1) {
    alphaBoolBlend {blendOption=Linear}
    boneToModify {boneName=dyn_dress_fr_mid_r_MSTR}
    constraintSetup [{targetBone {boneName=thigh_twist_01_r}}]
    constraintWeights [1]
   }
   n45 = "Constraint - dyn_dress_fr_mid_r_01 (P)"(ComponentPose=@n44, Alpha=1) {
    alphaBoolBlend {blendOption=Linear}
    boneToModify {boneName=dyn_dress_fr_mid_r_01}
    constraintSetup [{targetBone {boneName=thigh_twist_01_r}}]
    constraintWeights [1]
   }
   n46 = "Constraint - dyn_dress_fr_mid_r_02 (P)"(ComponentPose=@n45, Alpha=1) {
    alphaBoolBlend {blendOption=Linear}
    boneToModify {boneName=dyn_dress_fr_mid_r_02}
    constraintSetup [{targetBone {boneName=thigh_twist_01_r}}]
    constraintWeights [1]
   }
   n47 = "Constraint - dyn_dress_fr_mid_r_03 (P)"(ComponentPose=@n46, Alpha=1) {
    alphaBoolBlend {blendOption=Linear}
    boneToModify {boneName=dyn_dress_fr_mid_r_03}
    constraintSetup [{targetBone {boneName=thigh_twist_01_r}}]
    constraintWeights [1]
   }
   n48 = "Constraint - dyn_dress_fr_mid_r_04 (P)"(ComponentPose=@n47, Alpha=1) {
    alphaBoolBlend {blendOption=Linear}
    boneToModify {boneName=dyn_dress_fr_mid_r_04}
    constraintSetup [{targetBone {boneName=thigh_twist_01_r}}]
    constraintWeights [1]
   }
   n49 = "Constraint - dyn_dress_r_MSTR (PPP)"!(ComponentPose=@n48, Alpha=0.1, ConstraintWeights_0=1, ConstraintWeights_1:float=1, ConstraintWeights_2:float=1) -> (Pose) {
    alphaBoolBlend {blendOption=Linear}
    boneToModify {boneName=dyn_dress_r_MSTR}
    constraintSetup [{targetBone {boneName=dyn_dress_r_MSTR}}, {targetBone {boneName=dyn_dress_bk_r_MSTR}}, {targetBone {boneName=thigh_twist_01_r}}]
    constraintWeights [1,1,1]
   }
   n50 = "Constraint - dyn_dress_r_01 (PPP)"!(ComponentPose=@n49, Alpha=0.1, ConstraintWeights_0=1, ConstraintWeights_1:float=1, ConstraintWeights_2:float=1) -> (Pose) {
    alphaBoolBlend {blendOption=Linear}
    boneToModify {boneName=dyn_dress_r_01}
    constraintSetup [{targetBone {boneName=dyn_dress_r_01}}, {targetBone {boneName=dyn_dress_bk_r_01}}, {targetBone {boneName=thigh_twist_01_r}}]
    constraintWeights [1,1,1]
   }
   n51 = "Constraint - dyn_dress_r_02 (PPP)"!(ComponentPose=@n50, Alpha=0.1, ConstraintWeights_0=1, ConstraintWeights_1:float=1, ConstraintWeights_2:float=1) -> (Pose) {
    alphaBoolBlend {blendOption=Linear}
    boneToModify {boneName=dyn_dress_r_02}
    constraintSetup [{targetBone {boneName=dyn_dress_r_02}}, {targetBone {boneName=dyn_dress_bk_r_02}}, {targetBone {boneName=thigh_twist_01_r}}]
    constraintWeights [1,1,1]
   }
   n52 = "Constraint - dyn_dress_r_03 (PPP)"!(ComponentPose=@n51, Alpha=0.1, ConstraintWeights_0=1, ConstraintWeights_1:float=1, ConstraintWeights_2:float=1) -> (Pose) {
    alphaBoolBlend {blendOption=Linear}
    boneToModify {boneName=dyn_dress_r_03}
    constraintSetup [{targetBone {boneName=dyn_dress_r_03}}, {targetBone {boneName=dyn_dress_bk_r_03}}, {targetBone {boneName=thigh_twist_01_r}}]
    constraintWeights [1,1,1]
   }
   n53 = "Constraint - dyn_dress_r_04 (PPP)"!(ComponentPose=@n52, Alpha=0.1, ConstraintWeights_0=1, ConstraintWeights_1:float=1, ConstraintWeights_2:float=1) -> (Pose) {
    alphaBoolBlend {blendOption=Linear}
    boneToModify {boneName=dyn_dress_r_04}
    constraintSetup [{targetBone {boneName=dyn_dress_r_04}}, {targetBone {boneName=dyn_dress_bk_r_04}}, {targetBone {boneName=thigh_twist_01_r}}]
    constraintWeights [1,1,1]
   }
   n54 = Get bIsPlayingEmote()
   n55 = "Transform (Modify) Bone - Bone: dyn_dress_fr_l_MSTR"!(ComponentPose=@n53, bAlphaBoolEnabled:bool=@n54, Translation="2,0,2", Rotation="0,0,0", Scale="1,1,1") -> (Pose) {
    alphaInputType=Bool
    rotationMode=BMM_Additive
    scaleMode=BMM_Additive
    translationMode=BMM_Additive
    alphaBoolBlend {blendInTime=0.1, blendOption=Linear, blendOutTime=0.1}
    boneToModify {boneName=dyn_dress_fr_l_MSTR}
    translation {x=2, z=2}
   }
   n56 = "Transform (Modify) Bone - Bone: dyn_dress_l_MSTR"!(ComponentPose=@n55, bAlphaBoolEnabled:bool=@n54, Translation="2,0,0", Rotation="0,0,0", Scale="1,1,1") -> (Pose) {
    alphaInputType=Bool
    rotationMode=BMM_Additive
    scaleMode=BMM_Additive
    translationMode=BMM_Additive
    alphaBoolBlend {blendInTime=0.1, blendOption=Linear, blendOutTime=0.1}
    boneToModify {boneName=dyn_dress_l_MSTR}
    translation {x=2}
   }
   n57 = "Constraint - dyn_dress_fr_mid_l_MSTR (R)"(ComponentPose=Local To Component(LocalPose=Use cached pose 'preSkirt'()), Alpha=1) {
    alphaBoolBlend {blendOption=Linear}
    boneToModify {boneName=dyn_dress_fr_mid_l_MSTR}
    constraintSetup [{transformType=Rotation, targetBone {boneName=thigh_twist_01_l}}]
    constraintWeights [1]
   }
   n58 = "Constraint - dyn_dress_fr_mid_l_01 (R)"(ComponentPose=@n57, Alpha=1) {
    alphaBoolBlend {blendOption=Linear}
    boneToModify {boneName=dyn_dress_fr_mid_l_01}
    constraintSetup [{transformType=Rotation, targetBone {boneName=thigh_twist_01_l}}]
    constraintWeights [1]
   }
   n59 = "Constraint - dyn_dress_fr_mid_l_02 (R)"(ComponentPose=@n58, Alpha=1) {
    alphaBoolBlend {blendOption=Linear}
    boneToModify {boneName=dyn_dress_fr_mid_l_02}
    constraintSetup [{transformType=Rotation, targetBone {boneName=thigh_twist_01_l}}]
    constraintWeights [1]
   }
   n60 = "Constraint - dyn_dress_fr_mid_l_03 (R)"(ComponentPose=@n59, Alpha=1) {
    alphaBoolBlend {blendOption=Linear}
    boneToModify {boneName=dyn_dress_fr_mid_l_03}
    constraintSetup [{transformType=Rotation, targetBone {boneName=thigh_twist_01_l}}]
    constraintWeights [1]
   }
   n61 = "Constraint - dyn_dress_fr_mid_l_04 (R)"(ComponentPose=@n60, Alpha=1) {
    alphaBoolBlend {blendOption=Linear}
    boneToModify {boneName=dyn_dress_fr_mid_l_04}
    constraintSetup [{transformType=Rotation, targetBone {boneName=thigh_twist_01_l}}]
    constraintWeights [1]
   }
   n62 = "Constraint - dyn_dress_fr_l_MSTR (PR)"(ComponentPose=@n61, Alpha=1, ConstraintWeights_0=1, ConstraintWeights_1=1) {
    alphaBoolBlend {blendOption=Linear}
    boneToModify {boneName=dyn_dress_fr_l_MSTR}
    constraintSetup [{targetBone {boneName=dyn_dress_fr_mid_l_MSTR}}, {transformType=Rotation, targetBone {boneName=dyn_dress_l_MSTR}}]
    constraintWeights [1,1]
   }
   n63 = "Constraint - dyn_dress_fr_l_01 (PR)"(ComponentPose=@n62, Alpha=1, ConstraintWeights_0=1, ConstraintWeights_1=1) {
    alphaBoolBlend {blendOption=Linear}
    boneToModify {boneName=dyn_dress_fr_l_01}
    constraintSetup [{targetBone {boneName=dyn_dress_fr_mid_l_01}}, {transformType=Rotation, targetBone {boneName=dyn_dress_l_01}}]
    constraintWeights [1,1]
   }
   n64 = "Constraint - dyn_dress_fr_l_02 (PR)"(ComponentPose=@n63, Alpha=1, ConstraintWeights_0=1, ConstraintWeights_1=1) {
    alphaBoolBlend {blendOption=Linear}
    boneToModify {boneName=dyn_dress_fr_l_02}
    constraintSetup [{targetBone {boneName=dyn_dress_fr_mid_l_02}}, {transformType=Rotation, targetBone {boneName=dyn_dress_l_02}}]
    constraintWeights [1,1]
   }
   n65 = "Constraint - dyn_dress_fr_l_03 (PR)"(ComponentPose=@n64, Alpha=1, ConstraintWeights_0=1, ConstraintWeights_1=1) {
    alphaBoolBlend {blendOption=Linear}
    boneToModify {boneName=dyn_dress_fr_l_03}
    constraintSetup [{targetBone {boneName=dyn_dress_fr_mid_l_03}}, {transformType=Rotation, targetBone {boneName=dyn_dress_l_03}}]
    constraintWeights [1,1]
   }
   n66 = "Constraint - dyn_dress_fr_l_04 (PR)"(ComponentPose=@n65, Alpha=1, ConstraintWeights_0=1, ConstraintWeights_1=1) {
    alphaBoolBlend {blendOption=Linear}
    boneToModify {boneName=dyn_dress_fr_l_04}
    constraintSetup [{targetBone {boneName=dyn_dress_fr_mid_l_04}}, {transformType=Rotation, targetBone {boneName=dyn_dress_l_04}}]
    constraintWeights [1,1]
   }
   n67 = "Constraint - dyn_dress_fr_mid_r_MSTR (P)"(ComponentPose=@n66, Alpha=1) {
    alphaBoolBlend {blendOption=Linear}
    boneToModify {boneName=dyn_dress_fr_mid_r_MSTR}
    constraintSetup [{targetBone {boneName=thigh_twist_01_r}}]
    constraintWeights [1]
   }
   n68 = "Constraint - dyn_dress_fr_mid_r_01 (P)"(ComponentPose=@n67, Alpha=1) {
    alphaBoolBlend {blendOption=Linear}
    boneToModify {boneName=dyn_dress_fr_mid_r_01}
    constraintSetup [{targetBone {boneName=thigh_twist_01_r}}]
    constraintWeights [1]
   }
   n69 = "Constraint - dyn_dress_fr_mid_r_02 (P)"(ComponentPose=@n68, Alpha=1) {
    alphaBoolBlend {blendOption=Linear}
    boneToModify {boneName=dyn_dress_fr_mid_r_02}
    constraintSetup [{targetBone {boneName=thigh_twist_01_r}}]
    constraintWeights [1]
   }
   n70 = "Constraint - dyn_dress_fr_mid_r_03 (P)"(ComponentPose=@n69, Alpha=1) {
    alphaBoolBlend {blendOption=Linear}
    boneToModify {boneName=dyn_dress_fr_mid_r_03}
    constraintSetup [{targetBone {boneName=thigh_twist_01_r}}]
    constraintWeights [1]
   }
   n71 = "Constraint - dyn_dress_fr_mid_r_04 (P)"(ComponentPose=@n70, Alpha=1) {
    alphaBoolBlend {blendOption=Linear}
    boneToModify {boneName=dyn_dress_fr_mid_r_04}
    constraintSetup [{targetBone {boneName=thigh_twist_01_r}}]
    constraintWeights [1]
   }
   n72 = "Constraint - dyn_dress_fr_r_MSTR (PR)"(ComponentPose=@n71, Alpha=1, ConstraintWeights_0=1, ConstraintWeights_1=1) {
    alphaBoolBlend {blendOption=Linear}
    boneToModify {boneName=dyn_dress_fr_r_MSTR}
    constraintSetup [{targetBone {boneName=dyn_dress_fr_mid_r_MSTR}}, {transformType=Rotation, targetBone {boneName=dyn_dress_l_MSTR}}]
    constraintWeights [1,1]
   }
   n73 = "Constraint - dyn_dress_fr_r_01 (PR)"(ComponentPose=@n72, Alpha=1, ConstraintWeights_0=1, ConstraintWeights_1=1) {
    alphaBoolBlend {blendOption=Linear}
    boneToModify {boneName=dyn_dress_fr_r_01}
    constraintSetup [{targetBone {boneName=dyn_dress_fr_mid_r_01}}, {transformType=Rotation, targetBone {boneName=dyn_dress_l_01}}]
    constraintWeights [1,1]
   }
   n74 = "Constraint - dyn_dress_fr_r_02 (PR)"(ComponentPose=@n73, Alpha=1, ConstraintWeights_0=1, ConstraintWeights_1=1) {
    alphaBoolBlend {blendOption=Linear}
    boneToModify {boneName=dyn_dress_fr_r_02}
    constraintSetup [{targetBone {boneName=dyn_dress_fr_mid_r_02}}, {transformType=Rotation, targetBone {boneName=dyn_dress_l_02}}]
    constraintWeights [1,1]
   }
   n75 = "Constraint - dyn_dress_fr_r_03 (PR)"(ComponentPose=@n74, Alpha=1, ConstraintWeights_0=1, ConstraintWeights_1=1) {
    alphaBoolBlend {blendOption=Linear}
    boneToModify {boneName=dyn_dress_fr_r_03}
    constraintSetup [{targetBone {boneName=dyn_dress_fr_mid_r_03}}, {transformType=Rotation, targetBone {boneName=dyn_dress_l_03}}]
    constraintWeights [1,1]
   }
   n76 = "Constraint - dyn_dress_fr_r_04 (PR)"(ComponentPose=@n75, Alpha=1, ConstraintWeights_0=1, ConstraintWeights_1=1) {
    alphaBoolBlend {blendOption=Linear}
    boneToModify {boneName=dyn_dress_fr_r_04}
    constraintSetup [{targetBone {boneName=dyn_dress_fr_mid_r_04}}, {transformType=Rotation, targetBone {boneName=dyn_dress_l_04}}]
    constraintWeights [1,1]
   }
   n77 = RigidBody(ComponentPose=@n76, Alpha=@n36, OverrideWorldGravity=@n37) {
    bClampLinearTranslationLimitToRefPose=true
    bOverrideWorldGravity=true
    overridePhysicsAsset=/Game/Characters/Player/Female/Medium/Bodies/F_MED_Palespooky_Outlander/Meshes/F_MED_Palespooky_Outlander_Dyn_Skirt_NEW
    simulationSpace=BaseBoneSpace
    alphaBoolBlend {blendOption=Linear}
    baseBoneRef {boneName=root}
    componentLinearAccScale=<0.5 0.5 0.55>
    overrideWorldGravity {z=-980}
   }
   n78 = "Constraint - dyn_dress_fr_mid_l_MSTR (P)"(ComponentPose=@n77, Alpha=1) {
    alphaBoolBlend {blendOption=Linear}
    boneToModify {boneName=dyn_dress_fr_mid_l_MSTR}
    constraintSetup [{targetBone {boneName=thigh_twist_01_l}}]
    constraintWeights [1]
   }
   n79 = "Constraint - dyn_dress_fr_mid_l_01 (P)"(ComponentPose=@n78, Alpha=1) {
    alphaBoolBlend {blendOption=Linear}
    boneToModify {boneName=dyn_dress_fr_mid_l_01}
    constraintSetup [{targetBone {boneName=thigh_twist_01_l}}]
    constraintWeights [1]
   }
   n80 = "Constraint - dyn_dress_fr_mid_l_02 (P)"(ComponentPose=@n79, Alpha=1) {
    alphaBoolBlend {blendOption=Linear}
    boneToModify {boneName=dyn_dress_fr_mid_l_02}
    constraintSetup [{targetBone {boneName=thigh_twist_01_l}}]
    constraintWeights [1]
   }
   n81 = "Constraint - dyn_dress_fr_mid_l_03 (P)"(ComponentPose=@n80, Alpha=1) {
    alphaBoolBlend {blendOption=Linear}
    boneToModify {boneName=dyn_dress_fr_mid_l_03}
    constraintSetup [{targetBone {boneName=thigh_twist_01_l}}]
    constraintWeights [1]
   }
   n82 = "Constraint - dyn_dress_fr_mid_l_04 (P)"(ComponentPose=@n81, Alpha=1) {
    alphaBoolBlend {blendOption=Linear}
    boneToModify {boneName=dyn_dress_fr_mid_l_04}
    constraintSetup [{targetBone {boneName=thigh_twist_01_l}}]
    constraintWeights [1]
   }
   n83 = "Constraint - dyn_dress_fr_mid_r_MSTR (P)"(ComponentPose=@n82, Alpha=1) {
    alphaBoolBlend {blendOption=Linear}
    boneToModify {boneName=dyn_dress_fr_mid_r_MSTR}
    constraintSetup [{targetBone {boneName=thigh_twist_01_r}}]
    constraintWeights [1]
   }
   n84 = "Constraint - dyn_dress_fr_mid_r_01 (P)"(ComponentPose=@n83, Alpha=1) {
    alphaBoolBlend {blendOption=Linear}
    boneToModify {boneName=dyn_dress_fr_mid_r_01}
    constraintSetup [{targetBone {boneName=thigh_twist_01_r}}]
    constraintWeights [1]
   }
   n85 = "Constraint - dyn_dress_fr_mid_r_02 (P)"(ComponentPose=@n84, Alpha=1) {
    alphaBoolBlend {blendOption=Linear}
    boneToModify {boneName=dyn_dress_fr_mid_r_02}
    constraintSetup [{targetBone {boneName=thigh_twist_01_r}}]
    constraintWeights [1]
   }
   n86 = "Constraint - dyn_dress_fr_mid_r_03 (P)"(ComponentPose=@n85, Alpha=1) {
    alphaBoolBlend {blendOption=Linear}
    boneToModify {boneName=dyn_dress_fr_mid_r_03}
    constraintSetup [{targetBone {boneName=thigh_twist_01_r}}]
    constraintWeights [1]
   }
   n87 = "Constraint - dyn_dress_fr_mid_r_04 (P)"(ComponentPose=@n86, Alpha=1) {
    alphaBoolBlend {blendOption=Linear}
    boneToModify {boneName=dyn_dress_fr_mid_r_04}
    constraintSetup [{targetBone {boneName=thigh_twist_01_r}}]
    constraintWeights [1]
   }
   n88 = "Constraint - dyn_dress_r_MSTR (PPP)"(ComponentPose=@n87, Alpha=1, ConstraintWeights_0=0.3, ConstraintWeights_1=0.3, ConstraintWeights_2=0.4) {
    alphaBoolBlend {blendOption=Linear}
    boneToModify {boneName=dyn_dress_r_MSTR}
    constraintSetup [{targetBone {boneName=dyn_dress_r_MSTR}}, {targetBone {boneName=dyn_dress_bk_r_MSTR}}, {targetBone {boneName=thigh_twist_01_r}}]
    constraintWeights [1,1,1]
   }
   n89 = "Constraint - dyn_dress_r_01 (PPP)"(ComponentPose=@n88, Alpha=1, ConstraintWeights_0=0.3, ConstraintWeights_1=0.3, ConstraintWeights_2=0.4) {
    alphaBoolBlend {blendOption=Linear}
    boneToModify {boneName=dyn_dress_r_01}
    constraintSetup [{targetBone {boneName=dyn_dress_r_01}}, {targetBone {boneName=dyn_dress_bk_r_01}}, {targetBone {boneName=thigh_twist_01_r}}]
    constraintWeights [1,1,1]
   }
   n90 = "Constraint - dyn_dress_r_02 (PPP)"(ComponentPose=@n89, Alpha=1, ConstraintWeights_0=0.3, ConstraintWeights_1=0.3, ConstraintWeights_2=0.4) {
    alphaBoolBlend {blendOption=Linear}
    boneToModify {boneName=dyn_dress_r_02}
    constraintSetup [{targetBone {boneName=dyn_dress_r_02}}, {targetBone {boneName=dyn_dress_bk_r_02}}, {targetBone {boneName=thigh_twist_01_r}}]
    constraintWeights [1,1,1]
   }
   n91 = "Constraint - dyn_dress_r_03 (PPP)"(ComponentPose=@n90, Alpha=1, ConstraintWeights_0=0.3, ConstraintWeights_1=0.3, ConstraintWeights_2=0.4) {
    alphaBoolBlend {blendOption=Linear}
    boneToModify {boneName=dyn_dress_r_03}
    constraintSetup [{targetBone {boneName=dyn_dress_r_03}}, {targetBone {boneName=dyn_dress_bk_r_03}}, {targetBone {boneName=thigh_twist_01_r}}]
    constraintWeights [1,1,1]
   }
   n92 = "Constraint - dyn_dress_r_04 (PPP)"(ComponentPose=@n91, Alpha=1, ConstraintWeights_0=0.3, ConstraintWeights_1=0.3, ConstraintWeights_2=0.4) {
    alphaBoolBlend {blendOption=Linear}
    boneToModify {boneName=dyn_dress_r_04}
    constraintSetup [{targetBone {boneName=dyn_dress_r_04}}, {targetBone {boneName=dyn_dress_bk_r_04}}, {targetBone {boneName=thigh_twist_01_r}}]
    constraintWeights [1,1,1]
   }
   n93 = "Constraint - dyn_dress_r_05 (PPP)"!(ComponentPose=@n92, Alpha=1, ConstraintWeights_0=0.3, ConstraintWeights_1:float=0.3, ConstraintWeights_2:float=0.4) -> (Pose) {
    alphaBoolBlend {blendOption=Linear}
    boneToModify {boneName=dyn_dress_r_05}
    constraintSetup [{targetBone {boneName=dyn_dress_r_05}}, {targetBone {boneName=dyn_dress_bk_r_05}}, {targetBone {boneName=thigh_twist_01_r}}]
    constraintWeights [1,1,1]
   }
   n94 = "Constraint - dyn_dress_r_06 (PPP)"!(ComponentPose=@n93, Alpha=1, ConstraintWeights_0=0.3, ConstraintWeights_1:float=0.3, ConstraintWeights_2:float=0.4) -> (Pose) {
    alphaBoolBlend {blendOption=Linear}
    boneToModify {boneName=dyn_dress_r_06}
    constraintSetup [{targetBone {boneName=dyn_dress_r_06}}, {targetBone {boneName=dyn_dress_bk_r_06}}, {targetBone {boneName=thigh_twist_01_r}}]
    constraintWeights [1,1,1]
   }
   n95 = "Transform (Modify) Bone - Bone: dyn_dress_fr_l_MSTR"!(ComponentPose=@n94, Alpha:float=1, Translation="1,0,1", Rotation="0,0,0", Scale="1,1,1") -> (Pose) {
    rotationMode=BMM_Additive
    scaleMode=BMM_Additive
    translationMode=BMM_Additive
    alphaBoolBlend {blendOption=Linear}
    boneToModify {boneName=dyn_dress_fr_l_MSTR}
    translation {x=1, z=1}
   }
   n96 = Get crouch alpha()
   n97 = "Transform (Modify) Bone - Bone: dyn_dress_l_01"(ComponentPose=@n95, Alpha=@n96, Translation="1,0,0", Rotation="0,5,0", Scale="1,1,1") {
    rotationMode=BMM_Additive
    scaleMode=BMM_Additive
    translationMode=BMM_Additive
    alphaBoolBlend {blendOption=Linear}
    boneToModify {boneName=dyn_dress_l_01}
    rotation {yaw=5}
    translation {x=1}
   }
   n98 = "Transform (Modify) Bone - Bone: dyn_dress_fr_l_MSTR"(ComponentPose=@n97, Alpha=@n96, Translation="2,0.000,0.000", Rotation="0,0,0", Scale="1,1,1") {
    rotationMode=BMM_Additive
    scaleMode=BMM_Additive
    translationMode=BMM_Additive
    alphaBoolBlend {blendOption=Linear}
    boneToModify {boneName=dyn_dress_fr_l_MSTR}
    translation {x=2}
   }
   n99 = Blend Poses by bool(
    BlendPose_0=Component To Local(ComponentPose=@n56)
    BlendPose_1=Component To Local(ComponentPose=@n98)
    BlendTime_0=0.4
    BlendTime_1=0.4
    bActiveValue=Get bPlayingEmoteOrFullbodyFrontend() -> (bPlayingEmoteOrFullbodyFrontend:bool)
   ) {blendType=Linear}
   n100 = "Constraint - dyn_dress_fr_mid_l_MSTR (R)"(ComponentPose=Local To Component(LocalPose=Use cached pose 'preSkirt'()), Alpha=1) {
    alphaBoolBlend {blendOption=Linear}
    boneToModify {boneName=dyn_dress_fr_mid_l_MSTR}
    constraintSetup [{transformType=Rotation, targetBone {boneName=thigh_twist_01_l}}]
    constraintWeights [1]
   }
   n101 = "Constraint - dyn_dress_fr_mid_l_01 (R)"(ComponentPose=@n100, Alpha=1) {
    alphaBoolBlend {blendOption=Linear}
    boneToModify {boneName=dyn_dress_fr_mid_l_01}
    constraintSetup [{transformType=Rotation, targetBone {boneName=thigh_twist_01_l}}]
    constraintWeights [1]
   }
   n102 = "Constraint - dyn_dress_fr_mid_l_02 (R)"(ComponentPose=@n101, Alpha=1) {
    alphaBoolBlend {blendOption=Linear}
    boneToModify {boneName=dyn_dress_fr_mid_l_02}
    constraintSetup [{transformType=Rotation, targetBone {boneName=thigh_twist_01_l}}]
    constraintWeights [1]
   }
   n103 = "Constraint - dyn_dress_fr_mid_l_03 (R)"(ComponentPose=@n102, Alpha=1) {
    alphaBoolBlend {blendOption=Linear}
    boneToModify {boneName=dyn_dress_fr_mid_l_03}
    constraintSetup [{transformType=Rotation, targetBone {boneName=thigh_twist_01_l}}]
    constraintWeights [1]
   }
   n104 = "Constraint - dyn_dress_fr_mid_l_04 (R)"(ComponentPose=@n103, Alpha=1) {
    alphaBoolBlend {blendOption=Linear}
    boneToModify {boneName=dyn_dress_fr_mid_l_04}
    constraintSetup [{transformType=Rotation, targetBone {boneName=thigh_twist_01_l}}]
    constraintWeights [1]
   }
   n105 = "Constraint - dyn_dress_fr_l_MSTR (PR)"(ComponentPose=@n104, Alpha=1, ConstraintWeights_0=1, ConstraintWeights_1=1) {
    alphaBoolBlend {blendOption=Linear}
    boneToModify {boneName=dyn_dress_fr_l_MSTR}
    constraintSetup [{targetBone {boneName=dyn_dress_fr_mid_l_MSTR}}, {transformType=Rotation, targetBone {boneName=dyn_dress_l_MSTR}}]
    constraintWeights [1,1]
   }
   n106 = "Constraint - dyn_dress_fr_l_01 (PR)"(ComponentPose=@n105, Alpha=1, ConstraintWeights_0=1, ConstraintWeights_1=1) {
    alphaBoolBlend {blendOption=Linear}
    boneToModify {boneName=dyn_dress_fr_l_01}
    constraintSetup [{targetBone {boneName=dyn_dress_fr_mid_l_01}}, {transformType=Rotation, targetBone {boneName=dyn_dress_l_01}}]
    constraintWeights [1,1]
   }
   n107 = "Constraint - dyn_dress_fr_l_02 (PR)"(ComponentPose=@n106, Alpha=1, ConstraintWeights_0=1, ConstraintWeights_1=1) {
    alphaBoolBlend {blendOption=Linear}
    boneToModify {boneName=dyn_dress_fr_l_02}
    constraintSetup [{targetBone {boneName=dyn_dress_fr_mid_l_02}}, {transformType=Rotation, targetBone {boneName=dyn_dress_l_02}}]
    constraintWeights [1,1]
   }
   n108 = "Constraint - dyn_dress_fr_l_03 (PR)"(ComponentPose=@n107, Alpha=1, ConstraintWeights_0=1, ConstraintWeights_1=1) {
    alphaBoolBlend {blendOption=Linear}
    boneToModify {boneName=dyn_dress_fr_l_03}
    constraintSetup [{targetBone {boneName=dyn_dress_fr_mid_l_03}}, {transformType=Rotation, targetBone {boneName=dyn_dress_l_03}}]
    constraintWeights [1,1]
   }
   n109 = "Constraint - dyn_dress_fr_l_04 (PR)"(ComponentPose=@n108, Alpha=1, ConstraintWeights_0=1, ConstraintWeights_1=1) {
    alphaBoolBlend {blendOption=Linear}
    boneToModify {boneName=dyn_dress_fr_l_04}
    constraintSetup [{targetBone {boneName=dyn_dress_fr_mid_l_04}}, {transformType=Rotation, targetBone {boneName=dyn_dress_l_04}}]
    constraintWeights [1,1]
   }
   n110 = "Constraint - dyn_dress_fr_mid_r_MSTR (P)"(ComponentPose=@n109, Alpha=1) {
    alphaBoolBlend {blendOption=Linear}
    boneToModify {boneName=dyn_dress_fr_mid_r_MSTR}
    constraintSetup [{targetBone {boneName=thigh_twist_01_r}}]
    constraintWeights [1]
   }
   n111 = "Constraint - dyn_dress_fr_mid_r_01 (P)"(ComponentPose=@n110, Alpha=1) {
    alphaBoolBlend {blendOption=Linear}
    boneToModify {boneName=dyn_dress_fr_mid_r_01}
    constraintSetup [{targetBone {boneName=thigh_twist_01_r}}]
    constraintWeights [1]
   }
   n112 = "Constraint - dyn_dress_fr_mid_r_02 (P)"(ComponentPose=@n111, Alpha=1) {
    alphaBoolBlend {blendOption=Linear}
    boneToModify {boneName=dyn_dress_fr_mid_r_02}
    constraintSetup [{targetBone {boneName=thigh_twist_01_r}}]
    constraintWeights [1]
   }
   n113 = "Constraint - dyn_dress_fr_mid_r_03 (P)"(ComponentPose=@n112, Alpha=1) {
    alphaBoolBlend {blendOption=Linear}
    boneToModify {boneName=dyn_dress_fr_mid_r_03}
    constraintSetup [{targetBone {boneName=thigh_twist_01_r}}]
    constraintWeights [1]
   }
   n114 = "Constraint - dyn_dress_fr_mid_r_04 (P)"(ComponentPose=@n113, Alpha=1) {
    alphaBoolBlend {blendOption=Linear}
    boneToModify {boneName=dyn_dress_fr_mid_r_04}
    constraintSetup [{targetBone {boneName=thigh_twist_01_r}}]
    constraintWeights [1]
   }
   n115 = "Constraint - dyn_dress_fr_r_MSTR (PR)"(ComponentPose=@n114, Alpha=1, ConstraintWeights_0=1, ConstraintWeights_1=1) {
    alphaBoolBlend {blendOption=Linear}
    boneToModify {boneName=dyn_dress_fr_r_MSTR}
    constraintSetup [{targetBone {boneName=dyn_dress_fr_mid_r_MSTR}}, {transformType=Rotation, targetBone {boneName=dyn_dress_l_MSTR}}]
    constraintWeights [1,1]
   }
   n116 = "Constraint - dyn_dress_fr_r_01 (PR)"(ComponentPose=@n115, Alpha=1, ConstraintWeights_0=1, ConstraintWeights_1=1) {
    alphaBoolBlend {blendOption=Linear}
    boneToModify {boneName=dyn_dress_fr_r_01}
    constraintSetup [{targetBone {boneName=dyn_dress_fr_mid_r_01}}, {transformType=Rotation, targetBone {boneName=dyn_dress_l_01}}]
    constraintWeights [1,1]
   }
   n117 = "Constraint - dyn_dress_fr_r_02 (PR)"(ComponentPose=@n116, Alpha=1, ConstraintWeights_0=1, ConstraintWeights_1=1) {
    alphaBoolBlend {blendOption=Linear}
    boneToModify {boneName=dyn_dress_fr_r_02}
    constraintSetup [{targetBone {boneName=dyn_dress_fr_mid_r_02}}, {transformType=Rotation, targetBone {boneName=dyn_dress_l_02}}]
    constraintWeights [1,1]
   }
   n118 = "Constraint - dyn_dress_fr_r_03 (PR)"(ComponentPose=@n117, Alpha=1, ConstraintWeights_0=1, ConstraintWeights_1=1) {
    alphaBoolBlend {blendOption=Linear}
    boneToModify {boneName=dyn_dress_fr_r_03}
    constraintSetup [{targetBone {boneName=dyn_dress_fr_mid_r_03}}, {transformType=Rotation, targetBone {boneName=dyn_dress_l_03}}]
    constraintWeights [1,1]
   }
   n119 = "Constraint - dyn_dress_fr_r_04 (PR)"(ComponentPose=@n118, Alpha=1, ConstraintWeights_0=1, ConstraintWeights_1=1) {
    alphaBoolBlend {blendOption=Linear}
    boneToModify {boneName=dyn_dress_fr_r_04}
    constraintSetup [{targetBone {boneName=dyn_dress_fr_mid_r_04}}, {transformType=Rotation, targetBone {boneName=dyn_dress_l_04}}]
    constraintWeights [1,1]
   }
   n120 = "Transform (Modify) Bone - Bone: dyn_dress_bk_r_03"(ComponentPose=@n119, Alpha=1, Translation="0,0,0", Rotation="0,-30,0", Scale="1,1,1") {
    rotationMode=BMM_Additive
    rotationSpace=BCS_BoneSpace
    translationMode=BMM_Additive
    translationSpace=BCS_BoneSpace
    alphaBoolBlend {blendOption=Linear}
    alphaScaleBiasClamp {bInterpResult=true, interpSpeedDecreasing=30, interpSpeedIncreasing=30}
    boneToModify {boneName=dyn_dress_bk_r_03}
    rotation {yaw=-30}
   }
   n121 = "Transform (Modify) Bone - Bone: dyn_dress_bk_r_04"(ComponentPose=@n120, Alpha=1, Translation="0,0,0", Rotation="0,-30,0", Scale="1,1,1") {
    rotationMode=BMM_Additive
    rotationSpace=BCS_BoneSpace
    translationMode=BMM_Additive
    translationSpace=BCS_BoneSpace
    alphaBoolBlend {blendOption=Linear}
    alphaScaleBiasClamp {bInterpResult=true, interpSpeedDecreasing=30, interpSpeedIncreasing=30}
    boneToModify {boneName=dyn_dress_bk_r_04}
    rotation {yaw=-30}
   }
   n122 = "Transform (Modify) Bone - Bone: dyn_dress_bk_r_05"(ComponentPose=@n121, Alpha=1, Translation="0,0,0", Rotation="0,-30,0", Scale="1,1,1") {
    rotationMode=BMM_Additive
    rotationSpace=BCS_BoneSpace
    translationMode=BMM_Additive
    translationSpace=BCS_BoneSpace
    alphaBoolBlend {blendOption=Linear}
    alphaScaleBiasClamp {bInterpResult=true, interpSpeedDecreasing=30, interpSpeedIncreasing=30}
    boneToModify {boneName=dyn_dress_bk_r_05}
    rotation {yaw=-30}
   }
   n123 = "Transform (Modify) Bone - Bone: dyn_dress_bk_r_06"(ComponentPose=@n122, Alpha=1, Translation="0,0,0", Rotation="0,-30,0", Scale="1,1,1") {
    rotationMode=BMM_Additive
    rotationSpace=BCS_BoneSpace
    translationMode=BMM_Additive
    translationSpace=BCS_BoneSpace
    alphaBoolBlend {blendOption=Linear}
    alphaScaleBiasClamp {bInterpResult=true, interpSpeedDecreasing=30, interpSpeedIncreasing=30}
    boneToModify {boneName=dyn_dress_bk_r_06}
    rotation {yaw=-30}
   }
   n124 = RigidBody(ComponentPose=@n123, OverrideWorldGravity="0,0,0") {
    bClampLinearTranslationLimitToRefPose=true
    bOverrideWorldGravity=true
    overridePhysicsAsset=/Game/Characters/Player/Female/Medium/Bodies/F_MED_Palespooky_Outlander/Meshes/F_MED_Palespooky_Outlander_Crouching_Dyn_Skirt
    simulationSpace=BaseBoneSpace
    alphaBoolBlend {blendOption=Linear}
    baseBoneRef {boneName=pelvis}
    componentLinearAccScale=<0.5 0.5 0.55>
    overrideWorldGravity {z=-980}
    simSpaceSettings {
     maxAngularAcceleration=1000
     maxAngularVelocity=3
     maxLinearAcceleration=700
     maxLinearVelocity=500
     velocityScaleZ=0
     worldAlpha=1
     externalLinearDragV=<0.2 0.2>
    }
   }
   n125 = "Constraint - dyn_dress_fr_mid_l_MSTR (P)"(ComponentPose=@n124, Alpha=1) {
    alphaBoolBlend {blendOption=Linear}
    boneToModify {boneName=dyn_dress_fr_mid_l_MSTR}
    constraintSetup [{targetBone {boneName=thigh_twist_01_l}}]
    constraintWeights [1]
   }
   n126 = "Constraint - dyn_dress_fr_mid_l_01 (P)"(ComponentPose=@n125, Alpha=1) {
    alphaBoolBlend {blendOption=Linear}
    boneToModify {boneName=dyn_dress_fr_mid_l_01}
    constraintSetup [{targetBone {boneName=thigh_twist_01_l}}]
    constraintWeights [1]
   }
   n127 = "Constraint - dyn_dress_fr_mid_l_02 (P)"(ComponentPose=@n126, Alpha=1) {
    alphaBoolBlend {blendOption=Linear}
    boneToModify {boneName=dyn_dress_fr_mid_l_02}
    constraintSetup [{targetBone {boneName=thigh_twist_01_l}}]
    constraintWeights [1]
   }
   n128 = "Constraint - dyn_dress_fr_mid_l_03 (P)"(ComponentPose=@n127, Alpha=1) {
    alphaBoolBlend {blendOption=Linear}
    boneToModify {boneName=dyn_dress_fr_mid_l_03}
    constraintSetup [{targetBone {boneName=thigh_twist_01_l}}]
    constraintWeights [1]
   }
   n129 = "Constraint - dyn_dress_fr_mid_l_04 (P)"(ComponentPose=@n128, Alpha=1) {
    alphaBoolBlend {blendOption=Linear}
    boneToModify {boneName=dyn_dress_fr_mid_l_04}
    constraintSetup [{targetBone {boneName=thigh_twist_01_l}}]
    constraintWeights [1]
   }
   n130 = "Constraint - dyn_dress_fr_mid_r_MSTR (P)"(ComponentPose=@n129, Alpha=1) {
    alphaBoolBlend {blendOption=Linear}
    boneToModify {boneName=dyn_dress_fr_mid_r_MSTR}
    constraintSetup [{targetBone {boneName=thigh_twist_01_r}}]
    constraintWeights [1]
   }
   n131 = "Constraint - dyn_dress_fr_mid_r_01 (P)"(ComponentPose=@n130, Alpha=1) {
    alphaBoolBlend {blendOption=Linear}
    boneToModify {boneName=dyn_dress_fr_mid_r_01}
    constraintSetup [{targetBone {boneName=thigh_twist_01_r}}]
    constraintWeights [1]
   }
   n132 = "Constraint - dyn_dress_fr_mid_r_02 (P)"(ComponentPose=@n131, Alpha=1) {
    alphaBoolBlend {blendOption=Linear}
    boneToModify {boneName=dyn_dress_fr_mid_r_02}
    constraintSetup [{targetBone {boneName=thigh_twist_01_r}}]
    constraintWeights [1]
   }
   n133 = "Constraint - dyn_dress_fr_mid_r_03 (P)"(ComponentPose=@n132, Alpha=1) {
    alphaBoolBlend {blendOption=Linear}
    boneToModify {boneName=dyn_dress_fr_mid_r_03}
    constraintSetup [{targetBone {boneName=thigh_twist_01_r}}]
    constraintWeights [1]
   }
   n134 = "Constraint - dyn_dress_fr_mid_r_04 (P)"(ComponentPose=@n133, Alpha=1) {
    alphaBoolBlend {blendOption=Linear}
    boneToModify {boneName=dyn_dress_fr_mid_r_04}
    constraintSetup [{targetBone {boneName=thigh_twist_01_r}}]
    constraintWeights [1]
   }
   n135 = "Constraint - dyn_dress_r_MSTR (PPP)"(ComponentPose=@n134, Alpha=1, ConstraintWeights_0=0.3, ConstraintWeights_1=0.3, ConstraintWeights_2=0.4) {
    alphaBoolBlend {blendOption=Linear}
    boneToModify {boneName=dyn_dress_r_MSTR}
    constraintSetup [{targetBone {boneName=dyn_dress_r_MSTR}}, {targetBone {boneName=dyn_dress_bk_r_MSTR}}, {targetBone {boneName=thigh_twist_01_r}}]
    constraintWeights [1,1,1]
   }
   n136 = "Constraint - dyn_dress_r_01 (PPP)"(ComponentPose=@n135, Alpha=1, ConstraintWeights_0=0.3, ConstraintWeights_1=0.3, ConstraintWeights_2=0.4) {
    alphaBoolBlend {blendOption=Linear}
    boneToModify {boneName=dyn_dress_r_01}
    constraintSetup [{targetBone {boneName=dyn_dress_r_01}}, {targetBone {boneName=dyn_dress_bk_r_01}}, {targetBone {boneName=thigh_twist_01_r}}]
    constraintWeights [1,1,1]
   }
   n137 = "Constraint - dyn_dress_r_02 (PPP)"(ComponentPose=@n136, Alpha=1, ConstraintWeights_0=0.3, ConstraintWeights_1=0.3, ConstraintWeights_2=0.4) {
    alphaBoolBlend {blendOption=Linear}
    boneToModify {boneName=dyn_dress_r_02}
    constraintSetup [{targetBone {boneName=dyn_dress_r_02}}, {targetBone {boneName=dyn_dress_bk_r_02}}, {targetBone {boneName=thigh_twist_01_r}}]
    constraintWeights [1,1,1]
   }
   n138 = "Constraint - dyn_dress_r_03 (PPP)"(ComponentPose=@n137, Alpha=1, ConstraintWeights_0=0.3, ConstraintWeights_1=0.3, ConstraintWeights_2=0.4) {
    alphaBoolBlend {blendOption=Linear}
    boneToModify {boneName=dyn_dress_r_03}
    constraintSetup [{targetBone {boneName=dyn_dress_r_03}}, {targetBone {boneName=dyn_dress_bk_r_03}}, {targetBone {boneName=thigh_twist_01_r}}]
    constraintWeights [1,1,1]
   }
   n139 = "Constraint - dyn_dress_r_04 (PPP)"(ComponentPose=@n138, Alpha=1, ConstraintWeights_0=0.3, ConstraintWeights_1=0.3, ConstraintWeights_2=0.4) {
    alphaBoolBlend {blendOption=Linear}
    boneToModify {boneName=dyn_dress_r_04}
    constraintSetup [{targetBone {boneName=dyn_dress_r_04}}, {targetBone {boneName=dyn_dress_bk_r_04}}, {targetBone {boneName=thigh_twist_01_r}}]
    constraintWeights [1,1,1]
   }
   n140 = "Constraint - dyn_dress_r_05 (PPP)"(ComponentPose=@n139, Alpha=1, ConstraintWeights_0=0.3, ConstraintWeights_1=0.3, ConstraintWeights_2=0.4) {
    alphaBoolBlend {blendOption=Linear}
    boneToModify {boneName=dyn_dress_r_05}
    constraintSetup [{targetBone {boneName=dyn_dress_r_05}}, {targetBone {boneName=dyn_dress_bk_r_05}}, {targetBone {boneName=thigh_twist_01_r}}]
    constraintWeights [1,1,1]
   }
   n141 = "Constraint - dyn_dress_r_06 (PPP)"(ComponentPose=@n140, Alpha=1, ConstraintWeights_0=0.3, ConstraintWeights_1=0.3, ConstraintWeights_2=0.4) {
    alphaBoolBlend {blendOption=Linear}
    boneToModify {boneName=dyn_dress_r_06}
    constraintSetup [{targetBone {boneName=dyn_dress_r_06}}, {targetBone {boneName=dyn_dress_bk_r_06}}, {targetBone {boneName=thigh_twist_01_r}}]
    constraintWeights [1,1,1]
   }
   n142 = "Transform (Modify) Bone - Bone: dyn_dress_fr_l_MSTR"(ComponentPose=@n141, Alpha=1, Translation="2,0.000,0.000", Rotation="0,0,0", Scale="1,1,1") {
    rotationMode=BMM_Additive
    scaleMode=BMM_Additive
    translationMode=BMM_Additive
    alphaBoolBlend {blendOption=Linear}
    boneToModify {boneName=dyn_dress_fr_l_MSTR}
    translation {x=2}
   }
   n143 = "Transform (Modify) Bone - Bone: dyn_dress_l_01"(ComponentPose=@n142, Alpha=1, Translation="1,0,0", Rotation="0,5,0", Scale="1,1,1") {
    rotationMode=BMM_Additive
    scaleMode=BMM_Additive
    translationMode=BMM_Additive
    alphaBoolBlend {blendOption=Linear}
    boneToModify {boneName=dyn_dress_l_01}
    rotation {yaw=5}
    translation {x=1}
   }
   n144 = Component To Local(ComponentPose=@n143)
   n145 = Get bIsCrouching()
   n146 = "Constraint - dyn_knife (R)"(
    ComponentPose=Local To Component(LocalPose=Blend Poses by bool(BlendPose_0=@n144, BlendPose_1=@n99, BlendTime_0=0.4, BlendTime_1=0.4, bActiveValue=@n145) {blendType=Linear})
    Alpha=Get KnifeFollow() -> (KnifeFollow:double)
    ConstraintWeights_0=1
   ) {
    alphaBoolBlend {blendOption=Linear}
    boneToModify {boneName=dyn_knife}
    constraintSetup [{transformType=Rotation, targetBone {boneName=thigh_twist_01_l}}]
    constraintWeights [1]
   }
   n147 = "Transform (Modify) Bone - Bone: dyn_dress_bk_r_03"(ComponentPose=@n146, Alpha=Get crouch alpha(), Translation="0,0,0", Rotation="0,-30,0", Scale="1,1,1") {
    rotationMode=BMM_Additive
    rotationSpace=BCS_BoneSpace
    translationMode=BMM_Additive
    translationSpace=BCS_BoneSpace
    alphaBoolBlend {blendOption=Linear}
    alphaScaleBiasClamp {bInterpResult=true, interpSpeedDecreasing=30, interpSpeedIncreasing=30}
    boneToModify {boneName=dyn_dress_bk_r_03}
    rotation {yaw=-30}
   }
   n148 = "Transform (Modify) Bone - Bone: dyn_knife"(ComponentPose=@n147, Alpha=1, Translation="0,0,0", Rotation="22.005,16.776,4.048", Scale="0.8,0.8,0.8") {
    rotationMode=BMM_Additive
    rotationSpace=BCS_BoneSpace
    scaleMode=BMM_Additive
    scaleSpace=BCS_BoneSpace
    translationMode=BMM_Additive
    translationSpace=BCS_BoneSpace
    alphaBoolBlend {blendOption=Linear}
    boneToModify {boneName=dyn_knife}
    rotation=<rot 22.005 16.776 4.048>
    scale=<0.8 0.8 0.8>
   }
   n149 = "Transform (Modify) Bone - Bone: dyn_knife"(ComponentPose=@n148, Alpha=Get crouch alpha(), Translation="0,3,0", Rotation="70,50,-50", Scale="1,1,1") {
    rotationMode=BMM_Additive
    rotationSpace=BCS_BoneSpace
    scaleMode=BMM_Additive
    scaleSpace=BCS_BoneSpace
    translationMode=BMM_Additive
    translationSpace=BCS_BoneSpace
    alphaBoolBlend {blendOption=Linear}
    boneToModify {boneName=dyn_knife}
    rotation=<rot 70 50 -50>
    translation {y=3}
   }
   n150 = "Transform (Modify) Bone - Bone: lowerarm_twist_02_l"!(
    ComponentPose=@n149
    bAlphaBoolEnabled:bool=Get bIsLeftArmHidden() -> (bIsLeftArmHidden:bool)
    Translation="0,0,0"
    Rotation="0,0,0"
    Scale="0.7,0.4,0.4"
   ) -> (Pose) {
    alphaInputType=Bool
    rotationMode=BMM_Additive
    rotationSpace=BCS_BoneSpace
    scaleMode=BMM_Additive
    scaleSpace=BCS_BoneSpace
    translationMode=BMM_Additive
    translationSpace=BCS_BoneSpace
    alphaBoolBlend {blendOption=Linear}
    boneToModify {boneName=lowerarm_twist_02_l}
    scale=<0.7 0.4 0.4>
   }
   n151 = "Transform (Modify) Bone - Bone: lowerarm_twist_02_r"!(
    ComponentPose=@n150
    bAlphaBoolEnabled:bool=Get bIsRightArmHidden() -> (bIsRightArmHidden:bool)
    Translation="0,0,0"
    Rotation="0,0,0"
    Scale="0.7,0.4,0.4"
   ) -> (Pose) {
    alphaInputType=Bool
    rotationMode=BMM_Additive
    rotationSpace=BCS_BoneSpace
    scaleMode=BMM_Additive
    scaleSpace=BCS_BoneSpace
    translationMode=BMM_Additive
    translationSpace=BCS_BoneSpace
    alphaBoolBlend {blendOption=Linear}
    boneToModify {boneName=lowerarm_twist_02_r}
    scale=<0.7 0.4 0.4>
   }
   Output Pose(Result=Component To Local(ComponentPose=@n151)) {layerGroup=None}
   n152 = "Transform (Modify) Bone - Bone: dyn_root_slope"!(
    ComponentPose=Local To Component(LocalPose=Copy Pose From Mesh() {bUseAttachedParent=true})
    bAlphaBoolEnabled:bool=True
    Translation=Get slope_root_trans() -> (slope_root_trans:Vector)
    Rotation=Get slope_foot_rot() -> (slope_foot_rot:Rotator)
    Scale="1,1,1"
   ) -> (Pose) {
    alphaInputType=Bool
    rotationMode=BMM_Replace
    rotationSpace=BCS_WorldSpace
    translationMode=BMM_Replace
    translationSpace=BCS_WorldSpace
    alphaBoolBlend {blendOption=Linear}
    boneToModify {boneName=dyn_root_slope}
   }
   n153 = Get in vehicle scale alpha() -> (in vehicle scale alpha:double)
   n154 = "Transform (Modify) Bone - Bone: dyn_dress_bk_MSTR"(ComponentPose=@n152, Alpha=@n153, Translation="0,0,0", Rotation="0,0,0", Scale="0.1,0.1,0.1") {scaleMode=BMM_Replace, alphaBoolBlend {blendOption=Linear}, boneToModify {boneName=dyn_dress_bk_MSTR}, scale=<0.1 0.1 0.1>}
   n155 = "Transform (Modify) Bone - Bone: dyn_dress_bk_l_MSTR"(ComponentPose=@n154, Alpha=@n153, Translation="0,0,0", Rotation="0,0,0", Scale="0.1,0.1,0.1") {scaleMode=BMM_Replace, alphaBoolBlend {blendOption=Linear}, boneToModify {boneName=dyn_dress_bk_l_MSTR}, scale=<0.1 0.1 0.1>}
   n156 = "Transform (Modify) Bone - Bone: dyn_dress_l_MSTR"!(ComponentPose=@n155, Alpha:float=@n153, Translation="0,0,0", Rotation="0,0,0", Scale="0.1,0.1,0.1") -> (Pose) {scaleMode=BMM_Replace, alphaBoolBlend {blendOption=Linear}, boneToModify {boneName=dyn_dress_l_MSTR}, scale=<0.1 0.1 0.1>}
   n157 = "Transform (Modify) Bone - Bone: dyn_dress_bk_r_MSTR"(ComponentPose=@n156, Alpha=@n153, Translation="0,0,0", Rotation="0,0,0", Scale="0.1,0.1,0.1") {scaleMode=BMM_Replace, alphaBoolBlend {blendOption=Linear}, boneToModify {boneName=dyn_dress_bk_r_MSTR}, scale=<0.1 0.1 0.1>}
   n158 = "Transform (Modify) Bone - Bone: dyn_dress_r_MSTR"(ComponentPose=@n157, Alpha=@n153, Translation="0,0,0", Rotation="0,0,0", Scale="0.1,0.1,0.1") {scaleMode=BMM_Replace, alphaBoolBlend {blendOption=Linear}, boneToModify {boneName=dyn_dress_r_MSTR}, scale=<0.1 0.1 0.1>}
   n159 = Get crouch alpha()
   n160 = "Transform (Modify) Bone - Bone: dyn_dress_fr_mid_l_MSTR"(ComponentPose=@n158, Alpha=@n159, Translation="-3.497,-0.000,1.765", Rotation="0,0,0", Scale="1,1,1") {
    rotationMode=BMM_Additive
    rotationSpace=BCS_BoneSpace
    translationMode=BMM_Additive
    translationSpace=BCS_BoneSpace
    alphaBoolBlend {blendOption=Linear}
    alphaScaleBiasClamp {bInterpResult=true, interpSpeedDecreasing=30, interpSpeedIncreasing=30}
    boneToModify {boneName=dyn_dress_fr_mid_l_MSTR}
    translation {x=-3.497, z=1.765}
   }
   n161 = "Transform (Modify) Bone - Bone: dyn_dress_fr_l_MSTR"(ComponentPose=@n160, Alpha=@n159, Translation="-4.497,0.000,0.510", Rotation="7.727,36.528,6.069", Scale="1,1,1") {
    rotationMode=BMM_Additive
    rotationSpace=BCS_BoneSpace
    translationMode=BMM_Additive
    translationSpace=BCS_BoneSpace
    alphaBoolBlend {blendOption=Linear}
    alphaScaleBiasClamp {bInterpResult=true, interpSpeedDecreasing=30, interpSpeedIncreasing=30}
    boneToModify {boneName=dyn_dress_fr_l_MSTR}
    rotation=<rot 7.727 36.528 6.069>
    translation {x=-4.497, z=0.51}
   }
   n162 = "Transform (Modify) Bone - Bone: dyn_dress_fr_MSTR"(ComponentPose=@n161, Alpha=@n159, Translation="-3.947,-0.000,-0.000", Rotation="0,0,0", Scale="1,1,1") {
    rotationMode=BMM_Additive
    rotationSpace=BCS_BoneSpace
    translationMode=BMM_Additive
    translationSpace=BCS_BoneSpace
    alphaBoolBlend {blendOption=Linear}
    alphaScaleBiasClamp {bInterpResult=true, interpSpeedDecreasing=30, interpSpeedIncreasing=30}
    boneToModify {boneName=dyn_dress_fr_MSTR}
    translation=<-3.9469 0 0>
   }
   n163 = Anim Dynamics - Bone: dyn_bag_back(ComponentPose=@n162, Alpha=0.5, ExternalForce="0,0,0") {
    angularDampingOverride=0.99
    bAngularSpring=true
    bGravityOverrideInSimSpace=true
    bOverrideAngularDamping=true
    bOverrideLinearDamping=true
    gravityScale=0.3
    lODThreshold=1
    linearDampingOverride=0.99
    alphaBoolBlend {blendOption=Linear}
    boundBone {boneName=dyn_bag_back}
    physicsBodyDefinitions [{
      sphereCollisionRadius=0
      boundBone {boneName=dyn_bag_back}
      boxExtents=<7 7 4>
      constraintSetup {
       angularTargetAxis=AxisZ
       angularLimitsMax=<45 10 10>
       angularLimitsMin=<-130.9873 -10 -10>
       angularTarget=<0 0 1>
       linearAxesMax=<0 0 0.5>
       linearAxesMin=<0 0 -0.5>
      }
      localJointOffset=<0 0 -2>
     }]
    relativeSpaceBone {boneName=spine_01}
   }
   n164 = Anim Dynamics - Bone: dyn_bag_front(ComponentPose=@n163, Alpha=0.5, ExternalForce="0,0,0") {
    angularDampingOverride=0.99
    bAngularSpring=true
    bGravityOverrideInSimSpace=true
    bOverrideAngularDamping=true
    bOverrideLinearDamping=true
    gravityScale=0.2
    lODThreshold=1
    linearDampingOverride=0.99
    alphaBoolBlend {blendOption=Linear}
    boundBone {boneName=dyn_bag_front}
    physicsBodyDefinitions [{
      sphereCollisionRadius=0
      boundBone {boneName=dyn_bag_front}
      boxExtents=<15 15 10>
      constraintSetup {
       angularTargetAxis=AxisZ
       angularLimitsMax=<45 15.3184 25.2229>
       angularLimitsMin=<-130.9873 -15.3184 -22.1656>
       angularTarget=<0 0 1>
       linearAxesMax=<0 0 0.5>
       linearAxesMin=<0 0 -0.5>
      }
      localJointOffset=<0 0 -2>
     }]
    relativeSpaceBone {boneName=spine_01}
   }
   Save cached pose 'preSkirt'(Pose=Component To Local(ComponentPose=@n164)) {cacheName=preSkirt}
  ]}
 EventGraph {nodes [
   Event BlueprintUpdateAnimation()
   > n1 = Sequence!(execute) -> (then_0, then_1, then_2:Exec, then_3:Exec, then_4:Exec, then_5:Exec, then_6:Exec)
   > n2 = Branch(Condition=Get bIsSprinting())
   > Set sprinting alpha(sprinting alpha:double=1) -> (Output_Get:double)
   Set sprinting alpha(execute=@n2.else, sprinting alpha=0)
   n3 = Branch(execute=@n1.then_1, Condition=Get bIsSwimSprinting() -> (bIsSwimSprinting:bool))
   > Set gravity_override(gravity_override="0, 0,-200")
   n4 = Branch(execute=@n3.else, Condition=Get bIsSkyDiving())
   > n5 = Branch(Condition=Get bIsParachuteOpened())
   Set gravity_override(execute=[@n4.else, @n5.then], gravity_override="0, 0,-980")
   > Set rigids alpha(rigids alpha:double=1) -> (Output_Get:double)
   n6 = Branch(execute=@n5.else, Condition="float > float"(A=Get PawnVelocityZ(), B=50) {is=title K2Node_CallFunction})
   > n7 = Set gravity_override(gravity_override="0, 0,200")
   Set rigids alpha(execute=[@n7.then, @n8.then], rigids alpha=0.5)
   n8 = Set gravity_override(execute=@n6.else, gravity_override="0, 0,-200")
   n9 = Branch(execute=@n1.then_2, Condition=Get bIsOstrichVehicleDriver() -> (bIsOstrichVehicleDriver:bool))
   Set in vehicle rigid alpha(execute=[@n11.then, @n10.then, @n9.then], in vehicle rigid alpha:double=0) -> (Output_Get:double)
   > Set in vehicle scale alpha(in vehicle scale alpha:double=1) -> (Output_Get:double)
   n10 = Branch(execute=@n9.else, Condition=Get bIsFerretVehicleDriver() -> (bIsFerretVehicleDriver:bool))
   n11 = Branch(execute=@n10.else, Condition=Get bIsOctopusVehicleDriver() -> (bIsOctopusVehicleDriver:bool))
   Set in vehicle rigid alpha(execute=@n11.else, in vehicle rigid alpha=1)
   > Set in vehicle scale alpha(in vehicle scale alpha=0)
   n12 = BreakTransform(InTransform=GetDeltaTransformFromRefPose(self=GetOwnerSkeletalMeshComponent(), BoneName=thigh_l, BaseName=calf_l))
   Set KnifeFollow(execute=@n1.then_3, KnifeFollow:double=MapRangeClamped(Value=BreakRotator(InRot=@n12.Rotation).Yaw, InRangeA=-23, InRangeB=-60)) -> (Output_Get:double)
   Set bPlayingEmoteOrFullbodyFrontend(
    execute=@n1.then_4
    bPlayingEmoteOrFullbodyFrontend:bool=OR Boolean(A=Get bIsPlayingEmote(), B=Get bIsPlayingFullBodySlotInFrontEnd() -> (bIsPlayingFullBodySlotInFrontEnd:bool))
   ) -> (Output_Get:bool)
   n13 = Break Fort Anim Input Slope Info!(FortAnimInput_SlopeInfo:"Fort Anim Input Slope Info Structure (by ref)"=Get SlopeInfoInput() -> (SlopeInfoInput:Fort Anim Input Slope Info Structure)) -> (RootSlopeTranslation:Vector=~, LeftFootSlopeTranslation:Vector=~, RightFootSlopeTranslation:Vector=~, SlopeRotation:Rotator=~) {is=title K2Node_BreakStruct}
   Set slope_root_trans(execute=@n1.then_5, slope_root_trans:Vector=@n13) -> (Output_Get:Vector)
   > Set slope_foot_rot(slope_foot_rot:Rotator=@n13.SlopeRotation) -> (Output_Get:Rotator)
   Event OnParachuteOpened()
   > n14 = Delay(Duration=1.5)
   ResetDynamics(execute=[@n14.then, @n15.then, @n16.then], InTeleportType=TeleportPhysics)
   > ResetDynamics(InTeleportType=ResetPhysics)
   n15 = Event OnBeginSkydiving()
   n16 = Event OnEndSkydiving()
   Event OnBeginCrouching()
   > Set crouch alpha(crouch alpha:double=1) -> (Output_Get:double)
   Event OnEndCrouching()
   > Set crouch alpha(crouch alpha=0)
   Event OnEndPlayingEmote() {is=title K2Node_Event}
   Event OnBeginPlayingEmote() {is=title K2Node_Event}
   Set crouching(crouching:bool=Get bIsCrouching()) -> (Output_Get:bool)
  ]}
]
variables [
 bPlayingEmoteOrFullbodyFrontend {type=boolean}
 crouch alpha {type=number}
 crouching {type=boolean}
 gravity_override {type=Vector}
 in vehicle rigid alpha {type=number}
 in vehicle scale alpha {type=number}
 KnifeFollow {type=number}
 LookDown {type=number}
 LookLevel {type=number}
 LookUp {type=number}
 rigids alpha {type=number, default=1}
 slope_foot_rot {type=Rotator}
 slope_root_trans {type=Vector}
 sprinting alpha {type=number}
]
