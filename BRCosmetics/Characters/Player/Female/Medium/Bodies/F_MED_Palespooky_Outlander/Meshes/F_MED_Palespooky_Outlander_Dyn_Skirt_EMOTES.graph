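# Physics asset for the dynamic skirt on the female medium Palespooky Outlander body, apparently the variant used during emotes.
# Five six-box chains (back, back-left, back-right, left, right) hang from pelvis-linked master bones, alongside pelvis and leg capsules as body shapes.
# Back chains swing widely, while side chains are much stiffer, especially the left.
# Angular-free cross-constraints tie neighbouring chains at rows 3 or 5, asymmetrically on the sides, linking the panels together.
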PhysicsAsset /BRCosmetics/Characters/Player/Female/Medium/Bodies/F_MED_Palespooky_Outlander/Meshes/F_MED_Palespooky_Outlander_Dyn_Skirt_EMOTES
bodies {
 calf_l [{shapeType=Capsule, length=47.974, radius=7.4446, center=<-25.5048 -3.2385 0.0729>, rotation=<rot 87.134 -179.9944 -180>}]
 calf_r [{shapeType=Capsule, length=46.88, radius=8.5385, center=<29.0903 3.2385 -0.0731>, rotation=<rot 87.1272 -179.9944 -180>}]
 dyn_dress_bk_01 [{shapeType=Box, extentX=13.4375, extentY=14.9375, extentZ=1.5625, center=<5.954 0 -0.0003>}]
 dyn_dress_bk_02 [{shapeType=Box, extentX=13.4375, extentY=14.9375, extentZ=1.5625, center=<5.954 0 -0.0003>}]
 dyn_dress_bk_03 [{shapeType=Box, extentX=13.4375, extentY=14.9375, extentZ=1.5625, center=<5.954 0 -0.0003>}]
 dyn_dress_bk_04 [{shapeType=Box, extentX=13.4375, extentY=20.3019, extentZ=1.5625, center=<5.954 0 -0.0003>}]
 dyn_dress_bk_05 [{shapeType=Box, extentX=13.4375, extentY=30.2954, extentZ=1.5625, center=<5.954 0 -0.0003>}]
 dyn_dress_bk_06 [{shapeType=Box, extentX=18.0767, extentY=30.2954, extentZ=5.1499, center=<9.1899 0 -0.0006>}]
 dyn_dress_bk_MSTR [{shapeType=Capsule, length=0.2, radius=0.2}]
 dyn_dress_bk_l_01 [{shapeType=Box, extentX=13.4375, extentY=14.9375, extentZ=1.154, center=<5.954 0 -0.0003>, rotation=<rot 0 0 10.0004>}]
 dyn_dress_bk_l_02 [{shapeType=Box, extentX=13.4375, extentY=18.8081, extentZ=1.3815, center=<5.954 0 -0.0003>, rotation=<rot 0 0 10.0004>}]
 dyn_dress_bk_l_03 [{shapeType=Box, extentX=13.4375, extentY=18.8081, extentZ=3.03, center=<5.954 0 -0.0003>, rotation=<rot 0 0 10.0004>}]
 dyn_dress_bk_l_04 [{shapeType=Box, extentX=13.4375, extentY=14.9375, extentZ=1.5625, center=<5.954 0 -0.0003>}]
 dyn_dress_bk_l_05 [{shapeType=Box, extentX=13.4375, extentY=20.3019, extentZ=1.5625, center=<5.954 0 -0.0003>}]
 dyn_dress_bk_l_06 [{shapeType=Box, extentX=13.4375, extentY=22.7261, extentZ=1.5625, center=<5.954 0 -0.0003>, rotation=<rot 0 0 -12.657>}]
 dyn_dress_bk_l_MSTR [{shapeType=Capsule, length=0.2, radius=0.2}]
 dyn_dress_bk_r_01 [{shapeType=Box, extentX=13.4375, extentY=14.9375, extentZ=1.5625, center=<5.954 0 -0.0003>}]
 dyn_dress_bk_r_02 [{shapeType=Box, extentX=13.4375, extentY=17.8612, extentZ=1.9644, center=<5.954 0 -0.0003>}]
 dyn_dress_bk_r_03 [{shapeType=Box, extentX=13.4375, extentY=14.9375, extentZ=1.5625, center=<5.954 0 -0.0003>}]
 dyn_dress_bk_r_04 [{shapeType=Box, extentX=13.4375, extentY=14.9375, extentZ=1.5625, center=<5.954 0 -0.0003>}]
 dyn_dress_bk_r_05 [{shapeType=Box, extentX=13.4375, extentY=20.3019, extentZ=1.5625, center=<5.954 0 -0.0003>}]
 dyn_dress_bk_r_06 [{shapeType=Box, extentX=13.4375, extentY=24.2929, extentZ=1.5625, center=<5.954 0 -0.0003>, rotation=<rot 0 0 7.8753>}]
 dyn_dress_bk_r_MSTR [{shapeType=Capsule, length=0.2, radius=0.2}]
 dyn_dress_l_01 [{shapeType=Box, extentX=13.4375, extentY=14.9375, extentZ=1.5625, center=<5.954 0 -0.0003>}]
 dyn_dress_l_02 [{shapeType=Box, extentX=13.4375, extentY=14.9375, extentZ=2.8595, center=<5.954 0 -0.0003>}]
 dyn_dress_l_03 [{shapeType=Box, extentX=13.4375, extentY=14.9375, extentZ=2.8595, center=<5.954 0 -0.0003>}]
 dyn_dress_l_04 [{shapeType=Box, extentX=13.4375, extentY=14.9375, extentZ=1.5625, center=<5.954 0 -0.0003>}]
 dyn_dress_l_05 [{shapeType=Box, extentX=13.4375, extentY=14.9375, extentZ=1.5625, center=<5.954 0 -0.0003>}]
 dyn_dress_l_06 [{shapeType=Box, extentX=13.4375, extentY=14.9375, extentZ=1.5625, center=<5.954 0 -0.0003>}]
 dyn_dress_l_MSTR [{shapeType=Capsule, length=0.2, radius=0.2}]
 dyn_dress_r_01 [{shapeType=Box, extentX=13.4375, extentY=20.0302, extentZ=1.5625, center=<5.954 0 -0.0003>, rotation=<rot 0 0 -3.3758>}]
 dyn_dress_r_02 [{shapeType=Box, extentX=13.4375, extentY=20.0302, extentZ=3.6394, center=<5.954 0 -0.0003>, rotation=<rot 0 0 -3.3758>}]
 dyn_dress_r_03 [{shapeType=Box, extentX=13.4375, extentY=20.0302, extentZ=3.6394, center=<5.954 0 -0.0003>, rotation=<rot 0 0 -3.3758>}]
 dyn_dress_r_04 [{shapeType=Box, extentX=13.4375, extentY=20.0302, extentZ=2.3425, center=<5.954 0 -0.0003>}]
 dyn_dress_r_05 [{shapeType=Box, extentX=13.4375, extentY=20.0302, extentZ=2.3425, center=<5.954 0 -0.0003>}]
 dyn_dress_r_06 [{shapeType=Box, extentX=13.4375, extentY=23.1818, extentZ=2.3425, center=<5.954 0 -0.0003>, rotation=<rot 0 0 15.0005>}]
 dyn_dress_r_MSTR [{shapeType=Capsule, length=0.2, radius=0.2}]
 pelvis [{shapeType=Capsule, length=8.5448, radius=8.6423, center=<-5.3601 -2.3275 -0.0005>}]
 thigh_l [{shapeType=Capsule, length=28.752, radius=9.5465, center=<-22.5439 -2.5 1.9881>, rotation=<rot 82.0781 -180 -180>}]
 thigh_r [{shapeType=Capsule, length=28.752, radius=9.5465, center=<22.5439 2.5 -1.9881>, rotation=<rot 78.9172 -180 -180>}]
}
constraints [
 {bone1Name=dyn_dress_bk_01, bone2Name=dyn_dress_bk_MSTR, swing1LimitDegrees=45, swing2LimitDegrees=45, twistLimitDegrees=5}
 {bone1Name=dyn_dress_bk_02, bone2Name=dyn_dress_bk_01, swing1LimitDegrees=45, swing2LimitDegrees=45, twistLimitDegrees=5}
 {bone1Name=dyn_dress_bk_03, bone2Name=dyn_dress_bk_02, swing1LimitDegrees=45, swing2LimitDegrees=45, twistLimitDegrees=5}
 {bone1Name=dyn_dress_bk_04, bone2Name=dyn_dress_bk_03, swing1LimitDegrees=45, swing2LimitDegrees=45, twistLimitDegrees=5}
 {bone1Name=dyn_dress_bk_05, bone2Name=dyn_dress_bk_04, swing1LimitDegrees=45, swing2LimitDegrees=45, twistLimitDegrees=5}
 {bone1Name=dyn_dress_bk_06, bone2Name=dyn_dress_bk_05, swing1LimitDegrees=45, swing2LimitDegrees=45, twistLimitDegrees=5}
 {bone1Name=dyn_dress_bk_l_01, bone2Name=dyn_dress_bk_l_MSTR, swing1LimitDegrees=45, swing2LimitDegrees=45, twistLimitDegrees=5}
 {bone1Name=dyn_dress_bk_l_02, bone2Name=dyn_dress_bk_l_01, swing1LimitDegrees=45, swing2LimitDegrees=45, twistLimitDegrees=5}
 {bone1Name=dyn_dress_bk_l_03, bone2Name=dyn_dress_bk_l_02, swing1LimitDegrees=45, swing2LimitDegrees=45, twistLimitDegrees=5}
 {bone1Name=dyn_dress_bk_l_04, bone2Name=dyn_dress_bk_l_03, swing1LimitDegrees=45, swing2LimitDegrees=45, twistLimitDegrees=5}
 {bone1Name=dyn_dress_bk_l_05, bone2Name=dyn_dress_bk_l_04, swing1LimitDegrees=45, swing2LimitDegrees=45, twistLimitDegrees=5}
 {bone1Name=dyn_dress_bk_l_06, bone2Name=dyn_dress_bk_l_05, swing1LimitDegrees=45, swing2LimitDegrees=45, twistLimitDegrees=5}
 {bone1Name=dyn_dress_l_01, bone2Name=dyn_dress_l_MSTR, swing1LimitDegrees=10, swing2LimitDegrees=10, twistLimitDegrees=5}
 {bone1Name=dyn_dress_l_02, bone2Name=dyn_dress_l_01, swing1LimitDegrees=10, swing2LimitDegrees=5, twistLimitDegrees=5}
 {bone1Name=dyn_dress_l_03, bone2Name=dyn_dress_l_02, swing1LimitDegrees=2, swing2LimitDegrees=5, twistLimitDegrees=5}
 {bone1Name=dyn_dress_l_04, bone2Name=dyn_dress_l_03, swing1LimitDegrees=2, swing2LimitDegrees=5, twistLimitDegrees=5}
 {bone1Name=dyn_dress_l_05, bone2Name=dyn_dress_l_04, swing1LimitDegrees=2, swing2LimitDegrees=5, twistLimitDegrees=5}
 {bone1Name=dyn_dress_l_06, bone2Name=dyn_dress_l_05, swing1LimitDegrees=2, swing2LimitDegrees=5, twistLimitDegrees=5}
 {bone1Name=dyn_dress_bk_r_01, bone2Name=dyn_dress_bk_r_MSTR, swing1LimitDegrees=45, swing2LimitDegrees=45, twistLimitDegrees=5}
 {bone1Name=dyn_dress_bk_r_02, bone2Name=dyn_dress_bk_r_01, swing1LimitDegrees=45, swing2LimitDegrees=45, twistLimitDegrees=5}
 {bone1Name=dyn_dress_bk_r_03, bone2Name=dyn_dress_bk_r_02, swing1LimitDegrees=45, swing2LimitDegrees=45, twistLimitDegrees=5}
 {bone1Name=dyn_dress_bk_r_04, bone2Name=dyn_dress_bk_r_03, swing1LimitDegrees=45, swing2LimitDegrees=45, twistLimitDegrees=5}
 {bone1Name=dyn_dress_bk_r_05, bone2Name=dyn_dress_bk_r_04, swing1LimitDegrees=45, swing2LimitDegrees=45, twistLimitDegrees=5}
 {bone1Name=dyn_dress_bk_r_06, bone2Name=dyn_dress_bk_r_05, swing1LimitDegrees=45, swing2LimitDegrees=45, twistLimitDegrees=5}
 {bone1Name=dyn_dress_r_01, bone2Name=dyn_dress_r_MSTR, swing1LimitDegrees=10, swing2LimitDegrees=10, twistLimitDegrees=5}
 {bone1Name=dyn_dress_r_02, bone2Name=dyn_dress_r_01, swing1LimitDegrees=10, swing2LimitDegrees=10, twistLimitDegrees=5}
 {bone1Name=dyn_dress_r_03, bone2Name=dyn_dress_r_02, swing1LimitDegrees=10, swing2LimitDegrees=10, twistLimitDegrees=5}
 {bone1Name=dyn_dress_r_04, bone2Name=dyn_dress_r_03, swing1LimitDegrees=10, swing2LimitDegrees=10, twistLimitDegrees=5}
 {bone1Name=dyn_dress_r_05, bone2Name=dyn_dress_r_04, swing1LimitDegrees=10, swing2LimitDegrees=10, twistLimitDegrees=5}
 {bone1Name=dyn_dress_r_06, bone2Name=dyn_dress_r_05, swing1LimitDegrees=10, swing2LimitDegrees=10, twistLimitDegrees=5}
 {
  bone1Name=dyn_dress_bk_MSTR
  bone2Name=pelvis
  swing1LimitDegrees=45
  swing1Motion=Free
  swing2LimitDegrees=45
  swing2Motion=Free
  twistLimitDegrees=45
  twistMotion=Free
 }
 {
  bone1Name=dyn_dress_bk_l_MSTR
  bone2Name=pelvis
  swing1LimitDegrees=45
  swing1Motion=Free
  swing2LimitDegrees=45
  swing2Motion=Free
  twistLimitDegrees=45
  twistMotion=Free
 }
 {
  bone1Name=dyn_dress_l_MSTR
  bone2Name=pelvis
  swing1LimitDegrees=45
  swing1Motion=Free
  swing2LimitDegrees=45
  swing2Motion=Free
  twistLimitDegrees=45
  twistMotion=Free
 }
 {
  bone1Name=dyn_dress_bk_r_MSTR
  bone2Name=pelvis
  swing1LimitDegrees=45
  swing1Motion=Free
  swing2LimitDegrees=45
  swing2Motion=Free
  twistLimitDegrees=45
  twistMotion=Free
 }
 {
  bone1Name=dyn_dress_r_MSTR
  bone2Name=pelvis
  swing1LimitDegrees=45
  swing1Motion=Free
  swing2LimitDegrees=45
  swing2Motion=Free
  twistLimitDegrees=45
  twistMotion=Free
 }
 {
  bone1Name=thigh_l
  bone2Name=pelvis
  swing1LimitDegrees=45
  swing1Motion=Free
  swing2LimitDegrees=45
  swing2Motion=Free
  twistLimitDegrees=45
  twistMotion=Free
 }
 {
  bone1Name=calf_l
  bone2Name=thigh_l
  swing1LimitDegrees=45
  swing1Motion=Free
  swing2LimitDegrees=45
  swing2Motion=Free
  twistLimitDegrees=45
  twistMotion=Free
 }
 {
  bone1Name=thigh_r
  bone2Name=pelvis
  swing1LimitDegrees=45
  swing1Motion=Free
  swing2LimitDegrees=45
  swing2Motion=Free
  twistLimitDegrees=45
  twistMotion=Free
 }
 {
  bone1Name=calf_r
  bone2Name=thigh_r
  swing1LimitDegrees=45
  swing1Motion=Free
  swing2LimitDegrees=45
  swing2Motion=Free
  twistLimitDegrees=45
  twistMotion=Free
 }
 {
  bone1Name=dyn_dress_bk_03
  bone2Name=dyn_dress_bk_r_03
  swing1LimitDegrees=45
  swing1Motion=Free
  swing2LimitDegrees=45
  swing2Motion=Free
  twistLimitDegrees=45
  twistMotion=Free
 }
 {
  bone1Name=dyn_dress_bk_05
  bone2Name=dyn_dress_bk_r_05
  swing1LimitDegrees=45
  swing1Motion=Free
  swing2LimitDegrees=45
  swing2Motion=Free
  twistLimitDegrees=45
  twistMotion=Free
 }
 {
  bone1Name=dyn_dress_bk_03
  bone2Name=dyn_dress_bk_l_03
  swing1LimitDegrees=45
  swing1Motion=Free
  swing2LimitDegrees=45
  swing2Motion=Free
  twistLimitDegrees=45
  twistMotion=Free
 }
 {
  bone1Name=dyn_dress_bk_05
  bone2Name=dyn_dress_bk_l_05
  swing1LimitDegrees=45
  swing1Motion=Free
  swing2LimitDegrees=45
  swing2Motion=Free
  twistLimitDegrees=45
  twistMotion=Free
 }
 {
  bone1Name=dyn_dress_r_05
  bone2Name=dyn_dress_bk_r_05
  swing1LimitDegrees=45
  swing1Motion=Free
  swing2LimitDegrees=45
  swing2Motion=Free
  twistLimitDegrees=45
  twistMotion=Free
 }
 {
  bone1Name=dyn_dress_l_03
  bone2Name=dyn_dress_bk_l_03
  swing1LimitDegrees=45
  swing1Motion=Free
  swing2LimitDegrees=45
  swing2Motion=Free
  twistLimitDegrees=45
  twistMotion=Free
 }
]
