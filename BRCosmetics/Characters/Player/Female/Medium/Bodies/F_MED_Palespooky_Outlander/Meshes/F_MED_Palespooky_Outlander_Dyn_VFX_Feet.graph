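# Physics asset for the dynamic VFX foot bones on the female medium Palespooky Outlander body mesh.
# Each foot bone has a larger capsule body and anchors two three-bone chains of thin, identical capsules: a and c on the left, c and d on the right.
# Every chain link is constrained to its parent with the same swing and tight twist limits, giving bendable strands that hang off the feet.
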PhysicsAsset /BRCosmetics/Characters/Player/Female/Medium/Bodies/F_MED_Palespooky_Outlander/Meshes/F_MED_Palespooky_Outlander_Dyn_VFX_Feet
bodies {
 dyn_vfx_foot_l_a_01 [{shapeType=Capsule, length=5, radius=0.5, center=<2.5 0 0>, rotation=<rot 90 0 0>}]
 dyn_vfx_foot_l_a_02 [{shapeType=Capsule, length=5, radius=0.5, center=<2.5 0 0>, rotation=<rot 90 0 0>}]
 dyn_vfx_foot_l_a_03 [{shapeType=Capsule, length=5, radius=0.5, center=<2.5 0 0>, rotation=<rot 90 0 0>}]
 dyn_vfx_foot_l_c_01 [{shapeType=Capsule, length=5, radius=0.5, center=<2.5 0 0>, rotation=<rot 90 0 0>}]
 dyn_vfx_foot_l_c_02 [{shapeType=Capsule, length=5, radius=0.5, center=<2.5 0 0>, rotation=<rot 90 0 0>}]
 dyn_vfx_foot_l_c_03 [{shapeType=Capsule, length=5, radius=0.5, center=<2.5 0 0>, rotation=<rot 90 0 0>}]
 dyn_vfx_foot_r_c_01 [{shapeType=Capsule, length=5, radius=0.5, center=<2.5 0 0>, rotation=<rot 90 0 0>}]
 dyn_vfx_foot_r_c_02 [{shapeType=Capsule, length=5, radius=0.5, center=<2.5 0 0>, rotation=<rot 90 0 0>}]
 dyn_vfx_foot_r_c_03 [{shapeType=Capsule, length=5, radius=0.5, center=<2.5 0 0>, rotation=<rot 90 0 0>}]
 dyn_vfx_foot_r_d_01 [{shapeType=Capsule, length=5, radius=0.5, center=<2.5 0 0>, rotation=<rot 90 0 0>}]
 dyn_vfx_foot_r_d_02 [{shapeType=Capsule, length=5, radius=0.5, center=<2.5 0 0>, rotation=<rot 90 0 0>}]
 dyn_vfx_foot_r_d_03 [{shapeType=Capsule, length=5, radius=0.5, center=<2.5 0 0>, rotation=<rot 90 0 0>}]
 foot_l [{shapeType=Capsule, length=13.6641, radius=2.5016, center=<-3.3695 2.3772 0.0342>, rotation=<rot 0 41.0567 -86.531>}]
 foot_r [{shapeType=Capsule, length=12.9149, radius=3.2513, center=<3.37 -2.3774 -0.0352>, rotation=<rot 0 41.0622 -86.5306>}]
}
constraints [
 {bone1Name=dyn_vfx_foot_r_c_01, bone2Name=foot_r, swing1LimitDegrees=45, swing2LimitDegrees=45, twistLimitDegrees=10}
 {bone1Name=dyn_vfx_foot_r_c_02, bone2Name=dyn_vfx_foot_r_c_01, swing1LimitDegrees=45, swing2LimitDegrees=45, twistLimitDegrees=10}
 {bone1Name=dyn_vfx_foot_r_c_03, bone2Name=dyn_vfx_foot_r_c_02, swing1LimitDegrees=45, swing2LimitDegrees=45, twistLimitDegrees=10}
 {bone1Name=dyn_vfx_foot_r_d_01, bone2Name=foot_r, swing1LimitDegrees=45, swing2LimitDegrees=45, twistLimitDegrees=10}
 {bone1Name=dyn_vfx_foot_r_d_02, bone2Name=dyn_vfx_foot_r_d_01, swing1LimitDegrees=45, swing2LimitDegrees=45, twistLimitDegrees=10}
 {bone1Name=dyn_vfx_foot_r_d_03, bone2Name=dyn_vfx_foot_r_d_02, swing1LimitDegrees=45, swing2LimitDegrees=45, twistLimitDegrees=10}
 {bone1Name=dyn_vfx_foot_l_a_01, bone2Name=foot_l, swing1LimitDegrees=45, swing2LimitDegrees=45, twistLimitDegrees=10}
 {bone1Name=dyn_vfx_foot_l_a_02, bone2Name=dyn_vfx_foot_l_a_01, swing1LimitDegrees=45, swing2LimitDegrees=45, twistLimitDegrees=10}
 {bone1Name=dyn_vfx_foot_l_a_03, bone2Name=dyn_vfx_foot_l_a_02, swing1LimitDegrees=45, swing2LimitDegrees=45, twistLimitDegrees=10}
 {bone1Name=dyn_vfx_foot_l_c_01, bone2Name=foot_l, swing1LimitDegrees=45, swing2LimitDegrees=45, twistLimitDegrees=10}
 {bone1Name=dyn_vfx_foot_l_c_02, bone2Name=dyn_vfx_foot_l_c_01, swing1LimitDegrees=45, swing2LimitDegrees=45, twistLimitDegrees=10}
 {bone1Name=dyn_vfx_foot_l_c_03, bone2Name=dyn_vfx_foot_l_c_02, swing1LimitDegrees=45, swing2LimitDegrees=45, twistLimitDegrees=10}
]
