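# Physics asset for the dynamic hand VFX bones on the female medium Palespooky Outlander character body.
# Each hand has three chains of three identical thin capsules (a, c, d on the left; a, b, c on the right), each rooted to a small zero-length capsule on the hand bone.
# Constraints link each segment to its parent with wide swing limits and tight twist limits, so the chains can bend with little roll.
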PhysicsAsset /BRCosmetics/Characters/Player/Female/Medium/Bodies/F_MED_Palespooky_Outlander/Meshes/F_MED_Palespooky_Outlander_Dyn_VFX_Hand
bodies {
 dyn_vfx_hand_l_a_01 [{shapeType=Capsule, length=5, radius=0.5, center=<3 0 0>, rotation=<rot 90 0 0>}]
 dyn_vfx_hand_l_a_02 [{shapeType=Capsule, length=5, radius=0.5, center=<3 0 0>, rotation=<rot 90 0 0>}]
 dyn_vfx_hand_l_a_03 [{shapeType=Capsule, length=5, radius=0.5, center=<3 0 0>, rotation=<rot 90 0 0>}]
 dyn_vfx_hand_l_c_01 [{shapeType=Capsule, length=5, radius=0.5, center=<3 0 0>, rotation=<rot 90 0 0>}]
 dyn_vfx_hand_l_c_02 [{shapeType=Capsule, length=5, radius=0.5, center=<3 0 0>, rotation=<rot 90 0 0>}]
 dyn_vfx_hand_l_c_03 [{shapeType=Capsule, length=5, radius=0.5, center=<3 0 0>, rotation=<rot 90 0 0>}]
 dyn_vfx_hand_l_d_01 [{shapeType=Capsule, length=5, radius=0.5, center=<3 0 0>, rotation=<rot 90 0 0>}]
 dyn_vfx_hand_l_d_02 [{shapeType=Capsule, length=5, radius=0.5, center=<3 0 0>, rotation=<rot 90 0 0>}]
 dyn_vfx_hand_l_d_03 [{shapeType=Capsule, length=5, radius=0.5, center=<3 0 0>, rotation=<rot 90 0 0>}]
 dyn_vfx_hand_r_a_01 [{shapeType=Capsule, length=5, radius=0.5, center=<3 0 0>, rotation=<rot 90 0 0>}]
 dyn_vfx_hand_r_a_02 [{shapeType=Capsule, length=5, radius=0.5, center=<3 0 0>, rotation=<rot 90 0 0>}]
 dyn_vfx_hand_r_a_03 [{shapeType=Capsule, length=5, radius=0.5, center=<3 0 0>, rotation=<rot 90 0 0>}]
 dyn_vfx_hand_r_b_01 [{shapeType=Capsule, length=5, radius=0.5, center=<3 0 0>, rotation=<rot 90 0 0>}]
 dyn_vfx_hand_r_b_02 [{shapeType=Capsule, length=5, radius=0.5, center=<3 0 0>, rotation=<rot 90 0 0>}]
 dyn_vfx_hand_r_b_03 [{shapeType=Capsule, length=5, radius=0.5, center=<3 0 0>, rotation=<rot 90 0 0>}]
 dyn_vfx_hand_r_c_01 [{shapeType=Capsule, length=5, radius=0.5, center=<3 0 0>, rotation=<rot 90 0 0>}]
 dyn_vfx_hand_r_c_02 [{shapeType=Capsule, length=5, radius=0.5, center=<3 0 0>, rotation=<rot 90 0 0>}]
 dyn_vfx_hand_r_c_03 [{shapeType=Capsule, length=5, radius=0.5, center=<3 0 0>, rotation=<rot 90 0 0>}]
 hand_l [{shapeType=Capsule, length=0, radius=1, center=<1 0 0>}]
 hand_r [{shapeType=Capsule, length=0, radius=1, center=<1 0 0>}]
}
constraints [
 {bone1Name=dyn_vfx_hand_l_c_01, bone2Name=hand_l, swing1LimitDegrees=90, swing2LimitDegrees=90, twistLimitDegrees=5}
 {bone1Name=dyn_vfx_hand_l_c_02, bone2Name=dyn_vfx_hand_l_c_01, swing1LimitDegrees=90, swing2LimitDegrees=90, twistLimitDegrees=5}
 {bone1Name=dyn_vfx_hand_l_c_03, bone2Name=dyn_vfx_hand_l_c_02, swing1LimitDegrees=90, swing2LimitDegrees=90, twistLimitDegrees=5}
 {bone1Name=dyn_vfx_hand_l_d_01, bone2Name=hand_l, swing1LimitDegrees=90, swing2LimitDegrees=90, twistLimitDegrees=5}
 {bone1Name=dyn_vfx_hand_l_d_02, bone2Name=dyn_vfx_hand_l_d_01, swing1LimitDegrees=90, swing2LimitDegrees=90, twistLimitDegrees=5}
 {bone1Name=dyn_vfx_hand_l_d_03, bone2Name=dyn_vfx_hand_l_d_02, swing1LimitDegrees=90, swing2LimitDegrees=90, twistLimitDegrees=5}
 {bone1Name=dyn_vfx_hand_l_a_01, bone2Name=hand_l, swing1LimitDegrees=90, swing2LimitDegrees=90, twistLimitDegrees=5}
 {bone1Name=dyn_vfx_hand_l_a_02, bone2Name=dyn_vfx_hand_l_a_01, swing1LimitDegrees=90, swing2LimitDegrees=90, twistLimitDegrees=5}
 {bone1Name=dyn_vfx_hand_l_a_03, bone2Name=dyn_vfx_hand_l_a_02, swing1LimitDegrees=90, swing2LimitDegrees=90, twistLimitDegrees=5}
 {bone1Name=dyn_vfx_hand_r_a_02, bone2Name=dyn_vfx_hand_r_a_01, swing1LimitDegrees=90, swing2LimitDegrees=90, twistLimitDegrees=5}
 {bone1Name=dyn_vfx_hand_r_a_03, bone2Name=dyn_vfx_hand_r_a_02, swing1LimitDegrees=90, swing2LimitDegrees=90, twistLimitDegrees=5}
 {bone1Name=dyn_vfx_hand_r_b_02, bone2Name=dyn_vfx_hand_r_b_01, swing1LimitDegrees=90, swing2LimitDegrees=90, twistLimitDegrees=5}
 {bone1Name=dyn_vfx_hand_r_b_03, bone2Name=dyn_vfx_hand_r_b_02, swing1LimitDegrees=90, swing2LimitDegrees=90, twistLimitDegrees=5}
 {bone1Name=dyn_vfx_hand_r_c_02, bone2Name=dyn_vfx_hand_r_c_01, swing1LimitDegrees=90, swing2LimitDegrees=90, twistLimitDegrees=5}
 {bone1Name=dyn_vfx_hand_r_c_03, bone2Name=dyn_vfx_hand_r_c_02, swing1LimitDegrees=90, swing2LimitDegrees=90, twistLimitDegrees=5}
 {bone1Name=dyn_vfx_hand_r_a_01, bone2Name=hand_r, swing1LimitDegrees=90, swing2LimitDegrees=90, twistLimitDegrees=5}
 {bone1Name=dyn_vfx_hand_r_b_01, bone2Name=hand_r, swing1LimitDegrees=90, swing2LimitDegrees=90, twistLimitDegrees=5}
 {bone1Name=dyn_vfx_hand_r_c_01, bone2Name=hand_r, swing1LimitDegrees=90, swing2LimitDegrees=90, twistLimitDegrees=5}
]
