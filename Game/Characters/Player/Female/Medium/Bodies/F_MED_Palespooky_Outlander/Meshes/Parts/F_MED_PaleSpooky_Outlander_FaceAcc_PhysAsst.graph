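# Physics asset for the face accessory of the female medium Pale Spooky Outlander player character, defining dreadlock physics.
# Box bodies on the Dreads bones form jointed chains rooted on the head, some branching from shared segments.
# Roots have limited swing; many strands free swing on lower joints, while others stay limited with almost no twist.
# A head sphere and neck and clavicle capsules supply body shapes; only the head-to-neck joint is fully free in rotation.
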PhysicsAsset /Game/Characters/Player/Female/Medium/Bodies/F_MED_Palespooky_Outlander/Meshes/Parts/F_MED_PaleSpooky_Outlander_FaceAcc_PhysAsst
bodies {
 DYN_Dreads_0_1 [{shapeType=Box, extentX=6.4483, extentY=3.0851, extentZ=1.103, center=<3.3978 0 0>}]
 DYN_Dreads_0_2 [{shapeType=Box, extentX=6.9732, extentY=3.3825, extentZ=1.6445, center=<3.5475 0 0>}]
 DYN_Dreads_0_3 [{shapeType=Box, extentX=10.0278, extentY=3.0407, extentZ=2.2644, center=<5.0455 -0.0001 0>}]
 DYN_Dreads_0_4 [{shapeType=Box, extentX=8.7609, extentY=3.1741, extentZ=2.282, center=<4.7116 0 0>}]
 DYN_Dreads_10_1 [{shapeType=Box, extentX=6.2288, extentY=3.8567, extentZ=4.103, center=<3.1585 0 0.0003>, rotation=<rot -0.0001 0 15.2195>}]
 DYN_Dreads_10_2 [{shapeType=Box, extentX=7.9615, extentY=4.8233, extentZ=5.1445, center=<4.0884 0 0>}]
 DYN_Dreads_10_3 [{shapeType=Box, extentX=9.0202, extentY=4.5198, extentZ=4.8532, center=<4.5573 -0.0001 -0.0003>}]
 DYN_Dreads_10_4 [{shapeType=Box, extentX=5.7873, extentY=4.5624, extentZ=4.6752, center=<2.9778 0 0>}]
 DYN_Dreads_11_1 [{shapeType=Box, extentX=6.4519, extentY=3.8567, extentZ=1.103, center=<3.3218 0 0.0003>}]
 DYN_Dreads_11_2 [{shapeType=Box, extentX=5.416, extentY=3.3825, extentZ=2.1445, center=<2.7888 0 0>}]
 DYN_Dreads_11_3 [{shapeType=Box, extentX=7.6261, extentY=3.3825, extentZ=2.1445, center=<3.8558 -0.0001 0.0001>}]
 DYN_Dreads_11_4 [{shapeType=Box, extentX=8.6162, extentY=3.0407, extentZ=2.2644, center=<4.3694 -0.0001 0>}]
 DYN_Dreads_11_5 [{shapeType=Box, extentX=4.5392, extentY=2.8697, extentZ=2.282, center=<2.3266 0 0>}]
 DYN_Dreads_1_1 [{shapeType=Box, extentX=4.7185, extentY=6.4027, extentZ=1.103, center=<2.4194 0 0>}]
 DYN_Dreads_1_2 [{shapeType=Box, extentX=8.6228, extentY=6.2589, extentZ=1.6445, center=<4.3639 0 0>}]
 DYN_Dreads_1_3 [{shapeType=Box, extentX=8.6546, extentY=6.1909, extentZ=2.2644, center=<4.3883 -0.0001 0>}]
 DYN_Dreads_1_4 [{shapeType=Box, extentX=3.7274, extentY=5.3737, extentZ=2.282, center=<1.9584 0 -0.0001>}]
 DYN_Dreads_2_1 [{shapeType=Box, extentX=7.862, extentY=6.5704, extentZ=1.103, center=<4.0072 0 -0.0001>}]
 DYN_Dreads_2_2 [{shapeType=Box, extentX=8.0026, extentY=7.8386, extentZ=1.6445, center=<4.1712 0 0>}]
 DYN_Dreads_2_4 [{shapeType=Box, extentX=1.2379, extentY=2.8697, extentZ=2.282, center=<0.6485 0 0.0001>}]
 DYN_Dreads_2_5 [{shapeType=Box, extentX=7.412, extentY=3.0407, extentZ=2.2644, center=<3.8141 -0.0001 -0.0002>}]
 DYN_Dreads_2_6 [{shapeType=Box, extentX=1.2749, extentY=2.7305, extentZ=2.282, center=<0.7289 0 0.0001>}]
 DYN_Dreads_2_7 [{shapeType=Box, extentX=7.9657, extentY=3.0407, extentZ=2.2644, center=<4.0653 -0.0001 -0.0002>}]
 DYN_Dreads_2_8 [{shapeType=Box, extentX=1.3967, extentY=2.8697, extentZ=2.282, center=<0.7056 0 0.0001>}]
 DYN_Dreads_3_1 [{shapeType=Box, extentX=6.8022, extentY=3.0851, extentZ=1.103, center=<3.5436 0 -0.0002>}]
 DYN_Dreads_3_10 [{shapeType=Box, extentX=2.459, extentY=2.8697, extentZ=2.282, center=<1.2872 0 0.0001>}]
 DYN_Dreads_3_11 [{shapeType=Box, extentX=5.5859, extentY=3.0851, extentZ=1.103, center=<3.0078 0 -0.0001>}]
 DYN_Dreads_3_12 [{shapeType=Box, extentX=8.608, extentY=3.3825, extentZ=1.6445, center=<4.37 0 0>}]
 DYN_Dreads_3_13 [{shapeType=Box, extentX=4.7682, extentY=3.0851, extentZ=1.103, center=<2.5288 0 -0.0002>}]
 DYN_Dreads_3_2 [{shapeType=Box, extentX=8.608, extentY=3.3825, extentZ=1.6445, center=<4.37 0 0>}]
 DYN_Dreads_3_3 [{shapeType=Box, extentX=6.654, extentY=3.0407, extentZ=2.2644, center=<3.3965 -0.0001 -0.0002>}]
 DYN_Dreads_3_4 [{shapeType=Box, extentX=2.459, extentY=2.8697, extentZ=2.282, center=<1.2872 0 0.0001>}]
 DYN_Dreads_3_5 [{shapeType=Box, extentX=8.3601, extentY=3.0407, extentZ=2.2644, center=<4.2131 -0.0001 -0.0001>}]
 DYN_Dreads_3_6 [{shapeType=Box, extentX=1.0917, extentY=2.9119, extentZ=2.1209, center=<0.5793 0 0>}]
 DYN_Dreads_3_7 [{shapeType=Box, extentX=7.754, extentY=3.3825, extentZ=1.6445, center=<3.9583 0 -0.0001>}]
 DYN_Dreads_3_8 [{shapeType=Box, extentX=7.2225, extentY=3.0407, extentZ=2.2644, center=<3.694 -0.0001 -0.0002>}]
 DYN_Dreads_3_9 [{shapeType=Box, extentX=1.8023, extentY=2.8697, extentZ=2.282, center=<0.9285 0 0>}]
 DYN_Dreads_4_1 [{shapeType=Box, extentX=6.0358, extentY=3.0851, extentZ=0.7362, center=<3.1678 0 0>}]
 DYN_Dreads_4_10 [{shapeType=Box, extentX=6.2057, extentY=3.8567, extentZ=0.8811, center=<3.2173 0 0.0003>}]
 DYN_Dreads_4_11 [{shapeType=Box, extentX=4.696, extentY=3.0851, extentZ=1.103, center=<2.4253 -0.0001 0.0001>}]
 DYN_Dreads_4_2 [{shapeType=Box, extentX=7.5711, extentY=3.3825, extentZ=0.8714, center=<3.8411 0 -0.0002>}]
 DYN_Dreads_4_3 [{shapeType=Box, extentX=6.4366, extentY=3.0407, extentZ=2.2644, center=<3.2362 -0.0001 -0.0002>}]
 DYN_Dreads_4_4 [{shapeType=Box, extentX=2.3507, extentY=2.8697, extentZ=2.282, center=<1.2414 0 0.0001>}]
 DYN_Dreads_4_5 [{shapeType=Box, extentX=6.9149, extentY=3.3825, extentZ=1.6445, center=<3.5008 0 -0.0001>}]
 DYN_Dreads_4_6 [{shapeType=Box, extentX=6.4366, extentY=3.0407, extentZ=2.2644, center=<3.2362 -0.0001 -0.0002>}]
 DYN_Dreads_4_7 [{shapeType=Box, extentX=2.1693, extentY=2.8697, extentZ=2.282, center=<1.1511 0 0>}]
 DYN_Dreads_4_8 [{shapeType=Box, extentX=7.754, extentY=3.3825, extentZ=0.9067, center=<3.9583 0 -0.0001>}]
 DYN_Dreads_4_9 [{shapeType=Box, extentX=7.2225, extentY=3.0407, extentZ=2.2644, center=<3.6774 -0.0001 -0.0002>}]
 DYN_Dreads_5_1 [{shapeType=Box, extentX=6.0358, extentY=3.0851, extentZ=2.2036, center=<3.1678 0 0>}]
 DYN_Dreads_5_10 [{shapeType=Box, extentX=4.0652, extentY=2.9119, extentZ=2.1209, center=<2.1435 0 -0.0001>}]
 DYN_Dreads_5_11 [{shapeType=Box, extentX=4.1911, extentY=3.0851, extentZ=1.5482, center=<2.1255 0 -0.0002>}]
 DYN_Dreads_5_12 [{shapeType=Box, extentX=5.6633, extentY=3.0851, extentZ=1.103, center=<2.9529 0 0.0002>}]
 DYN_Dreads_5_2 [{shapeType=Box, extentX=7.0966, extentY=3.3825, extentZ=1.6445, center=<3.6223 0 -0.0001>}]
 DYN_Dreads_5_3 [{shapeType=Box, extentX=7.4053, extentY=3.0407, extentZ=2.2644, center=<3.7697 -0.0001 0>}]
 DYN_Dreads_5_4 [{shapeType=Box, extentX=3.1726, extentY=2.9119, extentZ=2.1209, center=<1.6398 0 -0.0001>}]
 DYN_Dreads_5_5 [{shapeType=Box, extentX=6.8593, extentY=3.3825, extentZ=1.6445, center=<3.5014 0 0>}]
 DYN_Dreads_5_6 [{shapeType=Box, extentX=7.4053, extentY=3.0407, extentZ=2.2644, center=<3.7697 -0.0001 0>}]
 DYN_Dreads_5_7 [{shapeType=Box, extentX=3.5673, extentY=2.9119, extentZ=2.1209, center=<1.8593 0 0>}]
 DYN_Dreads_5_8 [{shapeType=Box, extentX=7.0966, extentY=3.3825, extentZ=1.6445, center=<3.6223 0 -0.0001>}]
 DYN_Dreads_5_9 [{shapeType=Box, extentX=7.4053, extentY=3.0407, extentZ=2.2644, center=<3.7697 -0.0001 0>}]
 DYN_Dreads_6_2 [{shapeType=Box, extentX=7.4008, extentY=3.3825, extentZ=1.6445, center=<3.8068 0 0.0001>}]
 DYN_Dreads_6_3 [{shapeType=Box, extentX=7.1235, extentY=3.0407, extentZ=2.2644, center=<3.6891 -0.0001 0>}]
 DYN_Dreads_6_4 [{shapeType=Box, extentX=3.4688, extentY=2.8697, extentZ=2.282, center=<1.7894 0 0.0001>}]
 DYN_Dreads_6_6 [{shapeType=Box, extentX=8.2252, extentY=3.1839, extentZ=1.103, center=<4.26 0 0>}]
 DYN_Dreads_7_1 [{shapeType=Box, extentX=5.541, extentY=3.0851, extentZ=1.434, center=<2.9197 0 0>}]
 DYN_Dreads_7_2 [{shapeType=Box, extentX=6.9547, extentY=3.3825, extentZ=1.6445, center=<3.5248 0 -0.0001>}]
 DYN_Dreads_7_3 [{shapeType=Box, extentX=9.3513, extentY=3.0407, extentZ=2.2644, center=<4.7063 -0.0001 0.0001>}]
 DYN_Dreads_7_4 [{shapeType=Box, extentX=1.7654, extentY=2.8697, extentZ=2.282, center=<0.9758 0 0.0001>}]
 DYN_Dreads_8_1 [{shapeType=Box, extentX=7.521, extentY=7.8558, extentZ=1.103, center=<3.9449 0 -0.0001>}]
 DYN_Dreads_8_10 [{shapeType=Box, extentX=2.8785, extentY=2.8697, extentZ=2.282, center=<1.4965 0 0.0001>}]
 DYN_Dreads_8_2 [{shapeType=Box, extentX=7.6226, extentY=3.3825, extentZ=1.6445, center=<3.8626 0 0.0002>}]
 DYN_Dreads_8_3 [{shapeType=Box, extentX=6.814, extentY=3.4815, extentZ=2.2644, center=<3.5865 -0.0001 -0.0001>}]
 DYN_Dreads_8_4 [{shapeType=Box, extentX=3.8201, extentY=2.9682, extentZ=2.282, center=<1.9629 0 0.0001>}]
 DYN_Dreads_8_5 [{shapeType=Box, extentX=7.6226, extentY=3.3825, extentZ=1.6445, center=<3.8626 0 0.0002>}]
 DYN_Dreads_8_6 [{shapeType=Box, extentX=6.1412, extentY=3.4815, extentZ=2.2644, center=<3.126 -0.0001 0>}]
 DYN_Dreads_8_7 [{shapeType=Box, extentX=3.0577, extentY=2.8697, extentZ=2.282, center=<1.6201 0 0.0001>}]
 DYN_Dreads_8_8 [{shapeType=Box, extentX=7.6226, extentY=3.3825, extentZ=1.6445, center=<3.8626 0 0.0002>}]
 DYN_Dreads_8_9 [{shapeType=Box, extentX=7.8882, extentY=3.4815, extentZ=2.2644, center=<4.0176 -0.0001 0>}]
 DYN_Dreads_9_1 [{shapeType=Box, extentX=6.02, extentY=3.0851, extentZ=1.8497, center=<3.0566 0 0.0002>, rotation=<rot 0 0 2.0866>}]
 DYN_Dreads_9_2 [{shapeType=Box, extentX=7.332, extentY=3.3825, extentZ=1.6445, center=<3.7264 0 0.0001>}]
 DYN_Dreads_9_3 [{shapeType=Box, extentX=8.1035, extentY=3.0407, extentZ=2.2644, center=<4.1116 -0.0001 0>}]
 DYN_Dreads_9_4 [{shapeType=Box, extentX=4.5643, extentY=3.1699, extentZ=2.282, center=<2.4066 0 0.0001>}]
 Dreads_2_3 [{shapeType=Box, extentX=7.6067, extentY=3.0407, extentZ=2.2644, center=<3.851 -0.0001 -0.0002>}]
 clavicle_l [{shapeType=Capsule, length=5.75, radius=5.95, center=<5.6876 -1.0188 -1.4357>, rotation=<rot 0 -89.9995 94.9808>}]
 clavicle_r [
  {shapeType=Capsule, length=3.25, radius=6.2, center=<-7.4505 1.037 2.4801>, rotation=<rot 0 -89.9995 94.0647>}
  {shapeType=Capsule, length=2.5, radius=11.2, center=<-1.9635 -4.2852 6.597>, rotation=<rot 4.0491 -174.9871 90.3548>}
 ]
 head [{shapeType=Sphere, radius=11.8447, center=<5.2932 0.6341 0.1927>}]
 neck_01 [{shapeType=Capsule, length=9.57, radius=3, center=<-2.1534 -1.8622 0.8528>, rotation=<rot 1.1344 -20.503 2.459>}]
}
constraints [
 {bone1Name=DYN_Dreads_0_1, bone2Name=head, swing1LimitDegrees=35, swing2LimitDegrees=35, twistLimitDegrees=0.5}
 {bone1Name=DYN_Dreads_1_1, bone2Name=head, swing1LimitDegrees=35, swing2LimitDegrees=35, twistLimitDegrees=0.5}
 {bone1Name=DYN_Dreads_2_1, bone2Name=head, swing1LimitDegrees=35, swing2LimitDegrees=35, twistLimitDegrees=0.5}
 {bone1Name=DYN_Dreads_3_13, bone2Name=head, swing1LimitDegrees=25, swing2LimitDegrees=25, twistLimitDegrees=25}
 {bone1Name=DYN_Dreads_3_1, bone2Name=head, swing1LimitDegrees=30, swing2LimitDegrees=35, twistLimitDegrees=20}
 {bone1Name=DYN_Dreads_3_11, bone2Name=head, swing1LimitDegrees=7, swing2LimitDegrees=5, twistLimitDegrees=20}
 {bone1Name=DYN_Dreads_4_11, bone2Name=head, swing1LimitDegrees=15, swing2LimitDegrees=15, twistLimitDegrees=20}
 {bone1Name=DYN_Dreads_4_1, bone2Name=head, swing1LimitDegrees=10, swing2LimitDegrees=10, twistLimitDegrees=20}
 {bone1Name=DYN_Dreads_5_1, bone2Name=head, swing1LimitDegrees=8, swing2LimitDegrees=7, twistLimitDegrees=20}
 {bone1Name=DYN_Dreads_6_6, bone2Name=head, swing1LimitDegrees=20, swing2LimitDegrees=20, twistLimitDegrees=20}
 {bone1Name=DYN_Dreads_7_1, bone2Name=head, swing1LimitDegrees=20, swing2LimitDegrees=20, twistLimitDegrees=20}
 {bone1Name=DYN_Dreads_8_1, bone2Name=head, swing1LimitDegrees=20, swing2LimitDegrees=20, twistLimitDegrees=20}
 {bone1Name=DYN_Dreads_9_1, bone2Name=head, swing1LimitDegrees=35, swing2LimitDegrees=35, twistLimitDegrees=0.5}
 {bone1Name=DYN_Dreads_10_1, bone2Name=head, swing1LimitDegrees=35, swing2LimitDegrees=35, twistLimitDegrees=0.5}
 {bone1Name=DYN_Dreads_11_1, bone2Name=head, swing1LimitDegrees=35, swing2LimitDegrees=35, twistLimitDegrees=0.5}
 {bone1Name=DYN_Dreads_4_10, bone2Name=head, swing1LimitDegrees=8, swing2LimitDegrees=7, twistLimitDegrees=20}
 {bone1Name=DYN_Dreads_5_11, bone2Name=head, swing1LimitDegrees=10, swing2LimitDegrees=7, twistLimitDegrees=20}
 {bone1Name=DYN_Dreads_5_12, bone2Name=head, swing1LimitDegrees=10, swing2LimitDegrees=5, twistLimitDegrees=20}
 {bone1Name=DYN_Dreads_0_2, bone2Name=DYN_Dreads_0_1, swing1LimitDegrees=35, swing2LimitDegrees=35, twistLimitDegrees=0.5}
 {bone1Name=DYN_Dreads_0_3, bone2Name=DYN_Dreads_0_2, swing1LimitDegrees=35, swing2LimitDegrees=35, twistLimitDegrees=0.5}
 {bone1Name=DYN_Dreads_0_4, bone2Name=DYN_Dreads_0_3, swing1LimitDegrees=35, swing2LimitDegrees=35, twistLimitDegrees=0.5}
 {bone1Name=DYN_Dreads_1_2, bone2Name=DYN_Dreads_1_1, swing1LimitDegrees=35, swing2LimitDegrees=35, twistLimitDegrees=0.5}
 {bone1Name=DYN_Dreads_1_3, bone2Name=DYN_Dreads_1_2, swing1LimitDegrees=35, swing2LimitDegrees=35, twistLimitDegrees=0.5}
 {bone1Name=DYN_Dreads_1_4, bone2Name=DYN_Dreads_1_3, swing1LimitDegrees=35, swing2LimitDegrees=35, twistLimitDegrees=0.5}
 {
  bone1Name=DYN_Dreads_5_10
  bone2Name=DYN_Dreads_5_9
  swing1LimitDegrees=45
  swing1Motion=Free
  swing2LimitDegrees=45
  swing2Motion=Free
  twistLimitDegrees=20
 }
 {
  bone1Name=DYN_Dreads_5_9
  bone2Name=DYN_Dreads_5_8
  swing1LimitDegrees=45
  swing1Motion=Free
  swing2LimitDegrees=45
  swing2Motion=Free
  twistLimitDegrees=20
 }
 {bone1Name=DYN_Dreads_5_8, bone2Name=DYN_Dreads_5_12, swing1LimitDegrees=10, swing2LimitDegrees=10, twistLimitDegrees=20}
 {
  bone1Name=DYN_Dreads_5_7
  bone2Name=DYN_Dreads_5_6
  swing1LimitDegrees=45
  swing1Motion=Free
  swing2LimitDegrees=45
  swing2Motion=Free
  twistLimitDegrees=20
 }
 {
  bone1Name=DYN_Dreads_5_6
  bone2Name=DYN_Dreads_5_5
  swing1LimitDegrees=45
  swing1Motion=Free
  swing2LimitDegrees=45
  swing2Motion=Free
  twistLimitDegrees=20
 }
 {bone1Name=DYN_Dreads_5_5, bone2Name=DYN_Dreads_5_11, swing1LimitDegrees=20, swing2LimitDegrees=20, twistLimitDegrees=20}
 {
  bone1Name=DYN_Dreads_3_6
  bone2Name=DYN_Dreads_3_5
  swing1LimitDegrees=45
  swing1Motion=Free
  swing2LimitDegrees=45
  swing2Motion=Free
  twistLimitDegrees=20
 }
 {
  bone1Name=DYN_Dreads_3_5
  bone2Name=DYN_Dreads_3_12
  swing1LimitDegrees=45
  swing1Motion=Free
  swing2LimitDegrees=45
  swing2Motion=Free
  twistLimitDegrees=20
 }
 {bone1Name=DYN_Dreads_3_12, bone2Name=DYN_Dreads_3_13, swing1LimitDegrees=10, swing2LimitDegrees=10, twistLimitDegrees=20}
 {
  bone1Name=DYN_Dreads_3_9
  bone2Name=DYN_Dreads_3_8
  swing1LimitDegrees=45
  swing1Motion=Free
  swing2LimitDegrees=45
  swing2Motion=Free
  twistLimitDegrees=20
 }
 {
  bone1Name=DYN_Dreads_3_8
  bone2Name=DYN_Dreads_3_7
  swing1LimitDegrees=45
  swing1Motion=Free
  swing2LimitDegrees=45
  swing2Motion=Free
  twistLimitDegrees=20
 }
 {bone1Name=DYN_Dreads_3_7, bone2Name=DYN_Dreads_3_11, swing1LimitDegrees=10, swing2LimitDegrees=10, twistLimitDegrees=20}
 {
  bone1Name=DYN_Dreads_3_10
  bone2Name=DYN_Dreads_4_9
  swing1LimitDegrees=45
  swing1Motion=Free
  swing2LimitDegrees=45
  swing2Motion=Free
  twistLimitDegrees=20
 }
 {
  bone1Name=DYN_Dreads_4_9
  bone2Name=DYN_Dreads_4_8
  swing1LimitDegrees=45
  swing1Motion=Free
  swing2LimitDegrees=45
  swing2Motion=Free
  twistLimitDegrees=20
 }
 {bone1Name=DYN_Dreads_4_8, bone2Name=DYN_Dreads_4_11, swing1LimitDegrees=10, swing2LimitDegrees=10, twistLimitDegrees=20}
 {
  bone1Name=DYN_Dreads_4_7
  bone2Name=DYN_Dreads_4_6
  swing1LimitDegrees=45
  swing1Motion=Free
  swing2LimitDegrees=45
  swing2Motion=Free
  twistLimitDegrees=20
 }
 {
  bone1Name=DYN_Dreads_4_6
  bone2Name=DYN_Dreads_4_5
  swing1LimitDegrees=45
  swing1Motion=Free
  swing2LimitDegrees=45
  swing2Motion=Free
  twistLimitDegrees=20
 }
 {bone1Name=DYN_Dreads_4_5, bone2Name=DYN_Dreads_4_10, swing1LimitDegrees=10, swing2LimitDegrees=10, twistLimitDegrees=20}
 {bone1Name=DYN_Dreads_11_5, bone2Name=DYN_Dreads_11_4, swing1LimitDegrees=35, swing2LimitDegrees=35, twistLimitDegrees=0.5}
 {bone1Name=DYN_Dreads_11_4, bone2Name=DYN_Dreads_11_3, swing1LimitDegrees=35, swing2LimitDegrees=35, twistLimitDegrees=0.5}
 {bone1Name=DYN_Dreads_11_3, bone2Name=DYN_Dreads_11_2, swing1LimitDegrees=35, swing2LimitDegrees=35, twistLimitDegrees=0.5}
 {bone1Name=DYN_Dreads_11_2, bone2Name=DYN_Dreads_11_1, swing1LimitDegrees=35, swing2LimitDegrees=35, twistLimitDegrees=0.5}
 {bone1Name=DYN_Dreads_10_4, bone2Name=DYN_Dreads_10_3, swing1LimitDegrees=35, swing2LimitDegrees=35, twistLimitDegrees=0.5}
 {bone1Name=DYN_Dreads_10_3, bone2Name=DYN_Dreads_10_2, swing1LimitDegrees=35, swing2LimitDegrees=35, twistLimitDegrees=0.5}
 {bone1Name=DYN_Dreads_10_2, bone2Name=DYN_Dreads_10_1, swing1LimitDegrees=35, swing2LimitDegrees=35, twistLimitDegrees=0.5}
 {bone1Name=DYN_Dreads_9_4, bone2Name=DYN_Dreads_9_3, swing1LimitDegrees=35, swing2LimitDegrees=35, twistLimitDegrees=0.5}
 {bone1Name=DYN_Dreads_9_3, bone2Name=DYN_Dreads_9_2, swing1LimitDegrees=35, swing2LimitDegrees=35, twistLimitDegrees=0.5}
 {bone1Name=DYN_Dreads_9_2, bone2Name=DYN_Dreads_9_1, swing1LimitDegrees=35, swing2LimitDegrees=35, twistLimitDegrees=0.5}
 {
  bone1Name=DYN_Dreads_8_10
  bone2Name=DYN_Dreads_8_9
  swing1LimitDegrees=45
  swing1Motion=Free
  swing2LimitDegrees=45
  swing2Motion=Free
  twistLimitDegrees=20
 }
 {
  bone1Name=DYN_Dreads_8_9
  bone2Name=DYN_Dreads_8_8
  swing1LimitDegrees=45
  swing1Motion=Free
  swing2LimitDegrees=45
  swing2Motion=Free
  twistLimitDegrees=20
 }
 {bone1Name=DYN_Dreads_8_8, bone2Name=DYN_Dreads_8_1, swing1LimitDegrees=10, swing2LimitDegrees=10, twistLimitDegrees=20}
 {
  bone1Name=DYN_Dreads_8_7
  bone2Name=DYN_Dreads_8_6
  swing1LimitDegrees=45
  swing1Motion=Free
  swing2LimitDegrees=45
  swing2Motion=Free
  twistLimitDegrees=20
 }
 {
  bone1Name=DYN_Dreads_8_6
  bone2Name=DYN_Dreads_8_5
  swing1LimitDegrees=45
  swing1Motion=Free
  swing2LimitDegrees=45
  swing2Motion=Free
  twistLimitDegrees=20
 }
 {bone1Name=DYN_Dreads_8_5, bone2Name=DYN_Dreads_8_1, swing1LimitDegrees=10, swing2LimitDegrees=10, twistLimitDegrees=20}
 {
  bone1Name=DYN_Dreads_8_4
  bone2Name=DYN_Dreads_8_3
  swing1LimitDegrees=45
  swing1Motion=Free
  swing2LimitDegrees=45
  swing2Motion=Free
  twistLimitDegrees=20
 }
 {
  bone1Name=DYN_Dreads_8_3
  bone2Name=DYN_Dreads_8_2
  swing1LimitDegrees=45
  swing1Motion=Free
  swing2LimitDegrees=45
  swing2Motion=Free
  twistLimitDegrees=20
 }
 {bone1Name=DYN_Dreads_8_2, bone2Name=DYN_Dreads_8_1, swing1LimitDegrees=10, swing2LimitDegrees=10, twistLimitDegrees=20}
 {
  bone1Name=DYN_Dreads_6_4
  bone2Name=DYN_Dreads_6_3
  swing1LimitDegrees=45
  swing1Motion=Free
  swing2LimitDegrees=45
  swing2Motion=Free
  twistLimitDegrees=20
 }
 {
  bone1Name=DYN_Dreads_6_3
  bone2Name=DYN_Dreads_6_2
  swing1LimitDegrees=45
  swing1Motion=Free
  swing2LimitDegrees=45
  swing2Motion=Free
  twistLimitDegrees=20
 }
 {bone1Name=DYN_Dreads_6_2, bone2Name=DYN_Dreads_6_6, swing1LimitDegrees=10, swing2LimitDegrees=10, twistLimitDegrees=20}
 {
  bone1Name=DYN_Dreads_5_4
  bone2Name=DYN_Dreads_5_3
  swing1LimitDegrees=45
  swing1Motion=Free
  swing2LimitDegrees=45
  swing2Motion=Free
  twistLimitDegrees=20
 }
 {
  bone1Name=DYN_Dreads_5_3
  bone2Name=DYN_Dreads_5_2
  swing1LimitDegrees=45
  swing1Motion=Free
  swing2LimitDegrees=45
  swing2Motion=Free
  twistLimitDegrees=20
 }
 {bone1Name=DYN_Dreads_5_2, bone2Name=DYN_Dreads_5_1, swing1LimitDegrees=10, swing2LimitDegrees=10, twistLimitDegrees=20}
 {
  bone1Name=DYN_Dreads_4_4
  bone2Name=DYN_Dreads_4_3
  swing1LimitDegrees=45
  swing1Motion=Free
  swing2LimitDegrees=45
  swing2Motion=Free
  twistLimitDegrees=20
 }
 {
  bone1Name=DYN_Dreads_4_3
  bone2Name=DYN_Dreads_4_2
  swing1LimitDegrees=45
  swing1Motion=Free
  swing2LimitDegrees=45
  swing2Motion=Free
  twistLimitDegrees=20
 }
 {bone1Name=DYN_Dreads_4_2, bone2Name=DYN_Dreads_4_1, swing1LimitDegrees=10, swing2LimitDegrees=10, twistLimitDegrees=20}
 {
  bone1Name=DYN_Dreads_7_4
  bone2Name=DYN_Dreads_7_3
  swing1LimitDegrees=45
  swing1Motion=Free
  swing2LimitDegrees=45
  swing2Motion=Free
  twistLimitDegrees=20
 }
 {
  bone1Name=DYN_Dreads_7_3
  bone2Name=DYN_Dreads_7_2
  swing1LimitDegrees=45
  swing1Motion=Free
  swing2LimitDegrees=45
  swing2Motion=Free
  twistLimitDegrees=20
 }
 {bone1Name=DYN_Dreads_7_2, bone2Name=DYN_Dreads_7_1, swing1LimitDegrees=10, swing2LimitDegrees=10, twistLimitDegrees=20}
 {
  bone1Name=DYN_Dreads_3_4
  bone2Name=DYN_Dreads_3_3
  swing1LimitDegrees=45
  swing1Motion=Free
  swing2LimitDegrees=45
  swing2Motion=Free
  twistLimitDegrees=20
 }
 {
  bone1Name=DYN_Dreads_3_3
  bone2Name=DYN_Dreads_3_2
  swing1LimitDegrees=45
  swing1Motion=Free
  swing2LimitDegrees=45
  swing2Motion=Free
  twistLimitDegrees=20
 }
 {bone1Name=DYN_Dreads_3_2, bone2Name=DYN_Dreads_3_1, swing1LimitDegrees=10, swing2LimitDegrees=10, twistLimitDegrees=20}
 {bone1Name=DYN_Dreads_2_8, bone2Name=DYN_Dreads_2_7, swing1LimitDegrees=35, swing2LimitDegrees=35, twistLimitDegrees=0.5}
 {
  bone1Name=DYN_Dreads_2_7
  bone2Name=DYN_Dreads_2_2
  swing1LimitDegrees=45
  swing1Motion=Free
  swing2LimitDegrees=45
  swing2Motion=Free
  twistLimitDegrees=20
 }
 {bone1Name=DYN_Dreads_2_6, bone2Name=DYN_Dreads_2_5, swing1LimitDegrees=35, swing2LimitDegrees=35, twistLimitDegrees=0.5}
 {bone1Name=DYN_Dreads_2_5, bone2Name=DYN_Dreads_2_2, swing1LimitDegrees=35, swing2LimitDegrees=35, twistLimitDegrees=0.5}
 {bone1Name=DYN_Dreads_2_4, bone2Name=Dreads_2_3, swing1LimitDegrees=35, swing2LimitDegrees=35, twistLimitDegrees=0.5}
 {bone1Name=Dreads_2_3, bone2Name=DYN_Dreads_2_2, swing1LimitDegrees=35, swing2LimitDegrees=35, twistLimitDegrees=0.5}
 {bone1Name=DYN_Dreads_2_2, bone2Name=DYN_Dreads_2_1, swing1LimitDegrees=35, swing2LimitDegrees=35, twistLimitDegrees=0.5}
 {
  bone1Name=head
  bone2Name=neck_01
  swing1LimitDegrees=45
  swing1Motion=Free
  swing2LimitDegrees=45
  swing2Motion=Free
  twistLimitDegrees=45
  twistMotion=Free
 }
]
