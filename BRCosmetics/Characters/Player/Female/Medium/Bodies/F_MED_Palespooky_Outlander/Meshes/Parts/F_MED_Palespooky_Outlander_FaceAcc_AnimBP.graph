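# Animation blueprint for the face accessory of the female medium Palespooky Outlander cosmetic body, adding secondary sway to its dreadlocks.
# It copies the pose and curves from the attached parent mesh, then converts it to component space.
# A long serial chain of Anim Dynamics nodes, one per dread bone, applies angular springs with damping and angular limits.
# Simulation is root-relative with zeroed gravity, scaled-down component motion and varying spring stiffness.
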AnimBlueprint /BRCosmetics/Characters/Player/Female/Medium/Bodies/F_MED_Palespooky_Outlander/Meshes/Parts/F_MED_Palespooky_Outlander_FaceAcc_AnimBP
target_skeleton=&/F_MED_Palespooky_Outlander_FaceAcc_Skeleton
graphs [
 AnimGraph {nodes [
   n1 = Anim Dynamics - Bone: DYN_Dreads_8_1!(
    ComponentPose=Local To Component(LocalPose=Copy Pose From Mesh() {bCopyCurves=true, bUseAttachedParent=true})
    Alpha=1
    GravityOverride:Vector="0,0,0"
    ExternalForce="0,0,0"
   ) -> (Pose) {
    angularDampingOverride=1
    bAngularSpring=true
    bOverrideAngularDamping=true
    bUseGravityOverride=true
    simulationSpace=RootRelative
    alphaBoolBlend {blendOption=Linear}
    boundBone {boneName=DYN_Dreads_8_1}
    componentAppliedLinearAccClamp=<2500 2500 2500>
    componentLinearAccScale=<0.5 0.5 0.5>
    componentLinearVelScale=<0.5 0.5 0.5>
    physicsBodyDefinitions [{
      sphereCollisionRadius=0
      boundBone {boneName=DYN_Dreads_8_1}
      boxExtents=<12 12 8>
      constraintSetup {angularLimitsMax=<45 0 45>, angularLimitsMin=<-45 -45 -45>, angularTarget=<0 0 0>}
      localJointOffset=<1 0 0.5>
     }]
   }
   n2 = Anim Dynamics - Bone: DYN_Dreads_8_8!(ComponentPose=@n1, Alpha=1, GravityOverride:Vector="0,0,0", ExternalForce="0,0,0") -> (Pose) {
    angularDampingOverride=1
    bAngularSpring=true
    bOverrideAngularDamping=true
    bUseGravityOverride=true
    simulationSpace=RootRelative
    alphaBoolBlend {blendOption=Linear}
    boundBone {boneName=DYN_Dreads_8_8}
    componentAppliedLinearAccClamp=<2500 2500 2500>
    componentLinearAccScale=<0.5 0.5 0.5>
    componentLinearVelScale=<0.5 0.5 0.5>
    physicsBodyDefinitions [{
      sphereCollisionRadius=0
      boundBone {boneName=DYN_Dreads_8_8}
      boxExtents=<12 12 8>
      constraintSetup {angularLimitsMax=<45 0 45>, angularLimitsMin=<-45 -45 -45>, angularTarget=<0 0 0>}
      localJointOffset=<1 0 0.5>
     }]
   }
   n3 = Anim Dynamics - Bone: DYN_Dreads_8_9!(ComponentPose=@n2, Alpha=1, GravityOverride:Vector="0,0,0", ExternalForce="0,0,0") -> (Pose) {
    angularDampingOverride=1
    bAngularSpring=true
    bOverrideAngularDamping=true
    bUseGravityOverride=true
    simulationSpace=RootRelative
    alphaBoolBlend {blendOption=Linear}
    boundBone {boneName=DYN_Dreads_8_9}
    componentAppliedLinearAccClamp=<2500 2500 2500>
    componentLinearAccScale=<0.5 0.5 0.5>
    componentLinearVelScale=<0.5 0.5 0.5>
    physicsBodyDefinitions [{
      sphereCollisionRadius=0
      boundBone {boneName=DYN_Dreads_8_9}
      boxExtents=<12 12 8>
      constraintSetup {angularLimitsMax=<45 0 45>, angularLimitsMin=<-45 -45 -45>, angularTarget=<0 0 0>}
      localJointOffset=<1 0 0.5>
     }]
   }
   n4 = Anim Dynamics - Bone: DYN_Dreads_8_10!(ComponentPose=@n3, Alpha=1, GravityOverride:Vector="0,0,0", ExternalForce="0,0,0") -> (Pose) {
    angularDampingOverride=1
    bAngularSpring=true
    bOverrideAngularDamping=true
    bUseGravityOverride=true
    simulationSpace=RootRelative
    alphaBoolBlend {blendOption=Linear}
    boundBone {boneName=DYN_Dreads_8_10}
    componentAppliedLinearAccClamp=<2500 2500 2500>
    componentLinearAccScale=<0.5 0.5 0.5>
    componentLinearVelScale=<0.5 0.5 0.5>
    physicsBodyDefinitions [{
      sphereCollisionRadius=0
      boundBone {boneName=DYN_Dreads_8_10}
      boxExtents=<12 12 8>
      constraintSetup {angularLimitsMax=<45 0 45>, angularLimitsMin=<-45 -45 -45>, angularTarget=<0 0 0>}
      localJointOffset=<1 0 0.5>
     }]
   }
   n5 = Anim Dynamics - Bone: DYN_Dreads_8_2!(ComponentPose=@n4, Alpha=1, GravityOverride:Vector="0,0,0", ExternalForce="0,0,0") -> (Pose) {
    angularDampingOverride=1
    bAngularSpring=true
    bOverrideAngularDamping=true
    bUseGravityOverride=true
    simulationSpace=RootRelative
    alphaBoolBlend {blendOption=Linear}
    boundBone {boneName=DYN_Dreads_8_2}
    componentAppliedLinearAccClamp=<2500 2500 2500>
    componentLinearAccScale=<0.5 0.5 0.5>
    componentLinearVelScale=<0.5 0.5 0.5>
    physicsBodyDefinitions [{
      sphereCollisionRadius=0
      boundBone {boneName=DYN_Dreads_8_2}
      boxExtents=<12 12 8>
      constraintSetup {angularLimitsMax=<45 0 45>, angularLimitsMin=<-45 -45 -45>, angularTarget=<0 0 0>}
      localJointOffset=<2 0 0.5>
     }]
   }
   n6 = Anim Dynamics - Bone: DYN_Dreads_8_3!(ComponentPose=@n5, Alpha=1, GravityOverride:Vector="0,0,0", ExternalForce="0,0,0") -> (Pose) {
    angularDampingOverride=1
    bAngularSpring=true
    bOverrideAngularDamping=true
    bUseGravityOverride=true
    simulationSpace=RootRelative
    alphaBoolBlend {blendOption=Linear}
    boundBone {boneName=DYN_Dreads_8_3}
    componentAppliedLinearAccClamp=<2500 2500 2500>
    componentLinearAccScale=<0.5 0.5 0.5>
    componentLinearVelScale=<0.5 0.5 0.5>
    physicsBodyDefinitions [{
      sphereCollisionRadius=0
      boundBone {boneName=DYN_Dreads_8_3}
      boxExtents=<12 12 8>
      constraintSetup {angularLimitsMax=<45 0 45>, angularLimitsMin=<-45 -45 -45>, angularTarget=<0 0 0>}
      localJointOffset=<2 0 0.5>
     }]
   }
   n7 = Anim Dynamics - Bone: DYN_Dreads_8_4!(ComponentPose=@n6, Alpha=1, GravityOverride:Vector="0,0,0", ExternalForce="0,0,0") -> (Pose) {
    angularDampingOverride=1
    bAngularSpring=true
    bOverrideAngularDamping=true
    bUseGravityOverride=true
    simulationSpace=RootRelative
    alphaBoolBlend {blendOption=Linear}
    boundBone {boneName=DYN_Dreads_8_4}
    componentAppliedLinearAccClamp=<2500 2500 2500>
    componentLinearAccScale=<0.5 0.5 0.5>
    componentLinearVelScale=<0.5 0.5 0.5>
    physicsBodyDefinitions [{
      sphereCollisionRadius=0
      boundBone {boneName=DYN_Dreads_8_4}
      boxExtents=<12 12 8>
      constraintSetup {angularLimitsMax=<45 0 45>, angularLimitsMin=<-45 -45 -45>, angularTarget=<0 0 0>}
      localJointOffset=<2 0 0.5>
     }]
   }
   n8 = Anim Dynamics - Bone: DYN_Dreads_7_1!(ComponentPose=@n7, Alpha=1, GravityOverride:Vector="0,0,0", ExternalForce="0,0,0") -> (Pose) {
    angularDampingOverride=1
    angularSpringConstant=150
    bAngularSpring=true
    bOverrideAngularDamping=true
    bUseGravityOverride=true
    simulationSpace=RootRelative
    alphaBoolBlend {blendOption=Linear}
    boundBone {boneName=DYN_Dreads_7_1}
    componentAppliedLinearAccClamp=<2500 2500 2500>
    componentLinearAccScale=<0.5 0.5 0.5>
    componentLinearVelScale=<0.5 0.5 0.5>
    physicsBodyDefinitions [{
      sphereCollisionRadius=0
      boundBone {boneName=DYN_Dreads_7_1}
      boxExtents=<12 12 8>
      constraintSetup {angularLimitsMax=<45 0 45>, angularLimitsMin=<-45 -45 -45>, angularTarget=<0 0 0>}
      localJointOffset=<1 0 0.5>
     }]
   }
   n9 = Anim Dynamics - Bone: DYN_Dreads_7_2!(ComponentPose=@n8, Alpha=1, GravityOverride:Vector="0,0,0", ExternalForce="0,0,0") -> (Pose) {
    angularDampingOverride=1
    angularSpringConstant=150
    bAngularSpring=true
    bOverrideAngularDamping=true
    bUseGravityOverride=true
    simulationSpace=RootRelative
    alphaBoolBlend {blendOption=Linear}
    boundBone {boneName=DYN_Dreads_7_2}
    componentAppliedLinearAccClamp=<2500 2500 2500>
    componentLinearAccScale=<0.5 0.5 0.5>
    componentLinearVelScale=<0.5 0.5 0.5>
    physicsBodyDefinitions [{
      sphereCollisionRadius=0
      boundBone {boneName=DYN_Dreads_7_2}
      boxExtents=<12 12 8>
      constraintSetup {angularLimitsMax=<45 0 45>, angularLimitsMin=<-45 -45 -45>, angularTarget=<0 0 0>}
      localJointOffset=<1 0 0.5>
     }]
   }
   n10 = Anim Dynamics - Bone: DYN_Dreads_7_3!(ComponentPose=@n9, Alpha=1, GravityOverride:Vector="0,0,0", ExternalForce="0,0,0") -> (Pose) {
    angularDampingOverride=1
    angularSpringConstant=150
    bAngularSpring=true
    bOverrideAngularDamping=true
    bUseGravityOverride=true
    simulationSpace=RootRelative
    alphaBoolBlend {blendOption=Linear}
    boundBone {boneName=DYN_Dreads_7_3}
    componentAppliedLinearAccClamp=<2500 2500 2500>
    componentLinearAccScale=<0.5 0.5 0.5>
    componentLinearVelScale=<0.5 0.5 0.5>
    physicsBodyDefinitions [{
      sphereCollisionRadius=0
      boundBone {boneName=DYN_Dreads_7_3}
      boxExtents=<12 12 8>
      constraintSetup {angularLimitsMax=<45 0 45>, angularLimitsMin=<-45 -45 -45>, angularTarget=<0 0 0>}
      localJointOffset=<1 0 0.5>
     }]
   }
   n11 = Anim Dynamics - Bone: DYN_Dreads_7_4!(ComponentPose=@n10, Alpha=1, GravityOverride:Vector="0,0,0", ExternalForce="0,0,0") -> (Pose) {
    angularDampingOverride=1
    angularSpringConstant=150
    bAngularSpring=true
    bOverrideAngularDamping=true
    bUseGravityOverride=true
    simulationSpace=RootRelative
    alphaBoolBlend {blendOption=Linear}
    boundBone {boneName=DYN_Dreads_7_4}
    componentAppliedLinearAccClamp=<2500 2500 2500>
    componentLinearAccScale=<0.5 0.5 0.5>
    componentLinearVelScale=<0.5 0.5 0.5>
    physicsBodyDefinitions [{
      sphereCollisionRadius=0
      boundBone {boneName=DYN_Dreads_7_4}
      boxExtents=<12 12 8>
      constraintSetup {angularLimitsMax=<45 0 45>, angularLimitsMin=<-45 -45 -45>, angularTarget=<0 0 0>}
      localJointOffset=<1 0 0.5>
     }]
   }
   n12 = Anim Dynamics - Bone: DYN_Dreads_6_6!(ComponentPose=@n11, Alpha=1, GravityOverride:Vector="0,0,0", ExternalForce="0,0,0") -> (Pose) {
    angularDampingOverride=1
    bAngularSpring=true
    bOverrideAngularDamping=true
    bUseGravityOverride=true
    simulationSpace=RootRelative
    alphaBoolBlend {blendOption=Linear}
    boundBone {boneName=DYN_Dreads_6_6}
    componentAppliedLinearAccClamp=<2500 2500 2500>
    componentLinearAccScale=<0.5 0.5 0.5>
    componentLinearVelScale=<0.5 0.5 0.5>
    physicsBodyDefinitions [{
      sphereCollisionRadius=0
      boundBone {boneName=DYN_Dreads_6_6}
      boxExtents=<12 12 8>
      constraintSetup {angularLimitsMax=<45 0 45>, angularLimitsMin=<-45 -45 -45>, angularTarget=<0 0 0>}
      localJointOffset=<2 0 0.5>
     }]
   }
   n13 = Anim Dynamics - Bone: DYN_Dreads_6_2!(ComponentPose=@n12, Alpha=1, GravityOverride:Vector="0,0,0", ExternalForce="0,0,0") -> (Pose) {
    angularDampingOverride=1
    bAngularSpring=true
    bOverrideAngularDamping=true
    bUseGravityOverride=true
    simulationSpace=RootRelative
    alphaBoolBlend {blendOption=Linear}
    boundBone {boneName=DYN_Dreads_6_2}
    componentAppliedLinearAccClamp=<2500 2500 2500>
    componentLinearAccScale=<0.5 0.5 0.5>
    componentLinearVelScale=<0.5 0.5 0.5>
    physicsBodyDefinitions [{
      sphereCollisionRadius=0
      boundBone {boneName=DYN_Dreads_6_2}
      boxExtents=<12 12 8>
      constraintSetup {angularLimitsMax=<45 0 45>, angularLimitsMin=<-45 -45 -45>, angularTarget=<0 0 0>}
      localJointOffset=<2 0 0.5>
     }]
   }
   n14 = Anim Dynamics - Bone: DYN_Dreads_6_3!(ComponentPose=@n13, Alpha=1, GravityOverride:Vector="0,0,0", ExternalForce="0,0,0") -> (Pose) {
    angularDampingOverride=1
    bAngularSpring=true
    bOverrideAngularDamping=true
    bUseGravityOverride=true
    simulationSpace=RootRelative
    alphaBoolBlend {blendOption=Linear}
    boundBone {boneName=DYN_Dreads_6_3}
    componentAppliedLinearAccClamp=<2500 2500 2500>
    componentLinearAccScale=<0.5 0.5 0.5>
    componentLinearVelScale=<0.5 0.5 0.5>
    physicsBodyDefinitions [{
      sphereCollisionRadius=0
      boundBone {boneName=DYN_Dreads_6_3}
      boxExtents=<12 12 8>
      constraintSetup {angularLimitsMax=<45 0 45>, angularLimitsMin=<-45 -45 -45>, angularTarget=<0 0 0>}
      localJointOffset=<2 0 0.5>
     }]
   }
   n15 = Anim Dynamics - Bone: DYN_Dreads_6_4!(ComponentPose=@n14, Alpha=1, GravityOverride:Vector="0,0,0", ExternalForce="0,0,0") -> (Pose) {
    angularDampingOverride=1
    bAngularSpring=true
    bOverrideAngularDamping=true
    bUseGravityOverride=true
    simulationSpace=RootRelative
    alphaBoolBlend {blendOption=Linear}
    boundBone {boneName=DYN_Dreads_6_4}
    componentAppliedLinearAccClamp=<2500 2500 2500>
    componentLinearAccScale=<0.5 0.5 0.5>
    componentLinearVelScale=<0.5 0.5 0.5>
    physicsBodyDefinitions [{
      sphereCollisionRadius=0
      boundBone {boneName=DYN_Dreads_6_4}
      boxExtents=<12 12 8>
      constraintSetup {angularLimitsMax=<45 0 45>, angularLimitsMin=<-45 -45 -45>, angularTarget=<0 0 0>}
      localJointOffset=<2 0 0.5>
     }]
   }
   n16 = Anim Dynamics - Bone: DYN_Dreads_5_12!(ComponentPose=@n15, Alpha=1, GravityOverride:Vector="0,0,0", ExternalForce="0,0,0") -> (Pose) {
    angularDampingOverride=1
    angularSpringConstant=150
    bAngularSpring=true
    bOverrideAngularDamping=true
    bUseGravityOverride=true
    simulationSpace=RootRelative
    alphaBoolBlend {blendOption=Linear}
    boundBone {boneName=DYN_Dreads_5_12}
    componentAppliedLinearAccClamp=<2500 2500 2500>
    componentLinearAccScale=<0.5 0.5 0.5>
    componentLinearVelScale=<0.5 0.5 0.5>
    physicsBodyDefinitions [{
      sphereCollisionRadius=0
      boundBone {boneName=DYN_Dreads_5_12}
      boxExtents=<12 12 8>
      constraintSetup {angularLimitsMax=<45 0 45>, angularLimitsMin=<-45 -45 -45>, angularTarget=<0 0 0>}
      localJointOffset=<1 0 0.5>
     }]
   }
   n17 = Anim Dynamics - Bone: DYN_Dreads_5_8!(ComponentPose=@n16, Alpha=1, GravityOverride:Vector="0,0,0", ExternalForce="0,0,0") -> (Pose) {
    angularDampingOverride=1
    angularSpringConstant=150
    bAngularSpring=true
    bOverrideAngularDamping=true
    bUseGravityOverride=true
    simulationSpace=RootRelative
    alphaBoolBlend {blendOption=Linear}
    boundBone {boneName=DYN_Dreads_5_8}
    componentAppliedLinearAccClamp=<2500 2500 2500>
    componentLinearAccScale=<0.5 0.5 0.5>
    componentLinearVelScale=<0.5 0.5 0.5>
    physicsBodyDefinitions [{
      sphereCollisionRadius=0
      boundBone {boneName=DYN_Dreads_5_8}
      boxExtents=<12 12 8>
      constraintSetup {angularLimitsMax=<45 0 45>, angularLimitsMin=<-45 -45 -45>, angularTarget=<0 0 0>}
      localJointOffset=<1 0 0.5>
     }]
   }
   n18 = Anim Dynamics - Bone: DYN_Dreads_5_9!(ComponentPose=@n17, Alpha=1, GravityOverride:Vector="0,0,0", ExternalForce="0,0,0") -> (Pose) {
    angularDampingOverride=1
    angularSpringConstant=150
    bAngularSpring=true
    bOverrideAngularDamping=true
    bUseGravityOverride=true
    simulationSpace=RootRelative
    alphaBoolBlend {blendOption=Linear}
    boundBone {boneName=DYN_Dreads_5_9}
    componentAppliedLinearAccClamp=<2500 2500 2500>
    componentLinearAccScale=<0.5 0.5 0.5>
    componentLinearVelScale=<0.5 0.5 0.5>
    physicsBodyDefinitions [{
      sphereCollisionRadius=0
      boundBone {boneName=DYN_Dreads_5_9}
      boxExtents=<12 12 8>
      constraintSetup {angularLimitsMax=<45 0 45>, angularLimitsMin=<-45 -45 -45>, angularTarget=<0 0 0>}
      localJointOffset=<1 0 0.5>
     }]
   }
   n19 = Anim Dynamics - Bone: DYN_Dreads_5_10!(ComponentPose=@n18, Alpha=1, GravityOverride:Vector="0,0,0", ExternalForce="0,0,0") -> (Pose) {
    angularDampingOverride=1
    angularSpringConstant=150
    bAngularSpring=true
    bOverrideAngularDamping=true
    bUseGravityOverride=true
    simulationSpace=RootRelative
    alphaBoolBlend {blendOption=Linear}
    boundBone {boneName=DYN_Dreads_5_10}
    componentAppliedLinearAccClamp=<2500 2500 2500>
    componentLinearAccScale=<0.5 0.5 0.5>
    componentLinearVelScale=<0.5 0.5 0.5>
    physicsBodyDefinitions [{
      sphereCollisionRadius=0
      boundBone {boneName=DYN_Dreads_5_10}
      boxExtents=<12 12 8>
      constraintSetup {angularLimitsMax=<45 0 45>, angularLimitsMin=<-45 -45 -45>, angularTarget=<0 0 0>}
      localJointOffset=<1 0 0.5>
     }]
   }
   n20 = Anim Dynamics - Bone: DYN_Dreads_5_1!(ComponentPose=@n19, Alpha=1, GravityOverride:Vector="0,0,0", ExternalForce="0,0,0") -> (Pose) {
    angularDampingOverride=1
    bAngularSpring=true
    bOverrideAngularDamping=true
    bUseGravityOverride=true
    simulationSpace=RootRelative
    alphaBoolBlend {blendOption=Linear}
    boundBone {boneName=DYN_Dreads_5_1}
    componentAppliedLinearAccClamp=<2500 2500 2500>
    componentLinearAccScale=<0.5 0.5 0.5>
    componentLinearVelScale=<0.5 0.5 0.5>
    physicsBodyDefinitions [{
      sphereCollisionRadius=0
      boundBone {boneName=DYN_Dreads_5_1}
      boxExtents=<12 12 8>
      constraintSetup {angularLimitsMax=<45 0 45>, angularLimitsMin=<-45 -45 -45>, angularTarget=<0 0 0>}
      localJointOffset=<2 0 0.5>
     }]
   }
   n21 = Anim Dynamics - Bone: DYN_Dreads_5_2!(ComponentPose=@n20, Alpha=1, GravityOverride:Vector="0,0,0", ExternalForce="0,0,0") -> (Pose) {
    angularDampingOverride=1
    bAngularSpring=true
    bOverrideAngularDamping=true
    bUseGravityOverride=true
    simulationSpace=RootRelative
    alphaBoolBlend {blendOption=Linear}
    boundBone {boneName=DYN_Dreads_5_2}
    componentAppliedLinearAccClamp=<2500 2500 2500>
    componentLinearAccScale=<0.5 0.5 0.5>
    componentLinearVelScale=<0.5 0.5 0.5>
    physicsBodyDefinitions [{
      sphereCollisionRadius=0
      boundBone {boneName=DYN_Dreads_5_2}
      boxExtents=<12 12 8>
      constraintSetup {angularLimitsMax=<45 0 45>, angularLimitsMin=<-45 -45 -45>, angularTarget=<0 0 0>}
      localJointOffset=<2 0 0.5>
     }]
   }
   n22 = Anim Dynamics - Bone: DYN_Dreads_5_3!(ComponentPose=@n21, Alpha=1, GravityOverride:Vector="0,0,0", ExternalForce="0,0,0") -> (Pose) {
    angularDampingOverride=1
    bAngularSpring=true
    bOverrideAngularDamping=true
    bUseGravityOverride=true
    simulationSpace=RootRelative
    alphaBoolBlend {blendOption=Linear}
    boundBone {boneName=DYN_Dreads_5_3}
    componentAppliedLinearAccClamp=<2500 2500 2500>
    componentLinearAccScale=<0.5 0.5 0.5>
    componentLinearVelScale=<0.5 0.5 0.5>
    physicsBodyDefinitions [{
      sphereCollisionRadius=0
      boundBone {boneName=DYN_Dreads_5_3}
      boxExtents=<12 12 8>
      constraintSetup {angularLimitsMax=<45 0 45>, angularLimitsMin=<-45 -45 -45>, angularTarget=<0 0 0>}
      localJointOffset=<2 0 0.5>
     }]
   }
   n23 = Anim Dynamics - Bone: DYN_Dreads_5_4!(ComponentPose=@n22, Alpha=1, GravityOverride:Vector="0,0,0", ExternalForce="0,0,0") -> (Pose) {
    angularDampingOverride=1
    bAngularSpring=true
    bOverrideAngularDamping=true
    bUseGravityOverride=true
    simulationSpace=RootRelative
    alphaBoolBlend {blendOption=Linear}
    boundBone {boneName=DYN_Dreads_5_4}
    componentAppliedLinearAccClamp=<2500 2500 2500>
    componentLinearAccScale=<0.5 0.5 0.5>
    componentLinearVelScale=<0.5 0.5 0.5>
    physicsBodyDefinitions [{
      sphereCollisionRadius=0
      boundBone {boneName=DYN_Dreads_5_4}
      boxExtents=<12 12 8>
      constraintSetup {angularLimitsMax=<45 0 45>, angularLimitsMin=<-45 -45 -45>, angularTarget=<0 0 0>}
      localJointOffset=<2 0 0.5>
     }]
   }
   n24 = Anim Dynamics - Bone: DYN_Dreads_5_11!(ComponentPose=@n23, Alpha=1, GravityOverride:Vector="0,0,0", ExternalForce="0,0,0") -> (Pose) {
    angularDampingOverride=1
    angularSpringConstant=150
    bAngularSpring=true
    bOverrideAngularDamping=true
    bUseGravityOverride=true
    simulationSpace=RootRelative
    alphaBoolBlend {blendOption=Linear}
    boundBone {boneName=DYN_Dreads_5_11}
    componentAppliedLinearAccClamp=<2500 2500 2500>
    componentLinearAccScale=<0.5 0.5 0.5>
    componentLinearVelScale=<0.5 0.5 0.5>
    physicsBodyDefinitions [{
      sphereCollisionRadius=0
      boundBone {boneName=DYN_Dreads_5_11}
      boxExtents=<12 12 8>
      constraintSetup {angularLimitsMax=<45 0 45>, angularLimitsMin=<-45 -45 -45>, angularTarget=<0 0 0>}
      localJointOffset=<1 0 0.5>
     }]
   }
   n25 = Anim Dynamics - Bone: DYN_Dreads_5_5!(ComponentPose=@n24, Alpha=1, GravityOverride:Vector="0,0,0", ExternalForce="0,0,0") -> (Pose) {
    angularDampingOverride=1
    angularSpringConstant=150
    bAngularSpring=true
    bOverrideAngularDamping=true
    bUseGravityOverride=true
    simulationSpace=RootRelative
    alphaBoolBlend {blendOption=Linear}
    boundBone {boneName=DYN_Dreads_5_5}
    componentAppliedLinearAccClamp=<2500 2500 2500>
    componentLinearAccScale=<0.5 0.5 0.5>
    componentLinearVelScale=<0.5 0.5 0.5>
    physicsBodyDefinitions [{
      sphereCollisionRadius=0
      boundBone {boneName=DYN_Dreads_5_5}
      boxExtents=<12 12 8>
      constraintSetup {angularLimitsMax=<45 0 45>, angularLimitsMin=<-45 -45 -45>, angularTarget=<0 0 0>}
      localJointOffset=<1 0 0.5>
     }]
   }
   n26 = Anim Dynamics - Bone: DYN_Dreads_5_6!(ComponentPose=@n25, Alpha=1, GravityOverride:Vector="0,0,0", ExternalForce="0,0,0") -> (Pose) {
    angularDampingOverride=1
    angularSpringConstant=150
    bAngularSpring=true
    bOverrideAngularDamping=true
    bUseGravityOverride=true
    simulationSpace=RootRelative
    alphaBoolBlend {blendOption=Linear}
    boundBone {boneName=DYN_Dreads_5_6}
    componentAppliedLinearAccClamp=<2500 2500 2500>
    componentLinearAccScale=<0.5 0.5 0.5>
    componentLinearVelScale=<0.5 0.5 0.5>
    physicsBodyDefinitions [{
      sphereCollisionRadius=0
      boundBone {boneName=DYN_Dreads_5_6}
      boxExtents=<12 12 8>
      constraintSetup {angularLimitsMax=<45 0 45>, angularLimitsMin=<-45 -45 -45>, angularTarget=<0 0 0>}
      localJointOffset=<1 0 0.5>
     }]
   }
   n27 = Anim Dynamics - Bone: DYN_Dreads_5_7!(ComponentPose=@n26, Alpha=1, GravityOverride:Vector="0,0,0", ExternalForce="0,0,0") -> (Pose) {
    angularDampingOverride=1
    angularSpringConstant=150
    bAngularSpring=true
    bOverrideAngularDamping=true
    bUseGravityOverride=true
    simulationSpace=RootRelative
    alphaBoolBlend {blendOption=Linear}
    boundBone {boneName=DYN_Dreads_5_7}
    componentAppliedLinearAccClamp=<2500 2500 2500>
    componentLinearAccScale=<0.5 0.5 0.5>
    componentLinearVelScale=<0.5 0.5 0.5>
    physicsBodyDefinitions [{
      sphereCollisionRadius=0
      boundBone {boneName=DYN_Dreads_5_7}
      boxExtents=<12 12 8>
      constraintSetup {angularLimitsMax=<45 0 45>, angularLimitsMin=<-45 -45 -45>, angularTarget=<0 0 0>}
      localJointOffset=<1 0 0.5>
     }]
   }
   n28 = Anim Dynamics - Bone: DYN_Dreads_4_10!(ComponentPose=@n27, Alpha=1, GravityOverride:Vector="0,0,0", ExternalForce="0,0,0") -> (Pose) {
    angularDampingOverride=1
    bAngularSpring=true
    bOverrideAngularDamping=true
    bUseGravityOverride=true
    simulationSpace=RootRelative
    alphaBoolBlend {blendOption=Linear}
    boundBone {boneName=DYN_Dreads_4_10}
    componentAppliedLinearAccClamp=<2500 2500 2500>
    componentLinearAccScale=<0.5 0.5 0.5>
    componentLinearVelScale=<0.5 0.5 0.5>
    physicsBodyDefinitions [{
      sphereCollisionRadius=0
      boundBone {boneName=DYN_Dreads_4_10}
      boxExtents=<12 12 8>
      constraintSetup {angularLimitsMax=<45 0 45>, angularLimitsMin=<-45 -45 -45>, angularTarget=<0 0 0>}
      localJointOffset=<2 0 0.5>
     }]
   }
   n29 = Anim Dynamics - Bone: DYN_Dreads_4_5!(ComponentPose=@n28, Alpha=1, GravityOverride:Vector="0,0,0", ExternalForce="0,0,0") -> (Pose) {
    angularDampingOverride=1
    bAngularSpring=true
    bOverrideAngularDamping=true
    bUseGravityOverride=true
    simulationSpace=RootRelative
    alphaBoolBlend {blendOption=Linear}
    boundBone {boneName=DYN_Dreads_4_5}
    componentAppliedLinearAccClamp=<2500 2500 2500>
    componentLinearAccScale=<0.5 0.5 0.5>
    componentLinearVelScale=<0.5 0.5 0.5>
    physicsBodyDefinitions [{
      sphereCollisionRadius=0
      boundBone {boneName=DYN_Dreads_4_5}
      boxExtents=<12 12 8>
      constraintSetup {angularLimitsMax=<45 0 45>, angularLimitsMin=<-45 -45 -45>, angularTarget=<0 0 0>}
      localJointOffset=<2 0 0.5>
     }]
   }
   n30 = Anim Dynamics - Bone: DYN_Dreads_4_6!(ComponentPose=@n29, Alpha=1, GravityOverride:Vector="0,0,0", ExternalForce="0,0,0") -> (Pose) {
    angularDampingOverride=1
    bAngularSpring=true
    bOverrideAngularDamping=true
    bUseGravityOverride=true
    simulationSpace=RootRelative
    alphaBoolBlend {blendOption=Linear}
    boundBone {boneName=DYN_Dreads_4_6}
    componentAppliedLinearAccClamp=<2500 2500 2500>
    componentLinearAccScale=<0.5 0.5 0.5>
    componentLinearVelScale=<0.5 0.5 0.5>
    physicsBodyDefinitions [{
      sphereCollisionRadius=0
      boundBone {boneName=DYN_Dreads_4_6}
      boxExtents=<12 12 8>
      constraintSetup {angularLimitsMax=<45 0 45>, angularLimitsMin=<-45 -45 -45>, angularTarget=<0 0 0>}
      localJointOffset=<2 0 0.5>
     }]
   }
   n31 = Anim Dynamics - Bone: DYN_Dreads_4_7!(ComponentPose=@n30, Alpha=1, GravityOverride:Vector="0,0,0", ExternalForce="0,0,0") -> (Pose) {
    angularDampingOverride=1
    bAngularSpring=true
    bOverrideAngularDamping=true
    bUseGravityOverride=true
    simulationSpace=RootRelative
    alphaBoolBlend {blendOption=Linear}
    boundBone {boneName=DYN_Dreads_4_7}
    componentAppliedLinearAccClamp=<2500 2500 2500>
    componentLinearAccScale=<0.5 0.5 0.5>
    componentLinearVelScale=<0.5 0.5 0.5>
    physicsBodyDefinitions [{
      sphereCollisionRadius=0
      boundBone {boneName=DYN_Dreads_4_7}
      boxExtents=<12 12 8>
      constraintSetup {angularLimitsMax=<45 0 45>, angularLimitsMin=<-45 -45 -45>, angularTarget=<0 0 0>}
      localJointOffset=<2 0 0.5>
     }]
   }
   n32 = Anim Dynamics - Bone: DYN_Dreads_4_1!(ComponentPose=@n31, Alpha=1, GravityOverride:Vector="0,0,0", ExternalForce="0,0,0") -> (Pose) {
    angularDampingOverride=1
    angularSpringConstant=150
    bAngularSpring=true
    bOverrideAngularDamping=true
    bUseGravityOverride=true
    simulationSpace=RootRelative
    alphaBoolBlend {blendOption=Linear}
    boundBone {boneName=DYN_Dreads_4_1}
    componentAppliedLinearAccClamp=<2500 2500 2500>
    componentLinearAccScale=<0.5 0.5 0.5>
    componentLinearVelScale=<0.5 0.5 0.5>
    physicsBodyDefinitions [{
      sphereCollisionRadius=0
      boundBone {boneName=DYN_Dreads_4_1}
      boxExtents=<12 12 8>
      constraintSetup {angularLimitsMax=<45 0 45>, angularLimitsMin=<-45 -45 -45>, angularTarget=<0 0 0>}
      localJointOffset=<1 0 0.5>
     }]
   }
   n33 = Anim Dynamics - Bone: DYN_Dreads_4_2!(ComponentPose=@n32, Alpha=1, GravityOverride:Vector="0,0,0", ExternalForce="0,0,0") -> (Pose) {
    angularDampingOverride=1
    angularSpringConstant=150
    bAngularSpring=true
    bOverrideAngularDamping=true
    bUseGravityOverride=true
    simulationSpace=RootRelative
    alphaBoolBlend {blendOption=Linear}
    boundBone {boneName=DYN_Dreads_4_2}
    componentAppliedLinearAccClamp=<2500 2500 2500>
    componentLinearAccScale=<0.5 0.5 0.5>
    componentLinearVelScale=<0.5 0.5 0.5>
    physicsBodyDefinitions [{
      sphereCollisionRadius=0
      boundBone {boneName=DYN_Dreads_4_2}
      boxExtents=<12 12 8>
      constraintSetup {angularLimitsMax=<45 0 45>, angularLimitsMin=<-45 -45 -45>, angularTarget=<0 0 0>}
      localJointOffset=<1 0 0.5>
     }]
   }
   n34 = Anim Dynamics - Bone: DYN_Dreads_4_3!(ComponentPose=@n33, Alpha=1, GravityOverride:Vector="0,0,0", ExternalForce="0,0,0") -> (Pose) {
    angularDampingOverride=1
    angularSpringConstant=150
    bAngularSpring=true
    bOverrideAngularDamping=true
    bUseGravityOverride=true
    simulationSpace=RootRelative
    alphaBoolBlend {blendOption=Linear}
    boundBone {boneName=DYN_Dreads_4_3}
    componentAppliedLinearAccClamp=<2500 2500 2500>
    componentLinearAccScale=<0.5 0.5 0.5>
    componentLinearVelScale=<0.5 0.5 0.5>
    physicsBodyDefinitions [{
      sphereCollisionRadius=0
      boundBone {boneName=DYN_Dreads_4_3}
      boxExtents=<12 12 8>
      constraintSetup {angularLimitsMax=<45 0 45>, angularLimitsMin=<-45 -45 -45>, angularTarget=<0 0 0>}
      localJointOffset=<1 0 0.5>
     }]
   }
   n35 = Anim Dynamics - Bone: DYN_Dreads_4_4!(ComponentPose=@n34, Alpha=1, GravityOverride:Vector="0,0,0", ExternalForce="0,0,0") -> (Pose) {
    angularDampingOverride=1
    angularSpringConstant=150
    bAngularSpring=true
    bOverrideAngularDamping=true
    bUseGravityOverride=true
    simulationSpace=RootRelative
    alphaBoolBlend {blendOption=Linear}
    boundBone {boneName=DYN_Dreads_4_4}
    componentAppliedLinearAccClamp=<2500 2500 2500>
    componentLinearAccScale=<0.5 0.5 0.5>
    componentLinearVelScale=<0.5 0.5 0.5>
    physicsBodyDefinitions [{
      sphereCollisionRadius=0
      boundBone {boneName=DYN_Dreads_4_4}
      boxExtents=<12 12 8>
      constraintSetup {angularLimitsMax=<45 0 45>, angularLimitsMin=<-45 -45 -45>, angularTarget=<0 0 0>}
      localJointOffset=<1 0 0.5>
     }]
   }
   n36 = Anim Dynamics - Bone: DYN_Dreads_4_11!(ComponentPose=@n35, Alpha=1, GravityOverride:Vector="0,0,0", ExternalForce="0,0,0") -> (Pose) {
    angularDampingOverride=1
    bAngularSpring=true
    bOverrideAngularDamping=true
    bUseGravityOverride=true
    simulationSpace=RootRelative
    alphaBoolBlend {blendOption=Linear}
    boundBone {boneName=DYN_Dreads_4_11}
    componentAppliedLinearAccClamp=<2500 2500 2500>
    componentLinearAccScale=<0.5 0.5 0.5>
    componentLinearVelScale=<0.5 0.5 0.5>
    physicsBodyDefinitions [{
      sphereCollisionRadius=0
      boundBone {boneName=DYN_Dreads_4_11}
      boxExtents=<12 12 8>
      constraintSetup {angularLimitsMax=<45 0 45>, angularLimitsMin=<-45 -45 -45>, angularTarget=<0 0 0>}
      localJointOffset=<2 0 0.5>
     }]
   }
   n37 = Anim Dynamics - Bone: DYN_Dreads_4_8!(ComponentPose=@n36, Alpha=1, GravityOverride:Vector="0,0,0", ExternalForce="0,0,0") -> (Pose) {
    angularDampingOverride=1
    bAngularSpring=true
    bOverrideAngularDamping=true
    bUseGravityOverride=true
    simulationSpace=RootRelative
    alphaBoolBlend {blendOption=Linear}
    boundBone {boneName=DYN_Dreads_4_8}
    componentAppliedLinearAccClamp=<2500 2500 2500>
    componentLinearAccScale=<0.5 0.5 0.5>
    componentLinearVelScale=<0.5 0.5 0.5>
    physicsBodyDefinitions [{
      sphereCollisionRadius=0
      boundBone {boneName=DYN_Dreads_4_8}
      boxExtents=<12 12 8>
      constraintSetup {angularLimitsMax=<45 0 45>, angularLimitsMin=<-45 -45 -45>, angularTarget=<0 0 0>}
      localJointOffset=<2 0 0.5>
     }]
   }
   n38 = Anim Dynamics - Bone: DYN_Dreads_4_9!(ComponentPose=@n37, Alpha=1, GravityOverride:Vector="0,0,0", ExternalForce="0,0,0") -> (Pose) {
    angularDampingOverride=1
    bAngularSpring=true
    bOverrideAngularDamping=true
    bUseGravityOverride=true
    simulationSpace=RootRelative
    alphaBoolBlend {blendOption=Linear}
    boundBone {boneName=DYN_Dreads_4_9}
    componentAppliedLinearAccClamp=<2500 2500 2500>
    componentLinearAccScale=<0.5 0.5 0.5>
    componentLinearVelScale=<0.5 0.5 0.5>
    physicsBodyDefinitions [{
      sphereCollisionRadius=0
      boundBone {boneName=DYN_Dreads_4_9}
      boxExtents=<12 12 8>
      constraintSetup {angularLimitsMax=<45 0 45>, angularLimitsMin=<-45 -45 -45>, angularTarget=<0 0 0>}
      localJointOffset=<2 0 0.5>
     }]
   }
   n39 = Anim Dynamics - Bone: DYN_Dreads_3_10!(ComponentPose=@n38, Alpha=1, GravityOverride:Vector="0,0,0", ExternalForce="0,0,0") -> (Pose) {
    angularDampingOverride=1
    bAngularSpring=true
    bOverrideAngularDamping=true
    bUseGravityOverride=true
    simulationSpace=RootRelative
    alphaBoolBlend {blendOption=Linear}
    boundBone {boneName=DYN_Dreads_3_10}
    componentAppliedLinearAccClamp=<2500 2500 2500>
    componentLinearAccScale=<0.5 0.5 0.5>
    componentLinearVelScale=<0.5 0.5 0.5>
    physicsBodyDefinitions [{
      sphereCollisionRadius=0
      boundBone {boneName=DYN_Dreads_3_10}
      boxExtents=<12 12 8>
      constraintSetup {angularLimitsMax=<45 0 45>, angularLimitsMin=<-45 -45 -45>, angularTarget=<0 0 0>}
      localJointOffset=<2 0 0.5>
     }]
   }
   n40 = Anim Dynamics - Bone: DYN_Dreads_3_11!(ComponentPose=@n39, Alpha=1, GravityOverride:Vector="0,0,0", ExternalForce="0,0,0") -> (Pose) {
    angularDampingOverride=1
    angularSpringConstant=150
    bAngularSpring=true
    bOverrideAngularDamping=true
    bUseGravityOverride=true
    simulationSpace=RootRelative
    alphaBoolBlend {blendOption=Linear}
    boundBone {boneName=DYN_Dreads_3_11}
    componentAppliedLinearAccClamp=<2500 2500 2500>
    componentLinearAccScale=<0.5 0.5 0.5>
    componentLinearVelScale=<0.5 0.5 0.5>
    physicsBodyDefinitions [{
      sphereCollisionRadius=0
      boundBone {boneName=DYN_Dreads_3_11}
      boxExtents=<12 12 8>
      constraintSetup {angularLimitsMax=<45 0 45>, angularLimitsMin=<-45 -45 -45>, angularTarget=<0 0 0>}
      localJointOffset=<1 0 0.5>
     }]
   }
   n41 = Anim Dynamics - Bone: DYN_Dreads_3_7!(ComponentPose=@n40, Alpha=1, GravityOverride:Vector="0,0,0", ExternalForce="0,0,0") -> (Pose) {
    angularDampingOverride=1
    angularSpringConstant=150
    bAngularSpring=true
    bOverrideAngularDamping=true
    bUseGravityOverride=true
    simulationSpace=RootRelative
    alphaBoolBlend {blendOption=Linear}
    boundBone {boneName=DYN_Dreads_3_7}
    componentAppliedLinearAccClamp=<2500 2500 2500>
    componentLinearAccScale=<0.5 0.5 0.5>
    componentLinearVelScale=<0.5 0.5 0.5>
    physicsBodyDefinitions [{
      sphereCollisionRadius=0
      boundBone {boneName=DYN_Dreads_3_7}
      boxExtents=<12 12 8>
      constraintSetup {angularLimitsMax=<45 0 45>, angularLimitsMin=<-45 -45 -45>, angularTarget=<0 0 0>}
      localJointOffset=<1 0 0.5>
     }]
   }
   n42 = Anim Dynamics - Bone: DYN_Dreads_3_8!(ComponentPose=@n41, Alpha=1, GravityOverride:Vector="0,0,0", ExternalForce="0,0,0") -> (Pose) {
    angularDampingOverride=1
    angularSpringConstant=150
    bAngularSpring=true
    bOverrideAngularDamping=true
    bUseGravityOverride=true
    simulationSpace=RootRelative
    alphaBoolBlend {blendOption=Linear}
    boundBone {boneName=DYN_Dreads_3_8}
    componentAppliedLinearAccClamp=<2500 2500 2500>
    componentLinearAccScale=<0.5 0.5 0.5>
    componentLinearVelScale=<0.5 0.5 0.5>
    physicsBodyDefinitions [{
      sphereCollisionRadius=0
      boundBone {boneName=DYN_Dreads_3_8}
      boxExtents=<12 12 8>
      constraintSetup {angularLimitsMax=<45 0 45>, angularLimitsMin=<-45 -45 -45>, angularTarget=<0 0 0>}
      localJointOffset=<1 0 0.5>
     }]
   }
   n43 = Anim Dynamics - Bone: DYN_Dreads_3_9!(ComponentPose=@n42, Alpha=1, GravityOverride:Vector="0,0,0", ExternalForce="0,0,0") -> (Pose) {
    angularDampingOverride=1
    angularSpringConstant=150
    bAngularSpring=true
    bOverrideAngularDamping=true
    bUseGravityOverride=true
    simulationSpace=RootRelative
    alphaBoolBlend {blendOption=Linear}
    boundBone {boneName=DYN_Dreads_3_9}
    componentAppliedLinearAccClamp=<2500 2500 2500>
    componentLinearAccScale=<0.5 0.5 0.5>
    componentLinearVelScale=<0.5 0.5 0.5>
    physicsBodyDefinitions [{
      sphereCollisionRadius=0
      boundBone {boneName=DYN_Dreads_3_9}
      boxExtents=<12 12 8>
      constraintSetup {angularLimitsMax=<45 0 45>, angularLimitsMin=<-45 -45 -45>, angularTarget=<0 0 0>}
      localJointOffset=<1 0 0.5>
     }]
   }
   n44 = Anim Dynamics - Bone: DYN_Dreads_3_1!(ComponentPose=@n43, Alpha=1, GravityOverride:Vector="0,0,0", ExternalForce="0,0,0") -> (Pose) {
    angularDampingOverride=1
    bAngularSpring=true
    bOverrideAngularDamping=true
    bUseGravityOverride=true
    simulationSpace=RootRelative
    alphaBoolBlend {blendOption=Linear}
    boundBone {boneName=DYN_Dreads_3_1}
    componentAppliedLinearAccClamp=<2500 2500 2500>
    componentLinearAccScale=<0.5 0.5 0.5>
    componentLinearVelScale=<0.5 0.5 0.5>
    physicsBodyDefinitions [{
      sphereCollisionRadius=0
      boundBone {boneName=DYN_Dreads_3_1}
      boxExtents=<12 12 8>
      constraintSetup {angularLimitsMax=<45 0 45>, angularLimitsMin=<-45 -45 -45>, angularTarget=<0 0 0>}
      localJointOffset=<2 0 0.5>
     }]
   }
   n45 = Anim Dynamics - Bone: DYN_Dreads_3_2!(ComponentPose=@n44, Alpha=1, GravityOverride:Vector="0,0,0", ExternalForce="0,0,0") -> (Pose) {
    angularDampingOverride=1
    bAngularSpring=true
    bOverrideAngularDamping=true
    bUseGravityOverride=true
    simulationSpace=RootRelative
    alphaBoolBlend {blendOption=Linear}
    boundBone {boneName=DYN_Dreads_3_2}
    componentAppliedLinearAccClamp=<2500 2500 2500>
    componentLinearAccScale=<0.5 0.5 0.5>
    componentLinearVelScale=<0.5 0.5 0.5>
    physicsBodyDefinitions [{
      sphereCollisionRadius=0
      boundBone {boneName=DYN_Dreads_3_2}
      boxExtents=<12 12 8>
      constraintSetup {angularLimitsMax=<45 0 45>, angularLimitsMin=<-45 -45 -45>, angularTarget=<0 0 0>}
      localJointOffset=<2 0 0.5>
     }]
   }
   n46 = Anim Dynamics - Bone: DYN_Dreads_3_3!(ComponentPose=@n45, Alpha=1, GravityOverride:Vector="0,0,0", ExternalForce="0,0,0") -> (Pose) {
    angularDampingOverride=1
    bAngularSpring=true
    bOverrideAngularDamping=true
    bUseGravityOverride=true
    simulationSpace=RootRelative
    alphaBoolBlend {blendOption=Linear}
    boundBone {boneName=DYN_Dreads_3_3}
    componentAppliedLinearAccClamp=<2500 2500 2500>
    componentLinearAccScale=<0.5 0.5 0.5>
    componentLinearVelScale=<0.5 0.5 0.5>
    physicsBodyDefinitions [{
      sphereCollisionRadius=0
      boundBone {boneName=DYN_Dreads_3_3}
      boxExtents=<12 12 8>
      constraintSetup {angularLimitsMax=<45 0 45>, angularLimitsMin=<-45 -45 -45>, angularTarget=<0 0 0>}
      localJointOffset=<2 0 0.5>
     }]
   }
   n47 = Anim Dynamics - Bone: DYN_Dreads_3_4!(ComponentPose=@n46, Alpha=1, GravityOverride:Vector="0,0,0", ExternalForce="0,0,0") -> (Pose) {
    angularDampingOverride=1
    bAngularSpring=true
    bOverrideAngularDamping=true
    bUseGravityOverride=true
    simulationSpace=RootRelative
    alphaBoolBlend {blendOption=Linear}
    boundBone {boneName=DYN_Dreads_3_4}
    componentAppliedLinearAccClamp=<2500 2500 2500>
    componentLinearAccScale=<0.5 0.5 0.5>
    componentLinearVelScale=<0.5 0.5 0.5>
    physicsBodyDefinitions [{
      sphereCollisionRadius=0
      boundBone {boneName=DYN_Dreads_3_4}
      boxExtents=<12 12 8>
      constraintSetup {angularLimitsMax=<45 0 45>, angularLimitsMin=<-45 -45 -45>, angularTarget=<0 0 0>}
      localJointOffset=<2 0 0.5>
     }]
   }
   n48 = Anim Dynamics - Bone: DYN_Dreads_3_13!(ComponentPose=@n47, Alpha=1, GravityOverride:Vector="0,0,0", ExternalForce="0,0,0") -> (Pose) {
    angularDampingOverride=1
    angularSpringConstant=150
    bAngularSpring=true
    bOverrideAngularDamping=true
    bUseGravityOverride=true
    simulationSpace=RootRelative
    alphaBoolBlend {blendOption=Linear}
    boundBone {boneName=DYN_Dreads_3_13}
    componentAppliedLinearAccClamp=<2500 2500 2500>
    componentLinearAccScale=<0.5 0.5 0.5>
    componentLinearVelScale=<0.5 0.5 0.5>
    physicsBodyDefinitions [{
      sphereCollisionRadius=0
      boundBone {boneName=DYN_Dreads_3_13}
      boxExtents=<12 12 8>
      constraintSetup {angularLimitsMax=<25 0 45>, angularLimitsMin=<-25 -45 -45>, angularTarget=<0 0 0>}
      localJointOffset=<1 0 0.5>
     }]
   }
   n49 = Anim Dynamics - Bone: DYN_Dreads_3_12!(ComponentPose=@n48, Alpha=1, GravityOverride:Vector="0,0,0", ExternalForce="0,0,0") -> (Pose) {
    angularDampingOverride=1
    angularSpringConstant=150
    bAngularSpring=true
    bOverrideAngularDamping=true
    bUseGravityOverride=true
    simulationSpace=RootRelative
    alphaBoolBlend {blendOption=Linear}
    boundBone {boneName=DYN_Dreads_3_12}
    componentAppliedLinearAccClamp=<2500 2500 2500>
    componentLinearAccScale=<0.5 0.5 0.5>
    componentLinearVelScale=<0.5 0.5 0.5>
    physicsBodyDefinitions [{
      sphereCollisionRadius=0
      boundBone {boneName=DYN_Dreads_3_12}
      boxExtents=<12 12 8>
      constraintSetup {angularLimitsMax=<25 0 45>, angularLimitsMin=<-25 -45 -45>, angularTarget=<0 0 0>}
      localJointOffset=<1 0 0.5>
     }]
   }
   n50 = Anim Dynamics - Bone: DYN_Dreads_3_5!(ComponentPose=@n49, Alpha=1, GravityOverride:Vector="0,0,0", ExternalForce="0,0,0") -> (Pose) {
    angularDampingOverride=1
    angularSpringConstant=150
    bAngularSpring=true
    bOverrideAngularDamping=true
    bUseGravityOverride=true
    simulationSpace=RootRelative
    alphaBoolBlend {blendOption=Linear}
    boundBone {boneName=DYN_Dreads_3_5}
    componentAppliedLinearAccClamp=<2500 2500 2500>
    componentLinearAccScale=<0.5 0.5 0.5>
    componentLinearVelScale=<0.5 0.5 0.5>
    physicsBodyDefinitions [{
      sphereCollisionRadius=0
      boundBone {boneName=DYN_Dreads_3_5}
      boxExtents=<12 12 8>
      constraintSetup {angularLimitsMax=<25 0 45>, angularLimitsMin=<-25 -45 -45>, angularTarget=<0 0 0>}
      localJointOffset=<1 0 0.5>
     }]
   }
   n51 = Anim Dynamics - Bone: DYN_Dreads_3_6!(ComponentPose=@n50, Alpha=1, GravityOverride:Vector="0,0,0", ExternalForce="0,0,0") -> (Pose) {
    angularDampingOverride=1
    angularSpringConstant=150
    bAngularSpring=true
    bOverrideAngularDamping=true
    bUseGravityOverride=true
    simulationSpace=RootRelative
    alphaBoolBlend {blendOption=Linear}
    boundBone {boneName=DYN_Dreads_3_6}
    componentAppliedLinearAccClamp=<2500 2500 2500>
    componentLinearAccScale=<0.5 0.5 0.5>
    componentLinearVelScale=<0.5 0.5 0.5>
    physicsBodyDefinitions [{
      sphereCollisionRadius=0
      boundBone {boneName=DYN_Dreads_3_6}
      boxExtents=<12 12 8>
      constraintSetup {angularLimitsMax=<25 0 45>, angularLimitsMin=<-25 -45 -45>, angularTarget=<0 0 0>}
      localJointOffset=<1 0 0.5>
     }]
   }
   n52 = Anim Dynamics - Bone: DYN_Dreads_3_13(ComponentPose=@n51, Alpha=1, GravityOverride="0,0,0", ExternalForce="0,0,0") {
    angularDampingOverride=1
    bAngularSpring=true
    bOverrideAngularDamping=true
    bUseGravityOverride=true
    simulationSpace=RootRelative
    alphaBoolBlend {blendOption=Linear}
    boundBone {boneName=DYN_Dreads_3_13}
    componentAppliedLinearAccClamp=<2500 2500 2500>
    componentLinearAccScale=<0.5 0.5 0.5>
    componentLinearVelScale=<0.5 0.5 0.5>
    physicsBodyDefinitions [{
      sphereCollisionRadius=0
      boundBone {boneName=DYN_Dreads_3_13}
      boxExtents=<12 12 8>
      constraintSetup {angularLimitsMax=<25 0 45>, angularLimitsMin=<-25 -45 -45>, angularTarget=<0 0 0>}
      localJointOffset=<2 0 0.5>
     }]
   }
   n53 = Anim Dynamics - Bone: DYN_Dreads_3_12(ComponentPose=@n52, Alpha=1, GravityOverride="0,0,0", ExternalForce="0,0,0") {
    angularDampingOverride=1
    bAngularSpring=true
    bOverrideAngularDamping=true
    bUseGravityOverride=true
    simulationSpace=RootRelative
    alphaBoolBlend {blendOption=Linear}
    boundBone {boneName=DYN_Dreads_3_12}
    componentAppliedLinearAccClamp=<2500 2500 2500>
    componentLinearAccScale=<0.5 0.5 0.5>
    componentLinearVelScale=<0.5 0.5 0.5>
    physicsBodyDefinitions [{
      sphereCollisionRadius=0
      boundBone {boneName=DYN_Dreads_3_12}
      boxExtents=<12 12 8>
      constraintSetup {angularLimitsMax=<25 0 45>, angularLimitsMin=<-25 -45 -45>, angularTarget=<0 0 0>}
      localJointOffset=<2 0 0.5>
     }]
   }
   n54 = Anim Dynamics - Bone: DYN_Dreads_3_5(ComponentPose=@n53, Alpha=1, GravityOverride="0,0,0", ExternalForce="0,0,0") {
    angularDampingOverride=1
    bAngularSpring=true
    bOverrideAngularDamping=true
    bUseGravityOverride=true
    simulationSpace=RootRelative
    alphaBoolBlend {blendOption=Linear}
    boundBone {boneName=DYN_Dreads_3_5}
    componentAppliedLinearAccClamp=<2500 2500 2500>
    componentLinearAccScale=<0.5 0.5 0.5>
    componentLinearVelScale=<0.5 0.5 0.5>
    physicsBodyDefinitions [{
      sphereCollisionRadius=0
      boundBone {boneName=DYN_Dreads_3_5}
      boxExtents=<12 12 8>
      constraintSetup {angularLimitsMax=<25 0 45>, angularLimitsMin=<-25 -45 -45>, angularTarget=<0 0 0>}
      localJointOffset=<2 0 0.5>
     }]
   }
   n55 = Anim Dynamics - Bone: DYN_Dreads_3_6(ComponentPose=@n54, Alpha=1, GravityOverride="0,0,0", ExternalForce="0,0,0") {
    angularDampingOverride=1
    bAngularSpring=true
    bOverrideAngularDamping=true
    bUseGravityOverride=true
    simulationSpace=RootRelative
    alphaBoolBlend {blendOption=Linear}
    boundBone {boneName=DYN_Dreads_3_6}
    componentAppliedLinearAccClamp=<2500 2500 2500>
    componentLinearAccScale=<0.5 0.5 0.5>
    componentLinearVelScale=<0.5 0.5 0.5>
    physicsBodyDefinitions [{
      sphereCollisionRadius=0
      boundBone {boneName=DYN_Dreads_3_6}
      boxExtents=<12 12 8>
      constraintSetup {angularLimitsMax=<25 0 45>, angularLimitsMin=<-25 -45 -45>, angularTarget=<0 0 0>}
      localJointOffset=<2 0 0.5>
     }]
   }
   n56 = RigidBody!(
    ComponentPose=@n55
    Alpha=SelectFloat(A=0.8, B=1, bPickA=Get NewSkydiving() -> (NewSkydiving:bool))
    OverrideWorldGravity:Vector="0,0,-980"
    ExternalForce
   ) -> (Pose) {
    bClampLinearTranslationLimitToRefPose=true
    bOverrideWorldGravity=true
    overridePhysicsAsset=/Game/Characters/Player/Female/Medium/Bodies/F_MED_Palespooky_Outlander/Meshes/Parts/F_MED_PaleSpooky_Outlander_FaceAcc_PhysAsst
    simulationSpace=BaseBoneSpace
    alphaBoolBlend {blendOption=Linear}
    alphaScaleBiasClamp {bInterpResult=true}
    baseBoneRef {boneName=neck_02}
    componentAppliedLinearAccClamp=<2000 2000 2000>
    componentLinearAccScale=<0.4 0.4 0.5>
    overrideWorldGravity {z=-980}
    simSpaceSettings {worldAlpha=0.2}
   }
   n57 = "Constraint - DYN_Dreads_2_1 (R)"(ComponentPose=@n56, Alpha=0.5, ConstraintWeights_0=1) {
    alphaBoolBlend {blendOption=Linear}
    boneToModify {boneName=DYN_Dreads_2_1}
    constraintSetup [{transformType=Rotation, targetBone {boneName=DYN_Dreads_3_13}}]
    constraintWeights [1]
   }
   n58 = RigidBody(
    ComponentPose=Local To Component(LocalPose=Copy Pose From Mesh() {bCopyCurves=true, bUseAttachedParent=true})
    OverrideWorldGravity="0,0,-980"
   ) {
    bClampLinearTranslationLimitToRefPose=true
    bOverrideWorldGravity=true
    overridePhysicsAsset=/Game/Characters/Player/Female/Medium/Bodies/F_MED_Palespooky_Outlander/Meshes/Parts/F_MED_PaleSpooky_Outlander_FE_FaceAcc_PhysAsst
    simulationSpace=BaseBoneSpace
    alphaBoolBlend {blendOption=Linear}
    alphaScaleBiasClamp {bInterpResult=true}
    baseBoneRef {boneName=neck_01}
    componentAppliedLinearAccClamp=<2000 2000 2000>
    componentLinearAccScale=<0.4 0.4 0.5>
    overrideWorldGravity {z=-980}
   }
   Output Pose(Result=Blend Poses by bool(
     BlendPose_0=Component To Local(ComponentPose=@n58)
     BlendPose_1=Component To Local(ComponentPose=@n57)
     bActiveValue=Get NewFrontEnd() -> (NewFrontEnd:bool)
    ) {blendType=Linear}) {layerGroup=None}
  ]}
 EventGraph {nodes [
   Event BlueprintUpdateAnimation()
   > n1 = Sequence!(execute) -> (then_0, then_1, then_2:Exec)
   > n2 = Gravity Override State Function DEPRECATED(CharacterPartAnimInstance=Self-Reference(), GameTimeInSeconds=GetGameTimeInSeconds(), GravityOverrideParams=Get Gravity Override Params())
   > Set GravityOverride(GravityOverride=@n2)
   n3 = Branch(execute=@n1.then_1, Condition=Get bIsSkyDiving())
   > Set NewSkydiving(NewSkydiving:bool=true) -> (Output_Get:bool)
   Set NewSkydiving(execute=@n3.else, NewSkydiving=false)
   n4 = Branch(execute=@n1.then_2, Condition=Get bIsInFrontEnd())
   > Set NewFrontEnd(NewFrontEnd:bool=true) -> (Output_Get:bool)
   Set NewFrontEnd(execute=@n4.else, NewFrontEnd=false)
   TryGetPawnOwner()
  ]}
]
variables [
 Gravity Override Params {
  type=GravityOverrideParamsStruct
  default {skydive_FloatingDown {gravityOverride=<0 -300 600>}, skydive_GoingUp {gravityOverride=<0 -300 600>}}
 }
 GravityOverride {type=Vector}
 NewFrontEnd {type=boolean}
 NewSkydiving {type=boolean}
]
