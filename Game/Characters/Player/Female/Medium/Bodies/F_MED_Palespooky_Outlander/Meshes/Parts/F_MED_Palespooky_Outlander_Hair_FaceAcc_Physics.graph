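# Physics asset for the hair and face accessory part of the Palespooky Outlander medium female player body, setting up dynamic hair motion.
# Seven hair bone chains (back, sides, front) of capsules and boxes hang from the head through limited swing and twist constraints.
# Back and side chains have tighter second-swing limits than front strands.
# Head, neck, upper spine and right clavicle shapes form the body volumes, linked with free angular constraints.
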PhysicsAsset /Game/Characters/Player/Female/Medium/Bodies/F_MED_Palespooky_Outlander/Meshes/Parts/F_MED_Palespooky_Outlander_Hair_FaceAcc_Physics
bodies {
 clavicle_r [{shapeType=Capsule, length=5.5, radius=6.755, center=<-2.9639 0 -0.0457>, rotation=<rot 0 -89.9999 89.1164>}]
 dyn_hair_bk_00 [{shapeType=Capsule, length=7.5628, radius=1.6556, shapeName=dyn_hair_bk_00_capsule, center=<2.3957 0 -0.0002>}]
 dyn_hair_bk_01 [{shapeType=Capsule, length=7.3009, radius=1.3141, shapeName=dyn_hair_bk_01_capsule, center=<1.7736 0 0.0002>}]
 dyn_hair_bk_02 [{shapeType=Box, extentX=10.0425, extentY=1.8791, extentZ=7.5124, center=<5.0038 0 -0.0001>, rotation=<rot 0 0 -0.1696>}]
 dyn_hair_bk_03 [{shapeType=Box, extentX=12.4028, extentY=2.5131, extentZ=7.5124, center=<6.4212 0 0.0001>, rotation=<rot 0 0 -0.1696>}]
 dyn_hair_bk_l_00 [{shapeType=Capsule, length=7.5628, radius=1.6556, shapeName=dyn_hair_bk_l_00_capsule, center=<2.3957 0 -0.0002>}]
 dyn_hair_bk_l_01 [{shapeType=Capsule, length=7.3009, radius=1.3141, shapeName=dyn_hair_bk_l_01_capsule, center=<1.7736 0 0.0002>}]
 dyn_hair_bk_l_02 [{shapeType=Box, extentX=5.9004, extentY=1.8791, extentZ=7.5124, center=<2.5078 0 -0.0001>, rotation=<rot 0 0 -0.1696>}]
 dyn_hair_bk_l_03 [{shapeType=Capsule, length=3.5, radius=1.5, center=<0.4061 -0.0001 -0.0003>, rotation=<rot 0 0 -0.1696>}]
 dyn_hair_bk_r_00 [{shapeType=Capsule, length=7.5628, radius=1.6556, shapeName=dyn_hair_bk_r_00_capsule, center=<2.3957 0 -0.0002>}]
 dyn_hair_bk_r_01 [{shapeType=Capsule, length=7.3009, radius=1.3141, shapeName=dyn_hair_bk_r_01_capsule, center=<1.7736 0 0.0002>}]
 dyn_hair_bk_r_02 [{shapeType=Box, extentX=9.9004, extentY=1.8791, extentZ=7.5124, center=<4.4192 0 -0.0001>, rotation=<rot 0 0 -0.1696>}]
 dyn_hair_bk_r_03 [{shapeType=Capsule, length=4.8575, radius=1.5508, center=<2.2957 0.5168 -0.3759>, rotation=<rot -34.0021 0 -0.1697>}]
 dyn_hair_fl_l_01 [{shapeType=Capsule, length=7.5628, radius=1.6556, shapeName=dyn_hair_fl_l_01_capsule, center=<2.3957 0 -0.0002>}]
 dyn_hair_fl_l_02 [{shapeType=Capsule, length=7.3009, radius=1.3141, shapeName=dyn_hair_fl_l_02_capsule, center=<1.7736 0 0.0002>}]
 dyn_hair_fl_l_03 [{shapeType=Capsule, length=5.9121, radius=1.0879, shapeName=dyn_hair_fl_l_03_capsule, center=<3.0023 0.4143 0>}]
 dyn_hair_fl_l_04 [{shapeType=Capsule, length=3.5, radius=1.5, center=<0.484 0 -0.0003>, rotation=<rot 0 0 -0.1696>}]
 dyn_hair_fr_r_00 [{shapeType=Capsule, length=7.5628, radius=1.6556, shapeName=dyn_hair_fr_r_00_capsule, center=<2.3957 0 -0.0002>}]
 dyn_hair_fr_r_01 [{shapeType=Capsule, length=7.3009, radius=1.3141, shapeName=dyn_hair_fr_r_01_capsule, center=<1.7736 0 0.0002>}]
 dyn_hair_fr_r_02 [{shapeType=Capsule, length=8.2467, radius=1.7533, shapeName=dyn_hair_fr_r_02_capsule, center=<2.6552 0 0>}]
 dyn_hair_l_00 [{shapeType=Capsule, length=7.5628, radius=1.6556, shapeName=dyn_hair_l_00_capsule, center=<2.3957 0 -0.0002>}]
 dyn_hair_l_01 [{shapeType=Capsule, length=7.3009, radius=1.3141, shapeName=dyn_hair_l_01_capsule, center=<1.7736 0 0.0002>}]
 dyn_hair_l_02 [{shapeType=Capsule, length=4.8944, radius=1.6607, shapeName=dyn_hair_l_02_capsule, center=<2.8179 0 0>}]
 dyn_hair_l_03 [{shapeType=Capsule, length=5.375, radius=1.5, center=<-0.3336 0 -0.0002>, rotation=<rot 0 0 -0.1696>}]
 dyn_hair_r_00 [{shapeType=Capsule, length=7.5628, radius=1.6556, shapeName=dyn_hair_r_00_capsule, center=<2.3957 0 -0.0002>}]
 dyn_hair_r_01 [{shapeType=Capsule, length=7.3009, radius=1.3141, shapeName=dyn_hair_r_01_capsule, center=<1.7736 0 0.0002>}]
 dyn_hair_r_02 [{shapeType=Capsule, length=7.9021, radius=1.5023, shapeName=dyn_hair_r_02_capsule, center=<2.1722 0 0>}]
 dyn_hair_r_03 [{shapeType=Capsule, length=8.1228, radius=1.5508, center=<0.9672 -0.0002 -0.0003>, rotation=<rot 0 0 -0.1696>}]
 head [
  {shapeType=Sphere, radius=5.9177, center=<2.1887 -1.5458 -0.2821>}
  {shapeType=Capsule, length=6.2157, radius=5.272, center=<0.8118 3.4354 0>, rotation=<rot 90 0 0>}
  {shapeType=Capsule, length=6.7157, radius=4.772, center=<0.8118 1.1771 0>, rotation=<rot 90 0 0>}
  {shapeType=Capsule, length=5.9657, radius=4.522, center=<3.9622 -1.248 -0.1194>, rotation=<rot 90 0 -10.0004>}
 ]
 neck_01 [{shapeType=Capsule, length=3.5, radius=5.1364, center=<0.1189 1.9388 -0.001>, rotation=<rot 90 0 0>}]
 spine_05 [{shapeType=Capsule, length=25.75, radius=6.4679, center=<6.2857 -0.1194 -0.5296>}]
}
constraints [
 {bone1Name=dyn_hair_fr_r_02, bone2Name=dyn_hair_fr_r_01, swing1LimitDegrees=40, swing2LimitDegrees=40, twistLimitDegrees=10}
 {bone1Name=dyn_hair_r_02, bone2Name=dyn_hair_r_01, swing1LimitDegrees=40, swing2LimitDegrees=15, twistLimitDegrees=10}
 {bone1Name=dyn_hair_r_03, bone2Name=dyn_hair_r_02, swing1LimitDegrees=40, swing2LimitDegrees=15, twistLimitDegrees=10}
 {bone1Name=dyn_hair_bk_r_02, bone2Name=dyn_hair_bk_r_01, swing1LimitDegrees=40, swing2LimitDegrees=5, twistLimitDegrees=10}
 {bone1Name=dyn_hair_bk_r_03, bone2Name=dyn_hair_bk_r_02, swing1LimitDegrees=40, swing2LimitDegrees=5, twistLimitDegrees=10}
 {bone1Name=dyn_hair_l_02, bone2Name=dyn_hair_l_01, swing1LimitDegrees=40, swing2LimitDegrees=15, twistLimitDegrees=10}
 {bone1Name=dyn_hair_l_03, bone2Name=dyn_hair_l_02, swing1LimitDegrees=40, swing2LimitDegrees=5, twistLimitDegrees=10}
 {bone1Name=dyn_hair_bk_02, bone2Name=dyn_hair_bk_01, swing1LimitDegrees=40, swing2LimitDegrees=3, twistLimitDegrees=10}
 {bone1Name=dyn_hair_bk_03, bone2Name=dyn_hair_bk_02, swing1LimitDegrees=40, swing2LimitDegrees=3, twistLimitDegrees=10}
 {bone1Name=dyn_hair_bk_l_02, bone2Name=dyn_hair_bk_l_01, swing1LimitDegrees=40, swing2LimitDegrees=5, twistLimitDegrees=10}
 {bone1Name=dyn_hair_bk_l_03, bone2Name=dyn_hair_bk_l_02, swing1LimitDegrees=40, swing2LimitDegrees=15, twistLimitDegrees=10}
 {bone1Name=dyn_hair_fl_l_04, bone2Name=dyn_hair_fl_l_03, swing1LimitDegrees=40, swing2LimitDegrees=40, twistLimitDegrees=10}
 {bone1Name=dyn_hair_fl_l_02, bone2Name=dyn_hair_fl_l_01, swing1LimitDegrees=40, swing2LimitDegrees=40, twistLimitDegrees=10}
 {bone1Name=dyn_hair_fl_l_03, bone2Name=dyn_hair_fl_l_02, swing1LimitDegrees=40, swing2LimitDegrees=40, twistLimitDegrees=10}
 {bone1Name=dyn_hair_fr_r_01, bone2Name=dyn_hair_fr_r_00, swing1LimitDegrees=40, swing2LimitDegrees=40, twistLimitDegrees=10}
 {bone1Name=dyn_hair_r_01, bone2Name=dyn_hair_r_00, swing1LimitDegrees=40, swing2LimitDegrees=15, twistLimitDegrees=10}
 {bone1Name=dyn_hair_bk_r_01, bone2Name=dyn_hair_bk_r_00, swing1LimitDegrees=40, swing2LimitDegrees=5, twistLimitDegrees=10}
 {bone1Name=dyn_hair_l_01, bone2Name=dyn_hair_l_00, swing1LimitDegrees=40, swing2LimitDegrees=15, twistLimitDegrees=10}
 {bone1Name=dyn_hair_bk_01, bone2Name=dyn_hair_bk_00, swing1LimitDegrees=40, swing2LimitDegrees=3, twistLimitDegrees=10}
 {bone1Name=dyn_hair_bk_l_01, bone2Name=dyn_hair_bk_l_00, swing1LimitDegrees=40, swing2LimitDegrees=5, twistLimitDegrees=10}
 {bone1Name=dyn_hair_fl_l_01, bone2Name=head, swing1LimitDegrees=40, swing2LimitDegrees=40, twistLimitDegrees=10}
 {
  bone1Name=neck_01
  bone2Name=spine_05
  swing1LimitDegrees=45
  swing1Motion=Free
  swing2LimitDegrees=45
  swing2Motion=Free
  twistLimitDegrees=45
  twistMotion=Free
 }
 {
  bone1Name=head
  bone2Name=neck_01
  swing1LimitDegrees=45
  swing1Motion=Free
  swing2LimitDegrees=45
  swing2Motion=Free
  twistLimitDegrees=45
  twistMotion=Free
 }
 {bone1Name=dyn_hair_l_00, bone2Name=head, swing1LimitDegrees=40, swing2LimitDegrees=10, twistLimitDegrees=10}
 {bone1Name=dyn_hair_bk_00, bone2Name=head, swing1LimitDegrees=40, swing2LimitDegrees=3, twistLimitDegrees=10}
 {bone1Name=dyn_hair_bk_l_00, bone2Name=head, swing1LimitDegrees=40, swing2LimitDegrees=15, twistLimitDegrees=10}
 {bone1Name=dyn_hair_bk_r_00, bone2Name=head, swing1LimitDegrees=40, swing2LimitDegrees=5, twistLimitDegrees=10}
 {bone1Name=dyn_hair_r_00, bone2Name=head, swing1LimitDegrees=40, swing2LimitDegrees=5, twistLimitDegrees=10}
 {bone1Name=dyn_hair_fr_r_00, bone2Name=head, swing1LimitDegrees=40, swing2LimitDegrees=40, twistLimitDegrees=10}
 {
  bone1Name=clavicle_r
  bone2Name=spine_05
  swing1LimitDegrees=45
  swing1Motion=Free
  swing2LimitDegrees=45
  swing2Motion=Free
  twistLimitDegrees=45
  twistMotion=Free
 }
]
