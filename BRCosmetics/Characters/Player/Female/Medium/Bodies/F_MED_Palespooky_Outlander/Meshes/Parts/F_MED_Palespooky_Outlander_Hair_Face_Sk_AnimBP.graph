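# Animation Blueprint that adds secondary hair motion to the female Palespooky Outlander's hair and face cosmetic part.
# It copies the attached parent's pose and curves, runs a RigidBody physics-asset simulation, then chains spring AnimDynamics on front-left, left and front-right hair bones, with spherical limits around head, neck, upper spine and clavicles.
# Each update, a deprecated helper sets those nodes' gravity override.
# Back and right-side AnimDynamics chains are left disconnected from the output.
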AnimBlueprint /BRCosmetics/Characters/Player/Female/Medium/Bodies/F_MED_Palespooky_Outlander/Meshes/Parts/F_MED_Palespooky_Outlander_Hair_Face_Sk_AnimBP
target_skeleton=/Game/Characters/Player/Female/Medium/Bodies/F_MED_Palespooky_Outlander/Meshes/Parts/F_MED_Palespooky_Outlander_Hair_FaceAcc_Skeleton
graphs [
 AnimGraph {nodes [
   n1 = Anim Dynamics - Bone: dyn_hair_r_02!(ComponentPose, Alpha=1, GravityOverride:Vector="0,0,0", ExternalForce="0,0,0") -> (Pose) {
    angularDampingOverride=0.9999
    bAngularSpring=true
    bGravityOverrideInSimSpace=true
    bOverrideAngularDamping=true
    bUseGravityOverride=true
    bUseSphericalLimits=true
    simulationSpace=RootRelative
    alphaBoolBlend {blendOption=Linear}
    boundBone {boneName=dyn_hair_r_02}
    componentAppliedLinearAccClamp=<4000 4000 4000>
    componentLinearAccScale=<0.5 0.5 0.5>
    componentLinearVelScale=<0.5 0.5 0.5>
    physicsBodyDefinitions [{
      sphereCollisionRadius=0
      boundBone {boneName=dyn_hair_r_02}
      boxExtents=<8 16 16>
      constraintSetup {angularLimitsMax=<20 20 0>, angularLimitsMin=<-20 -20 -20>, angularTarget=<0 0 0>}
      localJointOffset=<2 0 0>
     }]
    sphericalLimits [
     {limitRadius=5, drivingBone {boneName=head}, sphereLocalOffset=<0 0 0>}
     {limitRadius=5, drivingBone {boneName=neck_01}, sphereLocalOffset=<0 0 0>}
     {limitRadius=5, drivingBone {boneName=neck_02}, sphereLocalOffset=<0 0 0>}
     {limitRadius=13, drivingBone {boneName=spine_05}, sphereLocalOffset=<0 0 0>}
     {limitRadius=9, drivingBone {boneName=clavicle_l}, sphereLocalOffset=<10 0 0>}
     {limitRadius=9, drivingBone {boneName=clavicle_r}, sphereLocalOffset=<-10 0 0>}
    ]
   }
   Anim Dynamics - Bone: dyn_hair_r_03!(ComponentPose=@n1, Alpha=1, GravityOverride:Vector="0,0,0", ExternalForce="0,0,0") -> (Pose) {
    angularDampingOverride=0.9999
    bAngularSpring=true
    bGravityOverrideInSimSpace=true
    bOverrideAngularDamping=true
    bUseGravityOverride=true
    bUseSphericalLimits=true
    simulationSpace=RootRelative
    alphaBoolBlend {blendOption=Linear}
    boundBone {boneName=dyn_hair_r_03}
    componentAppliedLinearAccClamp=<4000 4000 4000>
    componentLinearAccScale=<0.5 0.5 0.5>
    componentLinearVelScale=<0.5 0.5 0.5>
    physicsBodyDefinitions [{
      sphereCollisionRadius=0
      boundBone {boneName=dyn_hair_r_03}
      boxExtents=<8 16 16>
      constraintSetup {angularLimitsMax=<20 20 0>, angularLimitsMin=<-20 -20 -20>, angularTarget=<0 0 0>}
      localJointOffset=<2 0 0>
     }]
    sphericalLimits [
     {limitRadius=5, drivingBone {boneName=head}, sphereLocalOffset=<0 0 0>}
     {limitRadius=5, drivingBone {boneName=neck_01}, sphereLocalOffset=<0 0 0>}
     {limitRadius=5, drivingBone {boneName=neck_02}, sphereLocalOffset=<0 0 0>}
     {limitRadius=13, drivingBone {boneName=spine_05}, sphereLocalOffset=<0 0 0>}
     {limitRadius=9, drivingBone {boneName=clavicle_l}, sphereLocalOffset=<10 0 0>}
     {limitRadius=9, drivingBone {boneName=clavicle_r}, sphereLocalOffset=<-10 0 0>}
    ]
   }
   n2 = Anim Dynamics - Bone: dyn_hair_bk_00!(ComponentPose, Alpha=1, GravityOverride:Vector="0,0,0", ExternalForce="0,0,0") -> (Pose) {
    angularDampingOverride=0.9999
    bAngularSpring=true
    bGravityOverrideInSimSpace=true
    bOverrideAngularDamping=true
    bUseGravityOverride=true
    bUseSphericalLimits=true
    simulationSpace=RootRelative
    alphaBoolBlend {blendOption=Linear}
    boundBone {boneName=dyn_hair_bk_00}
    componentAppliedLinearAccClamp=<4000 4000 4000>
    componentLinearAccScale=<0.5 0.5 0.5>
    componentLinearVelScale=<0.5 0.5 0.5>
    physicsBodyDefinitions [{
      sphereCollisionRadius=0
      boundBone {boneName=dyn_hair_bk_00}
      boxExtents=<8 16 16>
      constraintSetup {angularLimitsMax=<20 20 0>, angularLimitsMin=<-20 -20 -20>, angularTarget=<0 0 0>}
      localJointOffset=<2 0 0>
     }]
    sphericalLimits [
     {limitRadius=5, drivingBone {boneName=head}, sphereLocalOffset=<0 0 0>}
     {limitRadius=5, drivingBone {boneName=neck_01}, sphereLocalOffset=<0 0 0>}
     {limitRadius=5, drivingBone {boneName=neck_02}, sphereLocalOffset=<0 0 0>}
     {limitRadius=13, drivingBone {boneName=spine_05}, sphereLocalOffset=<0 0 0>}
     {limitRadius=9, drivingBone {boneName=clavicle_l}, sphereLocalOffset=<10 0 0>}
     {limitRadius=9, drivingBone {boneName=clavicle_r}, sphereLocalOffset=<-10 0 0>}
    ]
   }
   n3 = Anim Dynamics - Bone: dyn_hair_bk_01!(ComponentPose=@n2, Alpha=1, GravityOverride:Vector="0,0,0", ExternalForce="0,0,0") -> (Pose) {
    angularDampingOverride=0.9999
    bAngularSpring=true
    bGravityOverrideInSimSpace=true
    bOverrideAngularDamping=true
    bUseGravityOverride=true
    bUseSphericalLimits=true
    simulationSpace=RootRelative
    alphaBoolBlend {blendOption=Linear}
    boundBone {boneName=dyn_hair_bk_01}
    componentAppliedLinearAccClamp=<4000 4000 4000>
    componentLinearAccScale=<0.5 0.5 0.5>
    componentLinearVelScale=<0.5 0.5 0.5>
    physicsBodyDefinitions [{
      sphereCollisionRadius=0
      boundBone {boneName=dyn_hair_bk_01}
      boxExtents=<8 16 16>
      constraintSetup {angularLimitsMax=<20 20 0>, angularLimitsMin=<-20 -20 -20>, angularTarget=<0 0 0>}
      localJointOffset=<2 0 0>
     }]
    sphericalLimits [
     {limitRadius=5, drivingBone {boneName=head}, sphereLocalOffset=<0 0 0>}
     {limitRadius=5, drivingBone {boneName=neck_01}, sphereLocalOffset=<0 0 0>}
     {limitRadius=5, drivingBone {boneName=neck_02}, sphereLocalOffset=<0 0 0>}
     {limitRadius=13, drivingBone {boneName=spine_05}, sphereLocalOffset=<0 0 0>}
     {limitRadius=9, drivingBone {boneName=clavicle_l}, sphereLocalOffset=<10 0 0>}
     {limitRadius=9, drivingBone {boneName=clavicle_r}, sphereLocalOffset=<-10 0 0>}
    ]
   }
   n4 = Anim Dynamics - Bone: dyn_hair_bk_r_00!(ComponentPose=@n3, Alpha=1, GravityOverride:Vector="0,0,0", ExternalForce="0,0,0") -> (Pose) {
    angularDampingOverride=0.9999
    bAngularSpring=true
    bGravityOverrideInSimSpace=true
    bOverrideAngularDamping=true
    bUseGravityOverride=true
    bUseSphericalLimits=true
    simulationSpace=RootRelative
    alphaBoolBlend {blendOption=Linear}
    boundBone {boneName=dyn_hair_bk_r_00}
    componentAppliedLinearAccClamp=<4000 4000 4000>
    componentLinearAccScale=<0.5 0.5 0.5>
    componentLinearVelScale=<0.5 0.5 0.5>
    physicsBodyDefinitions [{
      sphereCollisionRadius=0
      boundBone {boneName=dyn_hair_bk_r_00}
      boxExtents=<8 16 16>
      constraintSetup {angularLimitsMax=<20 20 0>, angularLimitsMin=<-20 -20 -20>, angularTarget=<0 0 0>}
      localJointOffset=<2 0 0>
     }]
    sphericalLimits [
     {limitRadius=5, drivingBone {boneName=head}, sphereLocalOffset=<0 0 0>}
     {limitRadius=5, drivingBone {boneName=neck_01}, sphereLocalOffset=<0 0 0>}
     {limitRadius=5, drivingBone {boneName=neck_02}, sphereLocalOffset=<0 0 0>}
     {limitRadius=13, drivingBone {boneName=spine_05}, sphereLocalOffset=<0 0 0>}
     {limitRadius=9, drivingBone {boneName=clavicle_l}, sphereLocalOffset=<10 0 0>}
     {limitRadius=9, drivingBone {boneName=clavicle_r}, sphereLocalOffset=<-10 0 0>}
    ]
   }
   n5 = Anim Dynamics - Bone: dyn_hair_bk_r_01!(ComponentPose=@n4, Alpha=1, GravityOverride:Vector="0,0,0", ExternalForce="0,0,0") -> (Pose) {
    angularDampingOverride=0.9999
    bAngularSpring=true
    bGravityOverrideInSimSpace=true
    bOverrideAngularDamping=true
    bUseGravityOverride=true
    bUseSphericalLimits=true
    simulationSpace=RootRelative
    alphaBoolBlend {blendOption=Linear}
    boundBone {boneName=dyn_hair_bk_r_01}
    componentAppliedLinearAccClamp=<4000 4000 4000>
    componentLinearAccScale=<0.5 0.5 0.5>
    componentLinearVelScale=<0.5 0.5 0.5>
    physicsBodyDefinitions [{
      sphereCollisionRadius=0
      boundBone {boneName=dyn_hair_bk_r_01}
      boxExtents=<8 16 16>
      constraintSetup {angularLimitsMax=<20 20 0>, angularLimitsMin=<-20 -20 -20>, angularTarget=<0 0 0>}
      localJointOffset=<2 0 0>
     }]
    sphericalLimits [
     {limitRadius=5, drivingBone {boneName=head}, sphereLocalOffset=<0 0 0>}
     {limitRadius=5, drivingBone {boneName=neck_01}, sphereLocalOffset=<0 0 0>}
     {limitRadius=5, drivingBone {boneName=neck_02}, sphereLocalOffset=<0 0 0>}
     {limitRadius=13, drivingBone {boneName=spine_05}, sphereLocalOffset=<0 0 0>}
     {limitRadius=9, drivingBone {boneName=clavicle_l}, sphereLocalOffset=<10 0 0>}
     {limitRadius=9, drivingBone {boneName=clavicle_r}, sphereLocalOffset=<-10 0 0>}
    ]
   }
   n6 = Anim Dynamics - Bone: dyn_hair_r_00!(ComponentPose=@n5, Alpha=1, GravityOverride:Vector="0,0,0", ExternalForce="0,0,0") -> (Pose) {
    angularDampingOverride=0.9999
    bAngularSpring=true
    bGravityOverrideInSimSpace=true
    bOverrideAngularDamping=true
    bUseGravityOverride=true
    bUseSphericalLimits=true
    simulationSpace=RootRelative
    alphaBoolBlend {blendOption=Linear}
    boundBone {boneName=dyn_hair_r_00}
    componentAppliedLinearAccClamp=<4000 4000 4000>
    componentLinearAccScale=<0.5 0.5 0.5>
    componentLinearVelScale=<0.5 0.5 0.5>
    physicsBodyDefinitions [{
      sphereCollisionRadius=0
      boundBone {boneName=dyn_hair_r_00}
      boxExtents=<8 16 16>
      constraintSetup {angularLimitsMax=<20 20 0>, angularLimitsMin=<-20 -20 -20>, angularTarget=<0 0 0>}
      localJointOffset=<2 0 0>
     }]
    sphericalLimits [
     {limitRadius=5, drivingBone {boneName=head}, sphereLocalOffset=<0 0 0>}
     {limitRadius=5, drivingBone {boneName=neck_01}, sphereLocalOffset=<0 0 0>}
     {limitRadius=5, drivingBone {boneName=neck_02}, sphereLocalOffset=<0 0 0>}
     {limitRadius=13, drivingBone {boneName=spine_05}, sphereLocalOffset=<0 0 0>}
     {limitRadius=9, drivingBone {boneName=clavicle_l}, sphereLocalOffset=<10 0 0>}
     {limitRadius=9, drivingBone {boneName=clavicle_r}, sphereLocalOffset=<-10 0 0>}
    ]
   }
   Anim Dynamics - Bone: dyn_hair_r_01!(ComponentPose=@n6, Alpha=1, GravityOverride:Vector="0,0,0", ExternalForce="0,0,0") -> (Pose) {
    angularDampingOverride=0.9999
    bAngularSpring=true
    bGravityOverrideInSimSpace=true
    bOverrideAngularDamping=true
    bUseGravityOverride=true
    bUseSphericalLimits=true
    simulationSpace=RootRelative
    alphaBoolBlend {blendOption=Linear}
    boundBone {boneName=dyn_hair_r_01}
    componentAppliedLinearAccClamp=<4000 4000 4000>
    componentLinearAccScale=<0.5 0.5 0.5>
    componentLinearVelScale=<0.5 0.5 0.5>
    physicsBodyDefinitions [{
      sphereCollisionRadius=0
      boundBone {boneName=dyn_hair_r_01}
      boxExtents=<8 16 16>
      constraintSetup {angularLimitsMax=<20 20 0>, angularLimitsMin=<-20 -20 -20>, angularTarget=<0 0 0>}
      localJointOffset=<2 0 0>
     }]
    sphericalLimits [
     {limitRadius=5, drivingBone {boneName=head}, sphereLocalOffset=<0 0 0>}
     {limitRadius=5, drivingBone {boneName=neck_01}, sphereLocalOffset=<0 0 0>}
     {limitRadius=5, drivingBone {boneName=neck_02}, sphereLocalOffset=<0 0 0>}
     {limitRadius=13, drivingBone {boneName=spine_05}, sphereLocalOffset=<0 0 0>}
     {limitRadius=9, drivingBone {boneName=clavicle_l}, sphereLocalOffset=<10 0 0>}
     {limitRadius=9, drivingBone {boneName=clavicle_r}, sphereLocalOffset=<-10 0 0>}
    ]
   }
   n7 = Anim Dynamics - Bone: dyn_hair_bk_02!(ComponentPose, Alpha=1, GravityOverride:Vector="0,0,0", ExternalForce="0,0,0") -> (Pose) {
    angularDampingOverride=0.9999
    bAngularSpring=true
    bGravityOverrideInSimSpace=true
    bOverrideAngularDamping=true
    bUseGravityOverride=true
    bUseSphericalLimits=true
    simulationSpace=RootRelative
    alphaBoolBlend {blendOption=Linear}
    boundBone {boneName=dyn_hair_bk_02}
    componentAppliedLinearAccClamp=<4000 4000 4000>
    componentLinearAccScale=<0.5 0.5 0.5>
    componentLinearVelScale=<0.5 0.5 0.5>
    physicsBodyDefinitions [{
      sphereCollisionRadius=0
      boundBone {boneName=dyn_hair_bk_02}
      boxExtents=<8 16 16>
      constraintSetup {angularLimitsMax=<20 20 0>, angularLimitsMin=<-20 -20 -20>, angularTarget=<0 0 0>}
      localJointOffset=<2 0 0>
     }]
    sphericalLimits [
     {limitRadius=5, drivingBone {boneName=head}, sphereLocalOffset=<0 0 0>}
     {limitRadius=5, drivingBone {boneName=neck_01}, sphereLocalOffset=<0 0 0>}
     {limitRadius=5, drivingBone {boneName=neck_02}, sphereLocalOffset=<0 0 0>}
     {limitRadius=13, drivingBone {boneName=spine_05}, sphereLocalOffset=<0 0 0>}
     {limitRadius=9, drivingBone {boneName=clavicle_l}, sphereLocalOffset=<10 0 0>}
     {limitRadius=9, drivingBone {boneName=clavicle_r}, sphereLocalOffset=<-10 0 0>}
    ]
   }
   Anim Dynamics - Bone: dyn_hair_bk_03!(ComponentPose=@n7, Alpha=1, GravityOverride:Vector="0,0,0", ExternalForce="0,0,0") -> (Pose) {
    angularDampingOverride=0.9999
    bAngularSpring=true
    bGravityOverrideInSimSpace=true
    bOverrideAngularDamping=true
    bUseGravityOverride=true
    bUseSphericalLimits=true
    simulationSpace=RootRelative
    alphaBoolBlend {blendOption=Linear}
    boundBone {boneName=dyn_hair_bk_03}
    componentAppliedLinearAccClamp=<4000 4000 4000>
    componentLinearAccScale=<0.5 0.5 0.5>
    componentLinearVelScale=<0.5 0.5 0.5>
    physicsBodyDefinitions [{
      sphereCollisionRadius=0
      boundBone {boneName=dyn_hair_bk_03}
      boxExtents=<8 16 16>
      constraintSetup {angularLimitsMax=<20 20 0>, angularLimitsMin=<-20 -20 -20>, angularTarget=<0 0 0>}
      localJointOffset=<2 0 0>
     }]
    sphericalLimits [
     {limitRadius=5, drivingBone {boneName=head}, sphereLocalOffset=<0 0 0>}
     {limitRadius=5, drivingBone {boneName=neck_01}, sphereLocalOffset=<0 0 0>}
     {limitRadius=5, drivingBone {boneName=neck_02}, sphereLocalOffset=<0 0 0>}
     {limitRadius=13, drivingBone {boneName=spine_05}, sphereLocalOffset=<0 0 0>}
     {limitRadius=9, drivingBone {boneName=clavicle_l}, sphereLocalOffset=<10 0 0>}
     {limitRadius=9, drivingBone {boneName=clavicle_r}, sphereLocalOffset=<-10 0 0>}
    ]
   }
   n8 = Anim Dynamics - Bone: dyn_hair_bk_r_02!(ComponentPose, Alpha=1, GravityOverride:Vector="0,0,0", ExternalForce="0,0,0") -> (Pose) {
    angularDampingOverride=0.9999
    bAngularSpring=true
    bGravityOverrideInSimSpace=true
    bOverrideAngularDamping=true
    bUseGravityOverride=true
    bUseSphericalLimits=true
    simulationSpace=RootRelative
    alphaBoolBlend {blendOption=Linear}
    boundBone {boneName=dyn_hair_bk_r_02}
    componentAppliedLinearAccClamp=<4000 4000 4000>
    componentLinearAccScale=<0.5 0.5 0.5>
    componentLinearVelScale=<0.5 0.5 0.5>
    physicsBodyDefinitions [{
      sphereCollisionRadius=0
      boundBone {boneName=dyn_hair_bk_r_02}
      boxExtents=<8 16 16>
      constraintSetup {angularLimitsMax=<20 20 0>, angularLimitsMin=<-20 -20 -20>, angularTarget=<0 0 0>}
      localJointOffset=<2 0 0>
     }]
    sphericalLimits [
     {limitRadius=5, drivingBone {boneName=head}, sphereLocalOffset=<0 0 0>}
     {limitRadius=5, drivingBone {boneName=neck_01}, sphereLocalOffset=<0 0 0>}
     {limitRadius=5, drivingBone {boneName=neck_02}, sphereLocalOffset=<0 0 0>}
     {limitRadius=13, drivingBone {boneName=spine_05}, sphereLocalOffset=<0 0 0>}
     {limitRadius=9, drivingBone {boneName=clavicle_l}, sphereLocalOffset=<10 0 0>}
     {limitRadius=9, drivingBone {boneName=clavicle_r}, sphereLocalOffset=<-10 0 0>}
    ]
   }
   Anim Dynamics - Bone: dyn_hair_bk_r_03!(ComponentPose=@n8, Alpha=1, GravityOverride:Vector="0,0,0", ExternalForce="0,0,0") -> (Pose) {
    angularDampingOverride=0.9999
    bAngularSpring=true
    bGravityOverrideInSimSpace=true
    bOverrideAngularDamping=true
    bUseGravityOverride=true
    bUseSphericalLimits=true
    simulationSpace=RootRelative
    alphaBoolBlend {blendOption=Linear}
    boundBone {boneName=dyn_hair_bk_r_03}
    componentAppliedLinearAccClamp=<4000 4000 4000>
    componentLinearAccScale=<0.5 0.5 0.5>
    componentLinearVelScale=<0.5 0.5 0.5>
    physicsBodyDefinitions [{
      sphereCollisionRadius=0
      boundBone {boneName=dyn_hair_bk_r_03}
      boxExtents=<8 16 16>
      constraintSetup {angularLimitsMax=<20 20 0>, angularLimitsMin=<-20 -20 -20>, angularTarget=<0 0 0>}
      localJointOffset=<2 0 0>
     }]
    sphericalLimits [
     {limitRadius=5, drivingBone {boneName=head}, sphereLocalOffset=<0 0 0>}
     {limitRadius=5, drivingBone {boneName=neck_01}, sphereLocalOffset=<0 0 0>}
     {limitRadius=5, drivingBone {boneName=neck_02}, sphereLocalOffset=<0 0 0>}
     {limitRadius=13, drivingBone {boneName=spine_05}, sphereLocalOffset=<0 0 0>}
     {limitRadius=9, drivingBone {boneName=clavicle_l}, sphereLocalOffset=<10 0 0>}
     {limitRadius=9, drivingBone {boneName=clavicle_r}, sphereLocalOffset=<-10 0 0>}
    ]
   }
   n9 = RigidBody!(
    ComponentPose=Local To Component(LocalPose=Copy Pose From Mesh() {bCopyCurves=true, bUseAttachedParent=true})
    Alpha
    OverrideWorldGravity:Vector="0,0,-980"
    ExternalForce
   ) -> (Pose) {
    bOverrideWorldGravity=true
    overridePhysicsAsset=/Game/Characters/Player/Female/Medium/Bodies/F_MED_Palespooky_Outlander/Meshes/Parts/F_MED_Palespooky_Outlander_Hair_FaceAcc_Physics
    simulationSpace=BaseBoneSpace
    alphaBoolBlend {blendOption=Linear}
    baseBoneRef {boneName=neck_01}
    componentAppliedLinearAccClamp=<2000 2000 2000>
    componentLinearAccScale=<0.3 0.35 0.35>
    componentLinearVelScale=<0.3 0.35 0.35>
    overrideWorldGravity {z=-980}
   }
   n10 = Get GravityOverride()
   n11 = Anim Dynamics - Bone: dyn_hair_fl_l_01!(ComponentPose=@n9, Alpha=1, GravityOverride:Vector=@n10, ExternalForce="0,0,0") -> (Pose) {
    angularDampingOverride=0.9999
    bAngularSpring=true
    bGravityOverrideInSimSpace=true
    bOverrideAngularDamping=true
    bUseGravityOverride=true
    bUseSphericalLimits=true
    simulationSpace=RootRelative
    alphaBoolBlend {blendOption=Linear}
    boundBone {boneName=dyn_hair_fl_l_01}
    componentAppliedLinearAccClamp=<4000 4000 4000>
    componentLinearAccScale=<0.5 0.5 0.5>
    componentLinearVelScale=<0.5 0.5 0.5>
    physicsBodyDefinitions [{
      sphereCollisionRadius=0
      boundBone {boneName=dyn_hair_fl_l_01}
      boxExtents=<8 16 16>
      constraintSetup {angularLimitsMax=<20 20 0>, angularLimitsMin=<-20 -20 -20>, angularTarget=<0 0 0>}
      localJointOffset=<2 0 0>
     }]
    sphericalLimits [
     {limitRadius=5, drivingBone {boneName=head}, sphereLocalOffset=<0 0 0>}
     {limitRadius=5, drivingBone {boneName=neck_01}, sphereLocalOffset=<0 0 0>}
     {limitRadius=5, drivingBone {boneName=neck_02}, sphereLocalOffset=<0 0 0>}
     {limitRadius=13, drivingBone {boneName=spine_05}, sphereLocalOffset=<0 0 0>}
     {limitRadius=9, drivingBone {boneName=clavicle_l}, sphereLocalOffset=<10 0 0>}
     {limitRadius=9, drivingBone {boneName=clavicle_r}, sphereLocalOffset=<-10 0 0>}
    ]
   }
   n12 = Anim Dynamics - Bone: dyn_hair_fl_l_02!(ComponentPose=@n11, Alpha=1, GravityOverride:Vector=@n10, ExternalForce="0,0,0") -> (Pose) {
    angularDampingOverride=0.9999
    bAngularSpring=true
    bGravityOverrideInSimSpace=true
    bOverrideAngularDamping=true
    bUseGravityOverride=true
    bUseSphericalLimits=true
    simulationSpace=RootRelative
    alphaBoolBlend {blendOption=Linear}
    boundBone {boneName=dyn_hair_fl_l_02}
    componentAppliedLinearAccClamp=<4000 4000 4000>
    componentLinearAccScale=<0.5 0.5 0.5>
    componentLinearVelScale=<0.5 0.5 0.5>
    physicsBodyDefinitions [{
      sphereCollisionRadius=0
      boundBone {boneName=dyn_hair_fl_l_02}
      boxExtents=<8 16 16>
      constraintSetup {angularLimitsMax=<20 20 0>, angularLimitsMin=<-20 -20 -20>, angularTarget=<0 0 0>}
      localJointOffset=<2 0 0>
     }]
    sphericalLimits [
     {limitRadius=5, drivingBone {boneName=head}, sphereLocalOffset=<0 0 0>}
     {limitRadius=5, drivingBone {boneName=neck_01}, sphereLocalOffset=<0 0 0>}
     {limitRadius=5, drivingBone {boneName=neck_02}, sphereLocalOffset=<0 0 0>}
     {limitRadius=13, drivingBone {boneName=spine_05}, sphereLocalOffset=<0 0 0>}
     {limitRadius=9, drivingBone {boneName=clavicle_l}, sphereLocalOffset=<10 0 0>}
     {limitRadius=9, drivingBone {boneName=clavicle_r}, sphereLocalOffset=<-10 0 0>}
    ]
   }
   n13 = Anim Dynamics - Bone: dyn_hair_fl_l_03!(ComponentPose=@n12, Alpha=1, GravityOverride:Vector=@n10, ExternalForce="0,0,0") -> (Pose) {
    angularDampingOverride=0.9999
    bAngularSpring=true
    bGravityOverrideInSimSpace=true
    bOverrideAngularDamping=true
    bUseGravityOverride=true
    bUseSphericalLimits=true
    simulationSpace=RootRelative
    alphaBoolBlend {blendOption=Linear}
    boundBone {boneName=dyn_hair_fl_l_03}
    componentAppliedLinearAccClamp=<4000 4000 4000>
    componentLinearAccScale=<0.5 0.5 0.5>
    componentLinearVelScale=<0.5 0.5 0.5>
    physicsBodyDefinitions [{
      sphereCollisionRadius=0
      boundBone {boneName=dyn_hair_fl_l_03}
      boxExtents=<8 16 16>
      constraintSetup {angularLimitsMax=<20 20 0>, angularLimitsMin=<-20 -20 -20>, angularTarget=<0 0 0>}
      localJointOffset=<2 0 0>
     }]
    sphericalLimits [
     {limitRadius=5, drivingBone {boneName=head}, sphereLocalOffset=<0 0 0>}
     {limitRadius=5, drivingBone {boneName=neck_01}, sphereLocalOffset=<0 0 0>}
     {limitRadius=5, drivingBone {boneName=neck_02}, sphereLocalOffset=<0 0 0>}
     {limitRadius=13, drivingBone {boneName=spine_05}, sphereLocalOffset=<0 0 0>}
     {limitRadius=9, drivingBone {boneName=clavicle_l}, sphereLocalOffset=<10 0 0>}
     {limitRadius=9, drivingBone {boneName=clavicle_r}, sphereLocalOffset=<-10 0 0>}
    ]
   }
   n14 = Anim Dynamics - Bone: dyn_hair_fl_l_04!(ComponentPose=@n13, Alpha=1, GravityOverride:Vector=@n10, ExternalForce="0,0,0") -> (Pose) {
    angularDampingOverride=0.9999
    bAngularSpring=true
    bGravityOverrideInSimSpace=true
    bOverrideAngularDamping=true
    bUseGravityOverride=true
    bUseSphericalLimits=true
    simulationSpace=RootRelative
    alphaBoolBlend {blendOption=Linear}
    boundBone {boneName=dyn_hair_fl_l_04}
    componentAppliedLinearAccClamp=<4000 4000 4000>
    componentLinearAccScale=<0.5 0.5 0.5>
    componentLinearVelScale=<0.5 0.5 0.5>
    physicsBodyDefinitions [{
      sphereCollisionRadius=0
      boundBone {boneName=dyn_hair_fl_l_04}
      boxExtents=<8 16 16>
      constraintSetup {angularLimitsMax=<20 20 0>, angularLimitsMin=<-20 -20 -20>, angularTarget=<0 0 0>}
      localJointOffset=<2 0 0>
     }]
    sphericalLimits [
     {limitRadius=5, drivingBone {boneName=head}, sphereLocalOffset=<0 0 0>}
     {limitRadius=5, drivingBone {boneName=neck_01}, sphereLocalOffset=<0 0 0>}
     {limitRadius=5, drivingBone {boneName=neck_02}, sphereLocalOffset=<0 0 0>}
     {limitRadius=13, drivingBone {boneName=spine_05}, sphereLocalOffset=<0 0 0>}
     {limitRadius=9, drivingBone {boneName=clavicle_l}, sphereLocalOffset=<10 0 0>}
     {limitRadius=9, drivingBone {boneName=clavicle_r}, sphereLocalOffset=<-10 0 0>}
    ]
   }
   n15 = Anim Dynamics - Bone: dyn_hair_l_00!(ComponentPose=@n14, Alpha=1, GravityOverride:Vector=@n10, ExternalForce="0,0,0") -> (Pose) {
    angularDampingOverride=0.9999
    bAngularSpring=true
    bGravityOverrideInSimSpace=true
    bOverrideAngularDamping=true
    bUseGravityOverride=true
    bUseSphericalLimits=true
    simulationSpace=RootRelative
    alphaBoolBlend {blendOption=Linear}
    boundBone {boneName=dyn_hair_l_00}
    componentAppliedLinearAccClamp=<4000 4000 4000>
    componentLinearAccScale=<0.5 0.5 0.5>
    componentLinearVelScale=<0.5 0.5 0.5>
    physicsBodyDefinitions [{
      sphereCollisionRadius=0
      boundBone {boneName=dyn_hair_l_00}
      boxExtents=<8 16 16>
      constraintSetup {angularLimitsMax=<20 20 0>, angularLimitsMin=<-20 -20 -20>, angularTarget=<0 0 0>}
      localJointOffset=<2 0 0>
     }]
    sphericalLimits [
     {limitRadius=5, drivingBone {boneName=head}, sphereLocalOffset=<0 0 0>}
     {limitRadius=5, drivingBone {boneName=neck_01}, sphereLocalOffset=<0 0 0>}
     {limitRadius=5, drivingBone {boneName=neck_02}, sphereLocalOffset=<0 0 0>}
     {limitRadius=13, drivingBone {boneName=spine_05}, sphereLocalOffset=<0 0 0>}
     {limitRadius=9, drivingBone {boneName=clavicle_l}, sphereLocalOffset=<10 0 0>}
     {limitRadius=9, drivingBone {boneName=clavicle_r}, sphereLocalOffset=<-10 0 0>}
    ]
   }
   n16 = Anim Dynamics - Bone: dyn_hair_l_01!(ComponentPose=@n15, Alpha=1, GravityOverride:Vector=@n10, ExternalForce="0,0,0") -> (Pose) {
    angularDampingOverride=0.9999
    bAngularSpring=true
    bGravityOverrideInSimSpace=true
    bOverrideAngularDamping=true
    bUseGravityOverride=true
    bUseSphericalLimits=true
    simulationSpace=RootRelative
    alphaBoolBlend {blendOption=Linear}
    boundBone {boneName=dyn_hair_l_01}
    componentAppliedLinearAccClamp=<4000 4000 4000>
    componentLinearAccScale=<0.5 0.5 0.5>
    componentLinearVelScale=<0.5 0.5 0.5>
    physicsBodyDefinitions [{
      sphereCollisionRadius=0
      boundBone {boneName=dyn_hair_l_01}
      boxExtents=<8 16 16>
      constraintSetup {angularLimitsMax=<20 20 0>, angularLimitsMin=<-20 -20 -20>, angularTarget=<0 0 0>}
      localJointOffset=<2 0 0>
     }]
    sphericalLimits [
     {limitRadius=5, drivingBone {boneName=head}, sphereLocalOffset=<0 0 0>}
     {limitRadius=5, drivingBone {boneName=neck_01}, sphereLocalOffset=<0 0 0>}
     {limitRadius=5, drivingBone {boneName=neck_02}, sphereLocalOffset=<0 0 0>}
     {limitRadius=13, drivingBone {boneName=spine_05}, sphereLocalOffset=<0 0 0>}
     {limitRadius=9, drivingBone {boneName=clavicle_l}, sphereLocalOffset=<10 0 0>}
     {limitRadius=9, drivingBone {boneName=clavicle_r}, sphereLocalOffset=<-10 0 0>}
    ]
   }
   n17 = Anim Dynamics - Bone: dyn_hair_l_02!(ComponentPose=@n16, Alpha=1, GravityOverride:Vector=@n10, ExternalForce="0,0,0") -> (Pose) {
    angularDampingOverride=0.9999
    bAngularSpring=true
    bGravityOverrideInSimSpace=true
    bOverrideAngularDamping=true
    bUseGravityOverride=true
    bUseSphericalLimits=true
    simulationSpace=RootRelative
    alphaBoolBlend {blendOption=Linear}
    boundBone {boneName=dyn_hair_l_02}
    componentAppliedLinearAccClamp=<4000 4000 4000>
    componentLinearAccScale=<0.5 0.5 0.5>
    componentLinearVelScale=<0.5 0.5 0.5>
    physicsBodyDefinitions [{
      sphereCollisionRadius=0
      boundBone {boneName=dyn_hair_l_02}
      boxExtents=<8 16 16>
      constraintSetup {angularLimitsMax=<20 20 0>, angularLimitsMin=<-20 -20 -20>, angularTarget=<0 0 0>}
      localJointOffset=<2 0 0>
     }]
    sphericalLimits [
     {limitRadius=5, drivingBone {boneName=head}, sphereLocalOffset=<0 0 0>}
     {limitRadius=5, drivingBone {boneName=neck_01}, sphereLocalOffset=<0 0 0>}
     {limitRadius=5, drivingBone {boneName=neck_02}, sphereLocalOffset=<0 0 0>}
     {limitRadius=13, drivingBone {boneName=spine_05}, sphereLocalOffset=<0 0 0>}
     {limitRadius=9, drivingBone {boneName=clavicle_l}, sphereLocalOffset=<10 0 0>}
     {limitRadius=9, drivingBone {boneName=clavicle_r}, sphereLocalOffset=<-10 0 0>}
    ]
   }
   n18 = Anim Dynamics - Bone: dyn_hair_l_03!(ComponentPose=@n17, Alpha=1, GravityOverride:Vector=@n10, ExternalForce="0,0,0") -> (Pose) {
    angularDampingOverride=0.9999
    bAngularSpring=true
    bGravityOverrideInSimSpace=true
    bOverrideAngularDamping=true
    bUseGravityOverride=true
    bUseSphericalLimits=true
    simulationSpace=RootRelative
    alphaBoolBlend {blendOption=Linear}
    boundBone {boneName=dyn_hair_l_03}
    componentAppliedLinearAccClamp=<4000 4000 4000>
    componentLinearAccScale=<0.5 0.5 0.5>
    componentLinearVelScale=<0.5 0.5 0.5>
    physicsBodyDefinitions [{
      sphereCollisionRadius=0
      boundBone {boneName=dyn_hair_l_03}
      boxExtents=<8 16 16>
      constraintSetup {angularLimitsMax=<20 20 0>, angularLimitsMin=<-20 -20 -20>, angularTarget=<0 0 0>}
      localJointOffset=<2 0 0>
     }]
    sphericalLimits [
     {limitRadius=5, drivingBone {boneName=head}, sphereLocalOffset=<0 0 0>}
     {limitRadius=5, drivingBone {boneName=neck_01}, sphereLocalOffset=<0 0 0>}
     {limitRadius=5, drivingBone {boneName=neck_02}, sphereLocalOffset=<0 0 0>}
     {limitRadius=13, drivingBone {boneName=spine_05}, sphereLocalOffset=<0 0 0>}
     {limitRadius=9, drivingBone {boneName=clavicle_l}, sphereLocalOffset=<10 0 0>}
     {limitRadius=9, drivingBone {boneName=clavicle_r}, sphereLocalOffset=<-10 0 0>}
    ]
   }
   n19 = Anim Dynamics - Bone: dyn_hair_fr_r_00!(ComponentPose=@n18, Alpha=1, GravityOverride:Vector=@n10, ExternalForce="0,0,0") -> (Pose) {
    angularDampingOverride=0.9999
    bAngularSpring=true
    bGravityOverrideInSimSpace=true
    bOverrideAngularDamping=true
    bUseGravityOverride=true
    bUseSphericalLimits=true
    simulationSpace=RootRelative
    alphaBoolBlend {blendOption=Linear}
    boundBone {boneName=dyn_hair_fr_r_00}
    componentAppliedLinearAccClamp=<4000 4000 4000>
    componentLinearAccScale=<0.5 0.5 0.5>
    componentLinearVelScale=<0.5 0.5 0.5>
    physicsBodyDefinitions [{
      sphereCollisionRadius=0
      boundBone {boneName=dyn_hair_fr_r_00}
      boxExtents=<8 16 16>
      constraintSetup {angularLimitsMax=<20 20 0>, angularLimitsMin=<-20 -20 -20>, angularTarget=<0 0 0>}
      localJointOffset=<2 0 0>
     }]
    sphericalLimits [
     {limitRadius=5, drivingBone {boneName=head}, sphereLocalOffset=<0 0 0>}
     {limitRadius=5, drivingBone {boneName=neck_01}, sphereLocalOffset=<0 0 0>}
     {limitRadius=5, drivingBone {boneName=neck_02}, sphereLocalOffset=<0 0 0>}
     {limitRadius=13, drivingBone {boneName=spine_05}, sphereLocalOffset=<0 0 0>}
     {limitRadius=9, drivingBone {boneName=clavicle_l}, sphereLocalOffset=<10 0 0>}
     {limitRadius=9, drivingBone {boneName=clavicle_r}, sphereLocalOffset=<-10 0 0>}
    ]
   }
   n20 = Anim Dynamics - Bone: dyn_hair_fr_r_01!(ComponentPose=@n19, Alpha=1, GravityOverride:Vector=@n10, ExternalForce="0,0,0") -> (Pose) {
    angularDampingOverride=0.9999
    bAngularSpring=true
    bGravityOverrideInSimSpace=true
    bOverrideAngularDamping=true
    bUseGravityOverride=true
    bUseSphericalLimits=true
    simulationSpace=RootRelative
    alphaBoolBlend {blendOption=Linear}
    boundBone {boneName=dyn_hair_fr_r_01}
    componentAppliedLinearAccClamp=<4000 4000 4000>
    componentLinearAccScale=<0.5 0.5 0.5>
    componentLinearVelScale=<0.5 0.5 0.5>
    physicsBodyDefinitions [{
      sphereCollisionRadius=0
      boundBone {boneName=dyn_hair_fr_r_01}
      boxExtents=<8 16 16>
      constraintSetup {angularLimitsMax=<20 20 0>, angularLimitsMin=<-20 -20 -20>, angularTarget=<0 0 0>}
      localJointOffset=<2 0 0>
     }]
    sphericalLimits [
     {limitRadius=5, drivingBone {boneName=head}, sphereLocalOffset=<0 0 0>}
     {limitRadius=5, drivingBone {boneName=neck_01}, sphereLocalOffset=<0 0 0>}
     {limitRadius=5, drivingBone {boneName=neck_02}, sphereLocalOffset=<0 0 0>}
     {limitRadius=13, drivingBone {boneName=spine_05}, sphereLocalOffset=<0 0 0>}
     {limitRadius=9, drivingBone {boneName=clavicle_l}, sphereLocalOffset=<10 0 0>}
     {limitRadius=9, drivingBone {boneName=clavicle_r}, sphereLocalOffset=<-10 0 0>}
    ]
   }
   n21 = Anim Dynamics - Bone: dyn_hair_fr_r_02!(ComponentPose=@n20, Alpha=1, GravityOverride:Vector=@n10, ExternalForce="0,0,0") -> (Pose) {
    angularDampingOverride=0.9999
    bAngularSpring=true
    bGravityOverrideInSimSpace=true
    bOverrideAngularDamping=true
    bUseGravityOverride=true
    bUseSphericalLimits=true
    simulationSpace=RootRelative
    alphaBoolBlend {blendOption=Linear}
    boundBone {boneName=dyn_hair_fr_r_02}
    componentAppliedLinearAccClamp=<4000 4000 4000>
    componentLinearAccScale=<0.5 0.5 0.5>
    componentLinearVelScale=<0.5 0.5 0.5>
    physicsBodyDefinitions [{
      sphereCollisionRadius=0
      boundBone {boneName=dyn_hair_fr_r_02}
      boxExtents=<8 16 16>
      constraintSetup {angularLimitsMax=<20 20 0>, angularLimitsMin=<-20 -20 -20>, angularTarget=<0 0 0>}
      localJointOffset=<2 0 0>
     }]
    sphericalLimits [
     {limitRadius=5, drivingBone {boneName=head}, sphereLocalOffset=<0 0 0>}
     {limitRadius=5, drivingBone {boneName=neck_01}, sphereLocalOffset=<0 0 0>}
     {limitRadius=5, drivingBone {boneName=neck_02}, sphereLocalOffset=<0 0 0>}
     {limitRadius=13, drivingBone {boneName=spine_05}, sphereLocalOffset=<0 0 0>}
     {limitRadius=9, drivingBone {boneName=clavicle_l}, sphereLocalOffset=<10 0 0>}
     {limitRadius=9, drivingBone {boneName=clavicle_r}, sphereLocalOffset=<-10 0 0>}
    ]
   }
   Output Pose(Result=Component To Local(ComponentPose=@n21)) {layerGroup=None}
  ]}
 EventGraph {nodes [
   Event BlueprintUpdateAnimation()
   > n1 = Gravity Override State Function DEPRECATED(CharacterPartAnimInstance=Self-Reference(), GameTimeInSeconds=GetGameTimeInSeconds(), GravityOverrideParams=Get Gravity Override Params())
   > Set GravityOverride(GravityOverride=@n1)
   TryGetPawnOwner()
  ]}
]
variables [
 Gravity Override Params {type=GravityOverrideParamsStruct}
 GravityOverride {type=Vector}
]
